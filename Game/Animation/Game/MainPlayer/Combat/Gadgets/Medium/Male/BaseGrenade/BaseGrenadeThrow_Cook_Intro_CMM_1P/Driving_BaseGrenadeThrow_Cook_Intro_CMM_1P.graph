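# Level Sequence for authoring the first-person intro of the base grenade cook-and-throw animation for the medium male player.
# A Fortnite_Biped_CtrlRig track poses one skeletal mesh actor, and the skeletal animation track has no sections.
# Dense keys over frames 0–21 on both FK and IK leg controls suggest baked animation.
# The FK/IK switches hold the legs in IK and the arms, spine and neck in FK.
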LevelSequence /Game/Animation/Game/MainPlayer/Combat/Gadgets/Medium/Male/BaseGrenade/BaseGrenadeThrow_Cook_Intro_CMM_1P/Driving_BaseGrenadeThrow_Cook_Intro_CMM_1P
bindings ["BaseGrenadeThrow_Cook_Intro_CMM_1P --> BaseGrenadeThrow_Cook_Intro_CMM_1P" {
  id=#1
  possessed_class=SkeletalMeshActor
  tracks [
   {
    type=MovieScene3DTransformTrack
    sections [{
      type=MovieScene3DTransformSection
      channels [
       Location.X {type=MovieSceneScriptingDoubleChannel, default=0}
       Location.Y {type=MovieSceneScriptingDoubleChannel, default=0}
       Location.Z {type=MovieSceneScriptingDoubleChannel, default=0}
       Rotation.X {type=MovieSceneScriptingDoubleChannel, default=0}
       Rotation.Y {type=MovieSceneScriptingDoubleChannel, default=0}
       Rotation.Z {type=MovieSceneScriptingDoubleChannel, default=0}
       Scale.X {type=MovieSceneScriptingDoubleChannel, default=1}
       Scale.Y {type=MovieSceneScriptingDoubleChannel, default=1}
       Scale.Z {type=MovieSceneScriptingDoubleChannel, default=1}
      ]
     }]
   }
   {type=MovieSceneSkeletalAnimationTrack, sections []}
   {type=MovieSceneSpawnTrack, sections [{type=MovieSceneSpawnSection, channels [None {type=MovieSceneScriptingBoolChannel, default=true}]}]}
   {
    type=MovieSceneControlRigParameterTrack
    display_name=Fortnite_Biped_CtrlRig
    sections [
     {
      type=MovieSceneControlRigParameterSection
      channels [
       leg_l_stretch_switch {type=MovieSceneScriptingBoolChannel, default=false, const [false,22]}
       leg_r_stretch_switch {type=MovieSceneScriptingBoolChannel, default=false, const [false,22]}
       arm_l_fk_local {type=MovieSceneScriptingBoolChannel, default=true, const [true,22]}
       Auto Interpolate_3 {type=MovieSceneScriptingBoolChannel, default=true, const [true,22]}
       Auto Interpolate {type=MovieSceneScriptingBoolChannel, default=true, const [true,22]}
       arm_l_stretch_switch {type=MovieSceneScriptingBoolChannel, default=false, const [false,22]}
       Keep Aligned {type=MovieSceneScriptingBoolChannel, default=false, const [false,22]}
       arm_r_fk_local {type=MovieSceneScriptingBoolChannel, default=true, const [true,22]}
       Auto Interpolate_4 {type=MovieSceneScriptingBoolChannel, default=true, const [true,22]}
       Auto Interpolate_2 {type=MovieSceneScriptingBoolChannel, default=true, const [true,22]}
       arm_r_stretch_switch {type=MovieSceneScriptingBoolChannel, default=false, const [false,22]}
       Keep Aligned_2 {type=MovieSceneScriptingBoolChannel, default=false, const [false,22]}
       head_local {type=MovieSceneScriptingBoolChannel, default=true, const [true,22]}
       Stretch {type=MovieSceneScriptingBoolChannel, default=false, const [false,22]}
       head_stretch {type=MovieSceneScriptingBoolChannel, default=false, const [false,22]}
       spine_stretch_switch {type=MovieSceneScriptingBoolChannel, default=false, const [false,22]}
       arm_l_fk_ik_switch {type=MovieSceneScriptingBoolChannel, default=false, const [false,22]}
       leg_l_fk_ik_switch {type=MovieSceneScriptingBoolChannel, default=true, const [true,22]}
       arm_r_fk_ik_switch {type=MovieSceneScriptingBoolChannel, default=false, const [false,22]}
       leg_r_fk_ik_switch {type=MovieSceneScriptingBoolChannel, default=true, const [true,22]}
       ShowBodyControls {type=MovieSceneScriptingBoolChannel, default=true, const [true,22]}
       spine_fk_ik_switch {type=MovieSceneScriptingBoolChannel, default=false, const [false,22]}
       neck_fk_ik_switch {type=MovieSceneScriptingBoolChannel, default=false, const [false,22]}
       space {type=MovieSceneScriptingByteChannel, default=1}
       global_ctrl.Location.X {type=MovieSceneScriptingFloatChannel, default=0, const [0,22]}
       global_ctrl.Location.Y {type=MovieSceneScriptingFloatChannel, default=0, const [0,22]}
       global_ctrl.Location.Z {type=MovieSceneScriptingFloatChannel, default=0, const [0,22]}
       global_ctrl.Rotation.X {type=MovieSceneScriptingFloatChannel, default=0, const [0,22]}
       global_ctrl.Rotation.Y {type=MovieSceneScriptingFloatChannel, default=0, const [0,22]}
       global_ctrl.Rotation.Z {type=MovieSceneScriptingFloatChannel, default=0, const [0,22]}
       root_ctrl.Location.X {type=MovieSceneScriptingFloatChannel, default=0, const [0,22]}
       root_ctrl.Location.Y {type=MovieSceneScriptingFloatChannel, default=0, const [0,22]}
       root_ctrl.Location.Z {type=MovieSceneScriptingFloatChannel, default=0, const [0,22]}
       root_ctrl.Rotation.X {type=MovieSceneScriptingFloatChannel, default=0, const [0,22]}
       root_ctrl.Rotation.Y {type=MovieSceneScriptingFloatChannel, default=0, const [0,22]}
       root_ctrl.Rotation.Z {type=MovieSceneScriptingFloatChannel, default=0, const [0,22]}
       root_ctrl.Scale.X {type=MovieSceneScriptingFloatChannel, default=1, const [1,22]}
       root_ctrl.Scale.Y {type=MovieSceneScriptingFloatChannel, default=1, const [1,22]}
       root_ctrl.Scale.Z {type=MovieSceneScriptingFloatChannel, default=1, const [1,22]}
       body_offset_ctrl.Location.X {type=MovieSceneScriptingFloatChannel, default=0, const [0,22]}
       body_offset_ctrl.Location.Y {type=MovieSceneScriptingFloatChannel, default=0, const [0,22]}
       body_offset_ctrl.Location.Z {type=MovieSceneScriptingFloatChannel, default=0, const [0,22]}
       body_offset_ctrl.Rotation.X {type=MovieSceneScriptingFloatChannel, default=0, const [0,22]}
       body_offset_ctrl.Rotation.Y {type=MovieSceneScriptingFloatChannel, default=0, const [0,22]}
       body_offset_ctrl.Rotation.Z {type=MovieSceneScriptingFloatChannel, default=0, const [0,22]}
       body_offset_ctrl.Scale.X {type=MovieSceneScriptingFloatChannel, default=1, const [1,22]}
       body_offset_ctrl.Scale.Y {type=MovieSceneScriptingFloatChannel, default=1, const [1,22]}
       body_offset_ctrl.Scale.Z {type=MovieSceneScriptingFloatChannel, default=1, const [1,22]}
       thigh_l_fk_ctrl.Location.X {type=MovieSceneScriptingFloatChannel, default=0, thinned_from=22, keys [[0,0],[21,0]]}
       thigh_l_fk_ctrl.Location.Y {type=MovieSceneScriptingFloatChannel, default=0, thinned_from=22, keys [[0,0],[21,0]]}
       thigh_l_fk_ctrl.Location.Z {type=MovieSceneScriptingFloatChannel, default=0, thinned_from=22, keys [[0,0],[21,0]]}
       thigh_l_fk_ctrl.Rotation.X {
        type=MovieSceneScriptingFloatChannel
        default=16.9314
        thinned_from=22
        keys [[0,16.9314],[1,13.991],[2,17.2262],[3,18.6984],[4,20.5528],[5,21.8067],[6,21.6375],[7,22.1905],[8,22.5271],[9,22.6874],[10,22.7118],[11,22.6412],[12,22.5168],[13,22.2721],[14,21.8663],[17,20.3925],[18,20.8782],[19,20.4456],[20,20.1379],[21,20.0182]]
       }
       thigh_l_fk_ctrl.Rotation.Y {
        type=MovieSceneScriptingFloatChannel
        default=9.2953
        thinned_from=22
        keys [[0,9.2953],[1,7.3198],[2,7.2274],[3,7.7405],[4,8.666],[5,9.497],[6,9.8414],[7,10.2812],[8,10.5802],[9,10.7604],[10,10.8431],[11,10.8489],[12,10.7981],[13,10.6456],[14,10.3714],[17,9.358],[18,9.2987],[19,9.0386],[20,8.8623],[21,8.8032]]
       }
       thigh_l_fk_ctrl.Rotation.Z {
        type=MovieSceneScriptingFloatChannel
        default=-18.7057
        thinned_from=22
        keys [[0,-18.7057],[1,-19.0782],[2,-18.7199],[3,-18.2951],[5,-17.0613],[6,-16.6483],[7,-16.1392],[8,-15.6834],[9,-15.3027],[10,-15.0164],[11,-14.8432],[12,-14.8011],[13,-14.9165],[14,-15.1622],[17,-16.1216],[18,-16.3202],[19,-16.5919],[20,-16.7797],[21,-16.846]]
       }
       thigh_l_fk_ctrl.Scale.X {type=MovieSceneScriptingFloatChannel, default=1, const [1,22]}
       thigh_l_fk_ctrl.Scale.Y {type=MovieSceneScriptingFloatChannel, default=1, const [1,22]}
       thigh_l_fk_ctrl.Scale.Z {type=MovieSceneScriptingFloatChannel, default=1, const [1,22]}
       calf_l_fk_ctrl.Location.X {
        type=MovieSceneScriptingFloatChannel
        default=0.0796
        thinned_from=22
        keys [[0,0.0796],[1,0.0908],[2,0.1001],[3,0.1063],[4,0.1111],[5,0.1141],[6,0.1157],[7,0.1165],[8,0.1167],[9,0.1165],[13,0.1139],[16,0.1111],[18,0.1098],[20,0.109],[21,0.1088]]
       }
       calf_l_fk_ctrl.Location.Y {
        type=MovieSceneScriptingFloatChannel
        default=0.0001
        thinned_from=22
        keys [[0,0.0001],[2,0.005],[3,0.0069],[4,0.0085],[5,0.0096],[6,0.0101],[7,0.0104],[9,0.0104],[19,0.0079],[21,0.0077]]
       }
       calf_l_fk_ctrl.Location.Z {type=MovieSceneScriptingFloatChannel, default=0.0028, thinned_from=22, keys [[0,0.0028],[3,0.0036],[21,0.0037]]}
       calf_l_fk_ctrl.Rotation.X {
        type=MovieSceneScriptingFloatChannel
        default=0.6968
        thinned_from=22
        keys [[0,0.6968],[1,0.785],[2,0.8574],[3,0.9053],[4,0.9414],[5,0.964],[6,0.9756],[7,0.9816],[8,0.9831],[9,0.9814],[13,0.9624],[16,0.9414],[18,0.9314],[20,0.9253],[21,0.9238]]
       }
       calf_l_fk_ctrl.Rotation.Y {
        type=MovieSceneScriptingFloatChannel
        default=0.1836
        thinned_from=22
        keys [[0,0.1836],[1,0.1966],[2,0.2055],[3,0.2104],[4,0.2135],[5,0.2153],[7,0.2165],[13,0.2152],[20,0.2122],[21,0.212]]
       }
       calf_l_fk_ctrl.Rotation.Z {
        type=MovieSceneScriptingFloatChannel
        default=18.9614
        thinned_from=22
        keys [[0,18.9614],[1,21.6433],[2,23.9141],[3,25.4586],[4,26.6463],[5,27.3976],[6,27.79],[7,27.9898],[8,28.0409],[9,27.9828],[13,27.3438],[16,26.6457],[18,26.3138],[20,26.1146],[21,26.063]]
       }
       calf_l_fk_ctrl.Scale.X {type=MovieSceneScriptingFloatChannel, default=1, const [1,22]}
       calf_l_fk_ctrl.Scale.Y {type=MovieSceneScriptingFloatChannel, default=1, const [1,22]}
       calf_l_fk_ctrl.Scale.Z {type=MovieSceneScriptingFloatChannel, default=1, const [1,22]}
       foot_l_fk_ctrl.Location.X {type=MovieSceneScriptingFloatChannel, default=0, thinned_from=22, keys [[0,0],[21,0]]}
       foot_l_fk_ctrl.Location.Y {type=MovieSceneScriptingFloatChannel, default=0, thinned_from=22, keys [[0,0],[21,0]]}
       foot_l_fk_ctrl.Location.Z {type=MovieSceneScriptingFloatChannel, default=0, thinned_from=22, keys [[0,0],[21,0]]}
       foot_l_fk_ctrl.Rotation.X {
        type=MovieSceneScriptingFloatChannel
        default=-12.9811
        thinned_from=22
        keys [[0,-12.9811],[1,-5.7416],[2,-5.1728],[3,-4.452],[4,-4.649],[5,-4.7829],[6,-3.8915],[7,-3.9484],[8,-3.9833],[9,-4.0001],[10,-4.0025],[12,-3.9791],[13,-3.9515],[17,-3.7594],[18,-4.6863],[19,-4.6443],[20,-4.6148],[21,-4.6038]]
       }
       foot_l_fk_ctrl.Rotation.Y {
        type=MovieSceneScriptingFloatChannel
        default=-2.1686
        thinned_from=22
        keys [[0,-2.1686],[1,-4.0558],[2,-4.0908],[3,-4.2764],[4,-4.2807],[5,-4.3225],[6,-4.6067],[9,-4.754],[10,-4.7908],[11,-4.8162],[12,-4.8273],[13,-4.8204],[14,-4.7974],[17,-4.6857],[18,-4.41],[19,-4.3747],[20,-4.3537],[21,-4.3537]]
       }
       foot_l_fk_ctrl.Rotation.Z {
        type=MovieSceneScriptingFloatChannel
        default=5.1878
        keys [[0,5.1878],[1,2.9503],[2,1.2536],[3,0.0511],[4,-0.8379],[5,-1.4435],[6,-1.8893],[7,-2.1331],[8,-2.2685],[9,-2.3187],[10,-2.3054],[11,-2.2501],[12,-2.174],[13,-2.0475],[14,-1.8519],[15,-1.625],[16,-1.4061],[17,-1.2139],[18,-0.9738],[19,-0.8348],[20,-0.7327],[21,-0.6751]]
       }
       foot_l_fk_ctrl.Scale.X {type=MovieSceneScriptingFloatChannel, default=1, const [1,22]}
       foot_l_fk_ctrl.Scale.Y {type=MovieSceneScriptingFloatChannel, default=1, const [1,22]}
       foot_l_fk_ctrl.Scale.Z {type=MovieSceneScriptingFloatChannel, default=1, const [1,22]}
       ball_l_fk_ctrl.Location.X {type=MovieSceneScriptingFloatChannel, default=0, const [0,22]}
       ball_l_fk_ctrl.Location.Y {type=MovieSceneScriptingFloatChannel, default=0, const [0,22]}
       ball_l_fk_ctrl.Location.Z {type=MovieSceneScriptingFloatChannel, default=0, const [0,22]}
       ball_l_fk_ctrl.Rotation.X {type=MovieSceneScriptingFloatChannel, default=0, const [0,22]}
       ball_l_fk_ctrl.Rotation.Y {type=MovieSceneScriptingFloatChannel, default=0, const [0,22]}
       ball_l_fk_ctrl.Rotation.Z {type=MovieSceneScriptingFloatChannel, default=0, const [0,22]}
       ball_l_fk_ctrl.Scale.X {type=MovieSceneScriptingFloatChannel, default=1, const [1,22]}
       ball_l_fk_ctrl.Scale.Y {type=MovieSceneScriptingFloatChannel, default=1, const [1,22]}
       ball_l_fk_ctrl.Scale.Z {type=MovieSceneScriptingFloatChannel, default=1, const [1,22]}
       foot_l_ik_ctrl.Location.X {type=MovieSceneScriptingFloatChannel, default=6.4165, thinned_from=22, keys [[0,6.4165],[21,6.416]]}
       foot_l_ik_ctrl.Location.Y {type=MovieSceneScriptingFloatChannel, default=21.921, thinned_from=22, keys [[0,21.921],[21,21.9206]]}
       foot_l_ik_ctrl.Location.Z {type=MovieSceneScriptingFloatChannel, default=0.0001, thinned_from=22, keys [[0,0.0001],[1,-0.0002],[21,-0.0002]]}
       foot_l_ik_ctrl.Rotation.X {type=MovieSceneScriptingFloatChannel, default=0.027, thinned_from=22, keys [[0,0.027],[21,0.0269]]}
       foot_l_ik_ctrl.Rotation.Y {type=MovieSceneScriptingFloatChannel, default=0.1392, thinned_from=22, keys [[0,0.1392],[1,0.1387],[4,0.1389],[21,0.1388]]}
       foot_l_ik_ctrl.Rotation.Z {type=MovieSceneScriptingFloatChannel, default=10.5353, thinned_from=22, keys [[0,10.5353],[21,10.5353]]}
       foot_l_ik_ctrl.Scale.X {type=MovieSceneScriptingFloatChannel, default=1, const [1,22]}
       foot_l_ik_ctrl.Scale.Y {type=MovieSceneScriptingFloatChannel, default=1, const [1,22]}
       foot_l_ik_ctrl.Scale.Z {type=MovieSceneScriptingFloatChannel, default=1, const [1,22]}
       ball_l_ik_ctrl.Location.X {type=MovieSceneScriptingFloatChannel, default=0, const [0,22]}
       ball_l_ik_ctrl.Location.Y {type=MovieSceneScriptingFloatChannel, default=0, const [0,22]}
       ball_l_ik_ctrl.Location.Z {type=MovieSceneScriptingFloatChannel, default=0, const [0,22]}
       ball_l_ik_ctrl.Rotation.X {type=MovieSceneScriptingFloatChannel, default=0, const [0,22]}
       ball_l_ik_ctrl.Rotation.Y {type=MovieSceneScriptingFloatChannel, default=0, const [0,22]}
       ball_l_ik_ctrl.Rotation.Z {type=MovieSceneScriptingFloatChannel, default=0, const [0,22]}
       ball_l_ik_ctrl.Scale.X {type=MovieSceneScriptingFloatChannel, default=1, const [1,22]}
       ball_l_ik_ctrl.Scale.Y {type=MovieSceneScriptingFloatChannel, default=1, const [1,22]}
       ball_l_ik_ctrl.Scale.Z {type=MovieSceneScriptingFloatChannel, default=1, const [1,22]}
       foot_bk1_l_ctrl.Location.X {type=MovieSceneScriptingFloatChannel, default=0, const [0,22]}
       foot_bk1_l_ctrl.Location.Y {type=MovieSceneScriptingFloatChannel, default=0, const [0,22]}
       foot_bk1_l_ctrl.Location.Z {type=MovieSceneScriptingFloatChannel, default=0, const [0,22]}
       foot_bk1_l_ctrl.Rotation.X {type=MovieSceneScriptingFloatChannel, default=0, const [0,22]}
       foot_bk1_l_ctrl.Rotation.Y {type=MovieSceneScriptingFloatChannel, default=0, const [0,22]}
       foot_bk1_l_ctrl.Rotation.Z {type=MovieSceneScriptingFloatChannel, default=0, const [0,22]}
       foot_bk1_l_ctrl.Scale.X {type=MovieSceneScriptingFloatChannel, default=1, const [1,22]}
       foot_bk1_l_ctrl.Scale.Y {type=MovieSceneScriptingFloatChannel, default=1, const [1,22]}
       foot_bk1_l_ctrl.Scale.Z {type=MovieSceneScriptingFloatChannel, default=1, const [1,22]}
       foot_roll_l_ctrl.Location.X {type=MovieSceneScriptingFloatChannel, default=0, const [0,22]}
       foot_roll_l_ctrl.Location.Y {type=MovieSceneScriptingFloatChannel, default=0, const [0,22]}
       foot_roll_l_ctrl.Location.Z {type=MovieSceneScriptingFloatChannel, default=0, const [0,22]}
       foot_roll_l_ctrl.Rotation.X {type=MovieSceneScriptingFloatChannel, default=0, const [0,22]}
       foot_roll_l_ctrl.Rotation.Y {type=MovieSceneScriptingFloatChannel, default=0, const [0,22]}
       foot_roll_l_ctrl.Rotation.Z {type=MovieSceneScriptingFloatChannel, default=0, const [0,22]}
       foot_roll_l_ctrl.Scale.X {type=MovieSceneScriptingFloatChannel, default=1, const [1,22]}
       foot_roll_l_ctrl.Scale.Y {type=MovieSceneScriptingFloatChannel, default=1, const [1,22]}
       foot_roll_l_ctrl.Scale.Z {type=MovieSceneScriptingFloatChannel, default=1, const [1,22]}
       roll_blend_l_ctrl {type=MovieSceneScriptingFloatChannel, default=1, const [1,22]}
       tip_l_ctrl.Location.X {type=MovieSceneScriptingFloatChannel, default=0, const [0,22]}
       tip_l_ctrl.Location.Y {type=MovieSceneScriptingFloatChannel, default=0, const [0,22]}
       tip_l_ctrl.Location.Z {type=MovieSceneScriptingFloatChannel, default=0, const [0,22]}
       tip_l_ctrl.Rotation.X {type=MovieSceneScriptingFloatChannel, default=0, const [0,22]}
       tip_l_ctrl.Rotation.Y {type=MovieSceneScriptingFloatChannel, default=0, const [0,22]}
       tip_l_ctrl.Rotation.Z {type=MovieSceneScriptingFloatChannel, default=0, const [0,22]}
       tip_l_ctrl.Scale.X {type=MovieSceneScriptingFloatChannel, default=1, const [1,22]}
       tip_l_ctrl.Scale.Y {type=MovieSceneScriptingFloatChannel, default=1, const [1,22]}
       tip_l_ctrl.Scale.Z {type=MovieSceneScriptingFloatChannel, default=1, const [1,22]}
       heel_l_ctrl.Location.X {type=MovieSceneScriptingFloatChannel, default=0, const [0,22]}
       heel_l_ctrl.Location.Y {type=MovieSceneScriptingFloatChannel, default=0, const [0,22]}
       heel_l_ctrl.Location.Z {type=MovieSceneScriptingFloatChannel, default=0, const [0,22]}
       heel_l_ctrl.Rotation.X {type=MovieSceneScriptingFloatChannel, default=0, const [0,22]}
       heel_l_ctrl.Rotation.Y {type=MovieSceneScriptingFloatChannel, default=0, const [0,22]}
       heel_l_ctrl.Rotation.Z {type=MovieSceneScriptingFloatChannel, default=0, const [0,22]}
       heel_l_ctrl.Scale.X {type=MovieSceneScriptingFloatChannel, default=1, const [1,22]}
       heel_l_ctrl.Scale.Y {type=MovieSceneScriptingFloatChannel, default=1, const [1,22]}
       heel_l_ctrl.Scale.Z {type=MovieSceneScriptingFloatChannel, default=1, const [1,22]}
       leg_l_softness {type=MovieSceneScriptingFloatChannel, default=0, const [0,22]}
       leg_l_pv_ik_ctrl.Location.X {
        type=MovieSceneScriptingFloatChannel
        default=7.6263
        thinned_from=22
        keys [[0,7.6263],[1,1.3275],[2,0.818],[3,0.1517],[4,0.3058],[5,0.4123],[6,-0.4017],[7,-0.3493],[8,-0.313],[9,-0.2905],[10,-0.2797],[11,-0.2782],[12,-0.2839],[13,-0.3026],[14,-0.3361],[17,-0.4601],[18,0.379],[19,0.3424],[20,0.3156],[21,0.3036]]
       }
       leg_l_pv_ik_ctrl.Location.Y {
        type=MovieSceneScriptingFloatChannel
        default=5.7802
        thinned_from=22
        keys [[0,5.7802],[1,7.4118],[2,8.1212],[3,8.6035],[4,8.9376],[5,9.1452],[6,9.2868],[7,9.341],[8,9.3541],[9,9.3372],[13,9.1583],[16,8.9641],[17,8.9131],[18,8.8396],[21,8.7712]]
       }
       leg_l_pv_ik_ctrl.Location.Z {
        type=MovieSceneScriptingFloatChannel
        default=8.3578
        thinned_from=22
        keys [[0,8.3578],[1,8.9067],[2,9.8633],[3,10.4922],[4,11.0433],[5,11.3948],[7,11.5931],[8,11.6199],[9,11.5955],[12,11.388],[13,11.3028],[16,10.9747],[17,10.8879],[18,10.8996],[20,10.8036],[21,10.7791]]
       }
       leg_l_pv_ik_ctrl.Rotation.X {
        type=MovieSceneScriptingFloatChannel
        default=-10.8031
        thinned_from=22
        keys [[0,-10.8031],[1,-11.4863],[2,-12.6206],[3,-13.3597],[4,-13.9988],[5,-14.404],[7,-14.6357],[8,-14.6662],[9,-14.6381],[13,-14.3014],[16,-13.9228],[17,-13.8225],[18,-13.8321],[20,-13.721],[21,-13.6926]]
       }
       leg_l_pv_ik_ctrl.Rotation.Y {
        type=MovieSceneScriptingFloatChannel
        default=0.3697
        keys [[0,0.3697],[1,0.1247],[2,0.1224],[3,0.1067],[4,0.129],[5,0.1439],[6,0.1083],[7,0.1137],[8,0.1162],[9,0.1165],[10,0.1153],[11,0.1132],[12,0.1108],[13,0.1074],[14,0.1027],[15,0.0974],[16,0.0924],[17,0.088],[18,0.1285],[19,0.1252],[20,0.1229],[21,0.1216]]
       }
       leg_l_pv_ik_ctrl.Rotation.Z {
        type=MovieSceneScriptingFloatChannel
        default=-9.0291
        thinned_from=22
        keys [[0,-9.0291],[1,-1.5006],[2,-0.8832],[3,-0.1086],[4,-0.2738],[5,-0.3874],[6,0.5295],[7,0.4716],[8,0.4311],[9,0.4057],[10,0.3929],[11,0.3904],[12,0.396],[13,0.4161],[14,0.4527],[17,0.5898],[18,-0.3594],[19,-0.3189],[20,-0.2893],[21,-0.2762]]
       }
       leg_l_pv_ik_ctrl.Scale.X {type=MovieSceneScriptingFloatChannel, default=1, const [1,22]}
       leg_l_pv_ik_ctrl.Scale.Y {type=MovieSceneScriptingFloatChannel, default=1, const [1,22]}
       leg_l_pv_ik_ctrl.Scale.Z {type=MovieSceneScriptingFloatChannel, default=1, const [1,22]}
       thigh_r_fk_ctrl.Location.X {type=MovieSceneScriptingFloatChannel, default=0, thinned_from=22, keys [[0,0],[21,0]]}
       thigh_r_fk_ctrl.Location.Y {type=MovieSceneScriptingFloatChannel, default=0, thinned_from=22, keys [[0,0],[21,0]]}
       thigh_r_fk_ctrl.Location.Z {type=MovieSceneScriptingFloatChannel, default=0, thinned_from=22, keys [[0,0],[21,0]]}
       thigh_r_fk_ctrl.Rotation.X {
        type=MovieSceneScriptingFloatChannel
        default=21.1483
        thinned_from=22
        keys [[0,21.1483],[1,16.2306],[2,10.758],[3,7.9206],[4,5.671],[5,4.131],[6,2.1861],[7,1.5135],[8,1.0996],[9,0.8975],[10,0.8611],[11,0.9449],[12,1.1034],[13,1.4076],[14,1.8981],[15,3.4512],[18,5.2144],[19,5.7243],[20,6.0838],[21,6.2177]]
       }
       thigh_r_fk_ctrl.Rotation.Y {
        type=MovieSceneScriptingFloatChannel
        default=11.2318
        thinned_from=22
        keys [[0,11.2318],[1,14.2929],[2,15.7842],[3,15.9632],[4,15.6601],[5,15.1946],[6,14.4947],[7,14.165],[8,13.9081],[9,13.7196],[10,13.5943],[11,13.527],[12,13.5129],[13,13.5826],[14,13.7423],[15,14.2349],[18,14.8658],[19,15.0293],[20,15.1362],[21,15.1659]]
       }
       thigh_r_fk_ctrl.Rotation.Z {
        type=MovieSceneScriptingFloatChannel
        default=2.2424
        thinned_from=22
        keys [[0,2.2424],[1,-1.5557],[2,-4.5226],[3,-5.834],[4,-6.6344],[5,-6.9428],[6,-6.9602],[7,-6.7962],[8,-6.5388],[10,-5.9339],[11,-5.6809],[12,-5.5236],[13,-5.4595],[15,-5.4143],[16,-5.4249],[17,-5.4078],[19,-5.2998],[20,-5.2587],[21,-5.26]]
       }
       thigh_r_fk_ctrl.Scale.X {type=MovieSceneScriptingFloatChannel, default=1, const [1,22]}
       thigh_r_fk_ctrl.Scale.Y {type=MovieSceneScriptingFloatChannel, default=1, const [1,22]}
       thigh_r_fk_ctrl.Scale.Z {type=MovieSceneScriptingFloatChannel, default=1, const [1,22]}
       calf_r_fk_ctrl.Location.X {
        type=MovieSceneScriptingFloatChannel
        default=0.0833
        thinned_from=22
        keys [[0,0.0833],[1,0.1004],[2,0.1114],[3,0.1158],[4,0.1186],[5,0.1194],[6,0.119],[11,0.1147],[12,0.114],[13,0.1137],[15,0.1138],[21,0.1146]]
       }
       calf_r_fk_ctrl.Location.Y {
        type=MovieSceneScriptingFloatChannel
        default=0.0007
        thinned_from=22
        keys [[0,0.0007],[1,0.0049],[2,0.0084],[3,0.01],[4,0.011],[5,0.0114],[12,0.0093],[21,0.0095]]
       }
       calf_r_fk_ctrl.Location.Z {type=MovieSceneScriptingFloatChannel, default=0.0029, thinned_from=22, keys [[0,0.0029],[1,0.0034],[2,0.0037],[21,0.0038]]}
       calf_r_fk_ctrl.Rotation.X {
        type=MovieSceneScriptingFloatChannel
        default=0.7173
        thinned_from=22
        keys [[0,0.7173],[1,0.8499],[2,0.9321],[3,0.9648],[4,0.9849],[5,0.9908],[7,0.9826],[11,0.9562],[12,0.9513],[13,0.9488],[21,0.9558]]
       }
       calf_r_fk_ctrl.Rotation.Y {
        type=MovieSceneScriptingFloatChannel
        default=0.1877
        thinned_from=22
        keys [[0,0.1877],[1,0.2056],[2,0.2138],[3,0.2164],[4,0.2179],[5,0.2182],[13,0.2152],[21,0.2158]]
       }
       calf_r_fk_ctrl.Rotation.Z {
        type=MovieSceneScriptingFloatChannel
        default=19.8335
        thinned_from=22
        keys [[0,19.8335],[1,24.0035],[2,26.7135],[3,27.8227],[4,28.517],[5,28.7207],[6,28.6099],[11,27.5314],[12,27.3637],[13,27.2777],[15,27.3191],[21,27.5165]]
       }
       calf_r_fk_ctrl.Scale.X {type=MovieSceneScriptingFloatChannel, default=1, const [1,22]}
       calf_r_fk_ctrl.Scale.Y {type=MovieSceneScriptingFloatChannel, default=1, const [1,22]}
       calf_r_fk_ctrl.Scale.Z {type=MovieSceneScriptingFloatChannel, default=1, const [1,22]}
       foot_r_fk_ctrl.Location.X {type=MovieSceneScriptingFloatChannel, default=0, thinned_from=22, keys [[0,0],[21,0]]}
       foot_r_fk_ctrl.Location.Y {type=MovieSceneScriptingFloatChannel, default=0, thinned_from=22, keys [[0,0],[21,0]]}
       foot_r_fk_ctrl.Location.Z {type=MovieSceneScriptingFloatChannel, default=0, thinned_from=22, keys [[0,0],[21,0]]}
       foot_r_fk_ctrl.Rotation.X {
        type=MovieSceneScriptingFloatChannel
        default=-4.6343
        thinned_from=22
        keys [[0,-4.6343],[1,-4.645],[2,-3.5588],[3,-3.3621],[4,-3.1141],[5,-2.9354],[6,-1.8281],[7,-1.7573],[8,-1.7055],[9,-1.6704],[10,-1.6498],[11,-1.6412],[12,-1.6425],[13,-1.6597],[14,-1.6941],[15,-2.7517],[17,-2.8416],[19,-2.9411],[20,-2.9762],[21,-2.9896]]
       }
       foot_r_fk_ctrl.Rotation.Y {
        type=MovieSceneScriptingFloatChannel
        default=-9.0207
        thinned_from=22
        keys [[0,-9.0207],[1,-9.5415],[2,-9.9327],[3,-10.0118],[5,-9.8736],[12,-9.7008],[14,-9.6778],[17,-9.6814],[20,-9.7395],[21,-9.7343]]
       }
       foot_r_fk_ctrl.Rotation.Z {
        type=MovieSceneScriptingFloatChannel
        default=-13.5451
        thinned_from=22
        keys [[0,-13.5451],[1,-15.0298],[2,-16.0683],[3,-16.3167],[4,-16.3913],[5,-16.3635],[6,-16.4685],[11,-16.1257],[13,-16.0499],[14,-16.0457],[15,-15.8909],[21,-15.9769]]
       }
       foot_r_fk_ctrl.Scale.X {type=MovieSceneScriptingFloatChannel, default=1, const [1,22]}
       foot_r_fk_ctrl.Scale.Y {type=MovieSceneScriptingFloatChannel, default=1, const [1,22]}
       foot_r_fk_ctrl.Scale.Z {type=MovieSceneScriptingFloatChannel, default=1, const [1,22]}
       ball_r_fk_ctrl.Location.X {type=MovieSceneScriptingFloatChannel, default=0, const [0,22]}
       ball_r_fk_ctrl.Location.Y {type=MovieSceneScriptingFloatChannel, default=0, const [0,22]}
       ball_r_fk_ctrl.Location.Z {type=MovieSceneScriptingFloatChannel, default=0, const [0,22]}
       ball_r_fk_ctrl.Rotation.X {type=MovieSceneScriptingFloatChannel, default=0, const [0,22]}
       ball_r_fk_ctrl.Rotation.Y {type=MovieSceneScriptingFloatChannel, default=0, const [0,22]}
       ball_r_fk_ctrl.Rotation.Z {type=MovieSceneScriptingFloatChannel, default=0, const [0,22]}
       ball_r_fk_ctrl.Scale.X {type=MovieSceneScriptingFloatChannel, default=1, const [1,22]}
       ball_r_fk_ctrl.Scale.Y {type=MovieSceneScriptingFloatChannel, default=1, const [1,22]}
       ball_r_fk_ctrl.Scale.Z {type=MovieSceneScriptingFloatChannel, default=1, const [1,22]}
       foot_r_ik_ctrl.Location.X {
        type=MovieSceneScriptingFloatChannel
        default=-0.7821
        thinned_from=22
        keys [[0,-0.7821],[1,-0.7642],[2,-0.7607],[3,-0.7864],[4,-0.8252],[5,-0.8486],[7,-0.8589],[9,-0.8632],[13,-0.8616],[14,-0.8626],[15,-0.8597],[17,-0.8582],[20,-0.8497],[21,-0.8485]]
       }
       foot_r_ik_ctrl.Location.Y {type=MovieSceneScriptingFloatChannel, default=-19.8362, thinned_from=22, keys [[0,-19.8362],[2,-19.8041],[5,-19.6649],[21,-19.6649]]}
       foot_r_ik_ctrl.Location.Z {
        type=MovieSceneScriptingFloatChannel
        default=0.0182
        thinned_from=22
        keys [[0,0.0182],[1,0.0594],[2,0.1014],[3,0.1444],[4,0.1886],[5,0.2137],[6,0.2215],[7,0.2269],[8,0.2305],[9,0.2324],[10,0.233],[11,0.2327],[13,0.2303],[14,0.2302],[15,0.2279],[16,0.2275],[17,0.2261],[18,0.2227],[19,0.2185],[20,0.2152],[21,0.2138]]
       }
       foot_r_ik_ctrl.Rotation.X {
        type=MovieSceneScriptingFloatChannel
        default=0.1027
        thinned_from=22
        keys [[0,0.1027],[1,0.2718],[2,0.4486],[3,0.6336],[4,0.8259],[5,0.9359],[6,0.9697],[7,0.9935],[8,1.009],[9,1.0173],[10,1.0201],[11,1.0186],[13,1.0084],[14,1.0067],[15,0.9979],[16,0.9961],[17,0.9899],[18,0.9749],[19,0.9571],[20,0.9422],[21,0.9359]]
       }
       foot_r_ik_ctrl.Rotation.Y {
        type=MovieSceneScriptingFloatChannel
        default=0.1142
        thinned_from=22
        keys [[0,0.1142],[1,0.082],[2,0.1003],[3,0.2306],[4,0.4116],[5,0.5215],[6,0.5506],[7,0.5711],[8,0.5844],[9,0.5915],[10,0.5939],[11,0.5927],[12,0.5889],[14,0.5789],[15,0.5748],[16,0.5733],[17,0.568],[18,0.5551],[19,0.5397],[20,0.5269],[21,0.5216]]
       }
       foot_r_ik_ctrl.Rotation.Z {type=MovieSceneScriptingFloatChannel, default=-36.2946, thinned_from=22, keys [[0,-36.2946],[2,-36.3098],[5,-36.7511],[21,-36.751]]}
       foot_r_ik_ctrl.Scale.X {type=MovieSceneScriptingFloatChannel, default=1, const [1,22]}
       foot_r_ik_ctrl.Scale.Y {type=MovieSceneScriptingFloatChannel, default=1, const [1,22]}
       foot_r_ik_ctrl.Scale.Z {type=MovieSceneScriptingFloatChannel, default=1, const [1,22]}
       ball_r_ik_ctrl.Location.X {type=MovieSceneScriptingFloatChannel, default=0, const [0,22]}
       ball_r_ik_ctrl.Location.Y {type=MovieSceneScriptingFloatChannel, default=0, const [0,22]}
       ball_r_ik_ctrl.Location.Z {type=MovieSceneScriptingFloatChannel, default=0, const [0,22]}
       ball_r_ik_ctrl.Rotation.X {type=MovieSceneScriptingFloatChannel, default=0, const [0,22]}
       ball_r_ik_ctrl.Rotation.Y {type=MovieSceneScriptingFloatChannel, default=0, const [0,22]}
       ball_r_ik_ctrl.Rotation.Z {type=MovieSceneScriptingFloatChannel, default=0, const [0,22]}
       ball_r_ik_ctrl.Scale.X {type=MovieSceneScriptingFloatChannel, default=1, const [1,22]}
       ball_r_ik_ctrl.Scale.Y {type=MovieSceneScriptingFloatChannel, default=1, const [1,22]}
       ball_r_ik_ctrl.Scale.Z {type=MovieSceneScriptingFloatChannel, default=1, const [1,22]}
       foot_bk1_r_ctrl.Location.X {type=MovieSceneScriptingFloatChannel, default=0, const [0,22]}
       foot_bk1_r_ctrl.Location.Y {type=MovieSceneScriptingFloatChannel, default=0, const [0,22]}
       foot_bk1_r_ctrl.Location.Z {type=MovieSceneScriptingFloatChannel, default=0, const [0,22]}
       foot_bk1_r_ctrl.Rotation.X {type=MovieSceneScriptingFloatChannel, default=0, const [0,22]}
       foot_bk1_r_ctrl.Rotation.Y {type=MovieSceneScriptingFloatChannel, default=0, const [0,22]}
       foot_bk1_r_ctrl.Rotation.Z {type=MovieSceneScriptingFloatChannel, default=0, const [0,22]}
       foot_bk1_r_ctrl.Scale.X {type=MovieSceneScriptingFloatChannel, default=1, const [1,22]}
       foot_bk1_r_ctrl.Scale.Y {type=MovieSceneScriptingFloatChannel, default=1, const [1,22]}
       foot_bk1_r_ctrl.Scale.Z {type=MovieSceneScriptingFloatChannel, default=1, const [1,22]}
       foot_roll_r_ctrl.Location.X {type=MovieSceneScriptingFloatChannel, default=0, const [0,22]}
       foot_roll_r_ctrl.Location.Y {type=MovieSceneScriptingFloatChannel, default=0, const [0,22]}
       foot_roll_r_ctrl.Location.Z {type=MovieSceneScriptingFloatChannel, default=0, const [0,22]}
       foot_roll_r_ctrl.Rotation.X {type=MovieSceneScriptingFloatChannel, default=0, const [0,22]}
       foot_roll_r_ctrl.Rotation.Y {type=MovieSceneScriptingFloatChannel, default=0, const [0,22]}
       foot_roll_r_ctrl.Rotation.Z {type=MovieSceneScriptingFloatChannel, default=0, const [0,22]}
       foot_roll_r_ctrl.Scale.X {type=MovieSceneScriptingFloatChannel, default=1, const [1,22]}
       foot_roll_r_ctrl.Scale.Y {type=MovieSceneScriptingFloatChannel, default=1, const [1,22]}
       foot_roll_r_ctrl.Scale.Z {type=MovieSceneScriptingFloatChannel, default=1, const [1,22]}
       roll_blend_r_ctrl {type=MovieSceneScriptingFloatChannel, default=1, const [1,22]}
       tip_r_ctrl.Location.X {type=MovieSceneScriptingFloatChannel, default=0, const [0,22]}
       tip_r_ctrl.Location.Y {type=MovieSceneScriptingFloatChannel, default=0, const [0,22]}
       tip_r_ctrl.Location.Z {type=MovieSceneScriptingFloatChannel, default=0, const [0,22]}
       tip_r_ctrl.Rotation.X {type=MovieSceneScriptingFloatChannel, default=0, const [0,22]}
       tip_r_ctrl.Rotation.Y {type=MovieSceneScriptingFloatChannel, default=0, const [0,22]}
       tip_r_ctrl.Rotation.Z {type=MovieSceneScriptingFloatChannel, default=0, const [0,22]}
       tip_r_ctrl.Scale.X {type=MovieSceneScriptingFloatChannel, default=1, const [1,22]}
       tip_r_ctrl.Scale.Y {type=MovieSceneScriptingFloatChannel, default=1, const [1,22]}
       tip_r_ctrl.Scale.Z {type=MovieSceneScriptingFloatChannel, default=1, const [1,22]}
       heel_r_ctrl.Location.X {type=MovieSceneScriptingFloatChannel, default=0, const [0,22]}
       heel_r_ctrl.Location.Y {type=MovieSceneScriptingFloatChannel, default=0, const [0,22]}
       heel_r_ctrl.Location.Z {type=MovieSceneScriptingFloatChannel, default=0, const [0,22]}
       heel_r_ctrl.Rotation.X {type=MovieSceneScriptingFloatChannel, default=0, const [0,22]}
       heel_r_ctrl.Rotation.Y {type=MovieSceneScriptingFloatChannel, default=0, const [0,22]}
       heel_r_ctrl.Rotation.Z {type=MovieSceneScriptingFloatChannel, default=0, const [0,22]}
       heel_r_ctrl.Scale.X {type=MovieSceneScriptingFloatChannel, default=1, const [1,22]}
       heel_r_ctrl.Scale.Y {type=MovieSceneScriptingFloatChannel, default=1, const [1,22]}
       heel_r_ctrl.Scale.Z {type=MovieSceneScriptingFloatChannel, default=1, const [1,22]}
       leg_r_softness {type=MovieSceneScriptingFloatChannel, default=0, const [0,22]}
       leg_r_pv_ik_ctrl.Location.X {
        type=MovieSceneScriptingFloatChannel
        default=3.2723
        thinned_from=22
        keys [[0,3.2723],[1,3.0237],[2,1.884],[3,1.587],[4,1.262],[5,1.0563],[6,0.0848],[7,0.0329],[8,0.0046],[9,-0.0048],[10,-0.0003],[11,0.0133],[12,0.0311],[13,0.0547],[14,0.0847],[15,1.0033],[16,1.0348],[17,1.0721],[19,1.1647],[20,1.1983],[21,1.2089]]
       }
       leg_r_pv_ik_ctrl.Location.Y {
        type=MovieSceneScriptingFloatChannel
        default=6.7095
        thinned_from=22
        keys [[0,6.7095],[1,7.9969],[2,8.8791],[3,9.2132],[4,9.428],[5,9.4971],[6,9.5155],[11,9.2172],[12,9.1693],[13,9.144],[14,9.1429],[15,9.1093],[21,9.1525]]
       }
       leg_r_pv_ik_ctrl.Location.Z {
        type=MovieSceneScriptingFloatChannel
        default=8.3034
        thinned_from=22
        keys [[0,8.3034],[1,10.1151],[2,11.2247],[3,11.7012],[4,11.9878],[5,12.0617],[6,11.9187],[11,11.4187],[12,11.344],[13,11.307],[14,11.3099],[15,11.4158],[21,11.5253]]
       }
       leg_r_pv_ik_ctrl.Rotation.X {
        type=MovieSceneScriptingFloatChannel
        default=-10.755
        thinned_from=22
        keys [[0,-10.755],[1,-12.9055],[2,-14.2003],[3,-14.7499],[4,-15.0795],[5,-15.1648],[6,-15.0058],[11,-14.4324],[12,-14.3462],[13,-14.3036],[14,-14.3067],[15,-14.4245],[21,-14.5496]]
       }
       leg_r_pv_ik_ctrl.Rotation.Y {
        type=MovieSceneScriptingFloatChannel
        default=0.2007
        thinned_from=22
        keys [[0,0.2007],[1,0.227],[2,0.2057],[3,0.2058],[4,0.1997],[5,0.1927],[6,0.1439],[7,0.1389],[8,0.1346],[9,0.1309],[10,0.1279],[11,0.1258],[12,0.1245],[13,0.1245],[14,0.126],[15,0.1716],[19,0.1813],[20,0.1831],[21,0.184]]
       }
       leg_r_pv_ik_ctrl.Rotation.Z {
        type=MovieSceneScriptingFloatChannel
        default=-3.8134
        thinned_from=22
        keys [[0,-3.8134],[1,-3.4121],[2,-2.0553],[3,-1.7022],[4,-1.3275],[5,-1.0949],[6,-0.0104],[7,0.0465],[8,0.0767],[9,0.0856],[10,0.079],[11,0.0623],[12,0.0412],[13,0.0141],[14,-0.0196],[15,-1.0528],[16,-1.0874],[17,-1.1287],[19,-1.2324],[20,-1.27],[21,-1.2815]]
       }
       leg_r_pv_ik_ctrl.Scale.X {type=MovieSceneScriptingFloatChannel, default=1, const [1,22]}
       leg_r_pv_ik_ctrl.Scale.Y {type=MovieSceneScriptingFloatChannel, default=1, const [1,22]}
       leg_r_pv_ik_ctrl.Scale.Z {type=MovieSceneScriptingFloatChannel, default=1, const [1,22]}
       upperarm_l_fk_ctrl.Location.X {type=MovieSceneScriptingFloatChannel, default=0, thinned_from=22, keys [[0,0],[21,0]]}
       upperarm_l_fk_ctrl.Location.Y {type=MovieSceneScriptingFloatChannel, default=0, thinned_from=22, keys [[0,0],[21,0]]}
       upperarm_l_fk_ctrl.Location.Z {type=MovieSceneScriptingFloatChannel, default=0, thinned_from=22, keys [[0,0],[21,0]]}
       upperarm_l_fk_ctrl.Rotation.X {
        type=MovieSceneScriptingFloatChannel
        default=33.0722
        thinned_from=22
        keys [[0,33.0722],[1,43.2583],[2,46.8232],[3,44.6368],[4,41.3756],[5,38.8909],[6,37.5917],[7,36.8641],[8,36.5599],[9,36.5851],[10,36.8836],[11,37.4056],[12,38.1155],[14,39.9772],[15,40.7474],[16,41.0568],[17,40.6797],[18,39.8359],[19,38.8864],[20,38.1837],[21,38.0714]]
       }
       upperarm_l_fk_ctrl.Rotation.Y {
        type=MovieSceneScriptingFloatChannel
        default=2.5531
        keys [[0,2.5531],[1,7.1112],[2,12.7034],[3,18.5951],[4,23.5969],[5,27.2509],[6,29.6731],[7,31.3327],[8,32.369],[9,32.9035],[10,33.0266],[11,32.875],[12,32.5348],[13,31.9263],[14,31.0928],[15,30.1892],[16,29.4245],[17,28.7835],[18,28.228],[19,27.7758],[20,27.4641],[21,27.3335]]
       }
       upperarm_l_fk_ctrl.Rotation.Z {
        type=MovieSceneScriptingFloatChannel
        default=14.786
        thinned_from=22
        keys [[0,14.786],[1,27.4373],[2,37.9768],[3,40.4358],[4,40.5529],[5,39.8505],[6,39.0672],[9,36.3775],[10,35.6555],[11,35.163],[12,34.9813],[13,35.2504],[14,35.883],[15,36.644],[16,37.2833],[17,37.6369],[18,37.7962],[21,37.9475]]
       }
       upperarm_l_fk_ctrl.Scale.X {type=MovieSceneScriptingFloatChannel, default=1, const [1,22]}
       upperarm_l_fk_ctrl.Scale.Y {type=MovieSceneScriptingFloatChannel, default=1, const [1,22]}
       upperarm_l_fk_ctrl.Scale.Z {type=MovieSceneScriptingFloatChannel, default=1, const [1,22]}
       lowerarm_l_fk_ctrl.Location.X {type=MovieSceneScriptingFloatChannel, default=0, thinned_from=22, keys [[0,0],[21,0]]}
       lowerarm_l_fk_ctrl.Location.Y {type=MovieSceneScriptingFloatChannel, default=0, thinned_from=22, keys [[0,0],[21,0]]}
       lowerarm_l_fk_ctrl.Location.Z {type=MovieSceneScriptingFloatChannel, default=0, thinned_from=22, keys [[0,0],[21,0]]}
       lowerarm_l_fk_ctrl.Rotation.X {
        type=MovieSceneScriptingFloatChannel
        default=-3.3654
        thinned_from=22
        keys [[0,-3.3654],[1,-3.3208],[3,-3.3224],[6,-3.3709],[8,-3.3889],[12,-3.3827],[16,-3.3452],[21,-3.3373]]
       }
       lowerarm_l_fk_ctrl.Rotation.Y {
        type=MovieSceneScriptingFloatChannel
        default=1.701
        thinned_from=22
        keys [[0,1.701],[1,1.6209],[2,1.6182],[3,1.6242],[4,1.6523],[5,1.6846],[6,1.7098],[7,1.7276],[8,1.7377],[9,1.741],[11,1.7343],[12,1.7282],[13,1.7167],[15,1.6804],[16,1.6666],[18,1.6532],[20,1.6506],[21,1.6523]]
       }
       lowerarm_l_fk_ctrl.Rotation.Z {
        type=MovieSceneScriptingFloatChannel
        default=44.8996
        thinned_from=22
        keys [[0,44.8996],[1,49.5829],[2,49.7352],[3,49.3934],[4,47.7698],[5,45.8726],[6,44.37],[7,43.2917],[8,42.6726],[9,42.4658],[10,42.5749],[12,43.2518],[13,43.9518],[15,46.1254],[16,46.9343],[17,47.4175],[18,47.7219],[19,47.8698],[20,47.8698],[21,47.7692]]
       }
       lowerarm_l_fk_ctrl.Scale.X {type=MovieSceneScriptingFloatChannel, default=1, const [1,22]}
       lowerarm_l_fk_ctrl.Scale.Y {type=MovieSceneScriptingFloatChannel, default=1, const [1,22]}
       lowerarm_l_fk_ctrl.Scale.Z {type=MovieSceneScriptingFloatChannel, default=1, const [1,22]}
       hand_l_fk_ctrl.Location.X {type=MovieSceneScriptingFloatChannel, default=0, thinned_from=22, keys [[0,0],[21,0]]}
       hand_l_fk_ctrl.Location.Y {type=MovieSceneScriptingFloatChannel, default=0, thinned_from=22, keys [[0,0],[21,0]]}
       hand_l_fk_ctrl.Location.Z {type=MovieSceneScriptingFloatChannel, default=0, thinned_from=22, keys [[0,0],[21,0]]}
       hand_l_fk_ctrl.Rotation.X {
        type=MovieSceneScriptingFloatChannel
        default=9.5132
        thinned_from=22
        keys [[0,9.5132],[1,3.1569],[2,-4.0852],[3,-7.3103],[4,-9.1588],[5,-10.4471],[6,-11.6022],[7,-12.5464],[8,-13.2336],[9,-13.6494],[10,-13.8113],[11,-13.7611],[12,-13.5535],[13,-13.1152],[14,-12.4429],[16,-10.89],[17,-10.0492],[18,-9.1431],[19,-8.3271],[20,-7.7538],[21,-7.5742]]
       }
       hand_l_fk_ctrl.Rotation.Y {
        type=MovieSceneScriptingFloatChannel
        default=3.4864
        keys [[0,3.4864],[1,-2.1368],[2,-0.4361],[3,4.7773],[4,10.5462],[5,14.9658],[6,17.5298],[7,18.8016],[8,18.9654],[9,18.2693],[10,16.9858],[11,15.3932],[12,13.762],[13,11.8352],[14,9.5829],[15,7.5654],[16,6.3124],[17,5.9107],[18,5.9741],[19,6.3218],[20,6.792],[21,7.2316]]
       }
       hand_l_fk_ctrl.Rotation.Z {
        type=MovieSceneScriptingFloatChannel
        default=-13.8805
        keys [[0,-13.8805],[1,-15.423],[2,-12.5361],[3,-8.8679],[4,-6.0637],[5,-5.0522],[6,-5.5136],[7,-6.4835],[8,-7.7028],[9,-8.9943],[10,-10.2379],[11,-11.3424],[12,-12.2201],[13,-12.7706],[14,-13.0274],[15,-13.0542],[16,-12.8654],[17,-12.3208],[18,-11.4591],[19,-10.5435],[20,-9.8456],[21,-9.6446]]
       }
       hand_l_fk_ctrl.Scale.X {type=MovieSceneScriptingFloatChannel, default=1, const [1,22]}
       hand_l_fk_ctrl.Scale.Y {type=MovieSceneScriptingFloatChannel, default=1, const [1,22]}
       hand_l_fk_ctrl.Scale.Z {type=MovieSceneScriptingFloatChannel, default=1, const [1,22]}
       thumb_01_l_ctrl.Location.X {type=MovieSceneScriptingFloatChannel, default=0, const [0,22]}
       thumb_01_l_ctrl.Location.Y {type=MovieSceneScriptingFloatChannel, default=0, const [0,22]}
       thumb_01_l_ctrl.Location.Z {type=MovieSceneScriptingFloatChannel, default=0, thinned_from=22, keys [[0,0],[21,0]]}
       thumb_01_l_ctrl.Rotation.X {
        type=MovieSceneScriptingFloatChannel
        default=2.7518
        thinned_from=22
        keys [[0,2.7518],[1,3.2049],[2,3.7315],[3,4.1443],[5,4.7547],[6,5.0387],[7,5.2737],[8,5.4339],[9,5.4931],[21,5.4826]]
       }
       thumb_01_l_ctrl.Rotation.Y {
        type=MovieSceneScriptingFloatChannel
        default=0.2016
        thinned_from=22
        keys [[0,0.2016],[1,0.5095],[2,0.8364],[3,1.0296],[4,1.0666],[5,1.0424],[6,0.9814],[7,0.9078],[8,0.8459],[9,0.8199],[13,0.8199],[14,0.8274],[16,0.8274],[17,0.8345],[19,0.8345],[20,0.8392],[21,0.8392]]
       }
       thumb_01_l_ctrl.Rotation.Z {
        type=MovieSceneScriptingFloatChannel
        default=-8.8142
        thinned_from=22
        keys [[0,-8.8142],[1,-7.5507],[2,-6.0556],[3,-4.8293],[4,-3.8361],[5,-2.7794],[6,-1.7679],[7,-0.9106],[8,-0.3164],[9,-0.0942],[13,-0.0942],[14,-0.1089],[16,-0.1089],[17,-0.1262],[19,-0.1262],[20,-0.1452],[21,-0.1452]]
       }
       thumb_01_l_ctrl.Scale.X {type=MovieSceneScriptingFloatChannel, default=1, const [1,22]}
       thumb_01_l_ctrl.Scale.Y {type=MovieSceneScriptingFloatChannel, default=1, const [1,22]}
       thumb_01_l_ctrl.Scale.Z {type=MovieSceneScriptingFloatChannel, default=1, const [1,22]}
       thumb_02_l_ctrl.Location.X {type=MovieSceneScriptingFloatChannel, default=0, thinned_from=22, keys [[0,0],[21,0]]}
       thumb_02_l_ctrl.Location.Y {type=MovieSceneScriptingFloatChannel, default=0, thinned_from=22, keys [[0,0],[21,0]]}
       thumb_02_l_ctrl.Location.Z {type=MovieSceneScriptingFloatChannel, default=0, thinned_from=22, keys [[0,0],[21,0]]}
       thumb_02_l_ctrl.Rotation.X {type=MovieSceneScriptingFloatChannel, default=0, thinned_from=22, keys [[0,0],[21,0]]}
       thumb_02_l_ctrl.Rotation.Y {type=MovieSceneScriptingFloatChannel, default=0, thinned_from=22, keys [[0,0],[21,0]]}
       thumb_02_l_ctrl.Rotation.Z {
        type=MovieSceneScriptingFloatChannel
        default=4.3305
        thinned_from=22
        keys [[0,4.3305],[1,6.209],[2,8.4466],[3,10.3125],[4,11.8727],[5,13.5584],[6,15.1873],[7,16.5767],[8,17.5442],[9,17.9072],[21,17.7918]]
       }
       thumb_02_l_ctrl.Scale.X {type=MovieSceneScriptingFloatChannel, default=1, const [1,22]}
       thumb_02_l_ctrl.Scale.Y {type=MovieSceneScriptingFloatChannel, default=1, const [1,22]}
       thumb_02_l_ctrl.Scale.Z {type=MovieSceneScriptingFloatChannel, default=1, const [1,22]}
       thumb_03_l_ctrl.Location.X {type=MovieSceneScriptingFloatChannel, default=0, thinned_from=22, keys [[0,0],[21,0]]}
       thumb_03_l_ctrl.Location.Y {type=MovieSceneScriptingFloatChannel, default=0, thinned_from=22, keys [[0,0],[21,0]]}
       thumb_03_l_ctrl.Location.Z {type=MovieSceneScriptingFloatChannel, default=0, thinned_from=22, keys [[0,0],[21,0]]}
       thumb_03_l_ctrl.Rotation.X {type=MovieSceneScriptingFloatChannel, default=0, thinned_from=22, keys [[0,0],[21,0]]}
       thumb_03_l_ctrl.Rotation.Y {type=MovieSceneScriptingFloatChannel, default=0, thinned_from=22, keys [[0,0],[21,0]]}
       thumb_03_l_ctrl.Rotation.Z {
        type=MovieSceneScriptingFloatChannel
        default=-12.3418
        thinned_from=22
        keys [[0,-12.3418],[1,-9.8063],[2,-6.9349],[3,-4.836],[4,-3.5449],[5,-2.3881],[6,-1.4105],[7,-0.6567],[8,-0.1716],[11,-0.1716],[12,-0.0126],[13,-0.0126],[14,-0.0302],[15,-0.0302],[16,-0.0501],[17,-0.0501],[18,-0.0678],[20,-0.0678],[21,-0.0803]]
       }
       thumb_03_l_ctrl.Scale.X {type=MovieSceneScriptingFloatChannel, default=1, const [1,22]}
       thumb_03_l_ctrl.Scale.Y {type=MovieSceneScriptingFloatChannel, default=1, const [1,22]}
       thumb_03_l_ctrl.Scale.Z {type=MovieSceneScriptingFloatChannel, default=1, const [1,22]}
       index_metacarpal_l_ctrl.Location.X {type=MovieSceneScriptingFloatChannel, default=0, const [0,22]}
       index_metacarpal_l_ctrl.Location.Y {type=MovieSceneScriptingFloatChannel, default=0, const [0,22]}
       index_metacarpal_l_ctrl.Location.Z {type=MovieSceneScriptingFloatChannel, default=0, const [0,22]}
       index_metacarpal_l_ctrl.Rotation.X {type=MovieSceneScriptingFloatChannel, default=0, const [0,22]}
       index_metacarpal_l_ctrl.Rotation.Y {type=MovieSceneScriptingFloatChannel, default=0, const [0,22]}
       index_metacarpal_l_ctrl.Rotation.Z {type=MovieSceneScriptingFloatChannel, default=0, const [0,22]}
       index_metacarpal_l_ctrl.Scale.X {type=MovieSceneScriptingFloatChannel, default=1, const [1,22]}
       index_metacarpal_l_ctrl.Scale.Y {type=MovieSceneScriptingFloatChannel, default=1, const [1,22]}
       index_metacarpal_l_ctrl.Scale.Z {type=MovieSceneScriptingFloatChannel, default=1, const [1,22]}
       index_01_l_ctrl.Location.X {type=MovieSceneScriptingFloatChannel, default=0, const [0,22]}
       index_01_l_ctrl.Location.Y {type=MovieSceneScriptingFloatChannel, default=0, thinned_from=22, keys [[0,0],[21,0]]}
       index_01_l_ctrl.Location.Z {type=MovieSceneScriptingFloatChannel, default=0, const [0,22]}
       index_01_l_ctrl.Rotation.X {type=MovieSceneScriptingFloatChannel, default=0.0001, thinned_from=22, keys [[0,0.0001],[9,-0.0006],[21,0]]}
       index_01_l_ctrl.Rotation.Y {
        type=MovieSceneScriptingFloatChannel
        default=-0.0004
        thinned_from=22
        keys [[0,-0.0004],[1,0],[4,-0.0001],[8,-0.0011],[10,-0.0011],[17,-0.0001],[21,-0.0001]]
       }
       index_01_l_ctrl.Rotation.Z {
        type=MovieSceneScriptingFloatChannel
        default=-11.5442
        thinned_from=22
        keys [[0,-11.5442],[1,-1.0271],[2,9.0484],[3,16.4711],[4,22.9694],[5,28.3536],[6,32.4342],[7,35.0215],[8,35.9261],[9,35.7848],[10,35.3917],[11,34.7925],[12,34.033],[13,33.1589],[16,30.3078],[17,29.4337],[18,28.6742],[19,28.0749],[20,27.6818],[21,27.5405]]
       }
       index_01_l_ctrl.Scale.X {type=MovieSceneScriptingFloatChannel, default=1, const [1,22]}
       index_01_l_ctrl.Scale.Y {type=MovieSceneScriptingFloatChannel, default=1, const [1,22]}
       index_01_l_ctrl.Scale.Z {type=MovieSceneScriptingFloatChannel, default=1, const [1,22]}
       index_02_l_ctrl.Location.X {type=MovieSceneScriptingFloatChannel, default=0, thinned_from=22, keys [[0,0],[21,0]]}
       index_02_l_ctrl.Location.Y {type=MovieSceneScriptingFloatChannel, default=0, thinned_from=22, keys [[0,0],[21,0]]}
       index_02_l_ctrl.Location.Z {type=MovieSceneScriptingFloatChannel, default=0, thinned_from=22, keys [[0,0],[21,0]]}
       index_02_l_ctrl.Rotation.X {type=MovieSceneScriptingFloatChannel, default=0, thinned_from=22, keys [[0,0],[21,0]]}
       index_02_l_ctrl.Rotation.Y {type=MovieSceneScriptingFloatChannel, default=0, thinned_from=22, keys [[0,0],[21,0]]}
       index_02_l_ctrl.Rotation.Z {
        type=MovieSceneScriptingFloatChannel
        default=-14.4419
        thinned_from=22
        keys [[0,-14.4419],[1,-9.5989],[2,-3.4481],[3,2.4394],[4,7.7031],[5,12.7964],[6,17.3684],[7,21.0684],[8,23.5456],[9,24.4493],[10,24.3058],[11,23.9091],[12,23.3098],[13,22.5586],[14,21.7061],[16,19.8997],[17,19.0472],[18,18.296],[19,17.6967],[20,17.3],[21,17.1565]]
       }
       index_02_l_ctrl.Scale.X {type=MovieSceneScriptingFloatChannel, default=1, const [1,22]}
       index_02_l_ctrl.Scale.Y {type=MovieSceneScriptingFloatChannel, default=1, const [1,22]}
       index_02_l_ctrl.Scale.Z {type=MovieSceneScriptingFloatChannel, default=1, const [1,22]}
       index_03_l_ctrl.Location.X {type=MovieSceneScriptingFloatChannel, default=-0.0001, thinned_from=22, keys [[0,-0.0001],[21,-0.0001]]}
       index_03_l_ctrl.Location.Y {type=MovieSceneScriptingFloatChannel, default=0.0003, thinned_from=22, keys [[0,0.0003],[21,0.0003]]}
       index_03_l_ctrl.Location.Z {type=MovieSceneScriptingFloatChannel, default=0, thinned_from=22, keys [[0,0],[21,0]]}
       index_03_l_ctrl.Rotation.X {type=MovieSceneScriptingFloatChannel, default=0, thinned_from=22, keys [[0,0],[21,0]]}
       index_03_l_ctrl.Rotation.Y {type=MovieSceneScriptingFloatChannel, default=0, thinned_from=22, keys [[0,0],[21,0]]}
       index_03_l_ctrl.Rotation.Z {
        type=MovieSceneScriptingFloatChannel
        default=-15.8066
        keys [[0,-15.8066],[1,-14.2121],[2,-11.9248],[3,-9.2471],[4,-6.482],[5,-3.1301],[6,1.075],[7,5.481],[8,9.4356],[9,12.2867],[10,13.382],[11,13.0115],[12,11.9955],[13,10.4775],[14,8.6009],[15,6.5092],[16,4.3458],[17,2.254],[18,0.3774],[19,-1.1406],[20,-2.1566],[21,-2.5272]]
       }
       index_03_l_ctrl.Scale.X {type=MovieSceneScriptingFloatChannel, default=1, const [1,22]}
       index_03_l_ctrl.Scale.Y {type=MovieSceneScriptingFloatChannel, default=1, const [1,22]}
       index_03_l_ctrl.Scale.Z {type=MovieSceneScriptingFloatChannel, default=1, const [1,22]}
       middle_metacarpal_l_ctrl.Location.X {type=MovieSceneScriptingFloatChannel, default=0, const [0,22]}
       middle_metacarpal_l_ctrl.Location.Y {type=MovieSceneScriptingFloatChannel, default=0, const [0,22]}
       middle_metacarpal_l_ctrl.Location.Z {type=MovieSceneScriptingFloatChannel, default=0, const [0,22]}
       middle_metacarpal_l_ctrl.Rotation.X {type=MovieSceneScriptingFloatChannel, default=0, const [0,22]}
       middle_metacarpal_l_ctrl.Rotation.Y {type=MovieSceneScriptingFloatChannel, default=0, const [0,22]}
       middle_metacarpal_l_ctrl.Rotation.Z {type=MovieSceneScriptingFloatChannel, default=0, const [0,22]}
       middle_metacarpal_l_ctrl.Scale.X {type=MovieSceneScriptingFloatChannel, default=1, const [1,22]}
       middle_metacarpal_l_ctrl.Scale.Y {type=MovieSceneScriptingFloatChannel, default=1, const [1,22]}
       middle_metacarpal_l_ctrl.Scale.Z {type=MovieSceneScriptingFloatChannel, default=1, const [1,22]}
       middle_01_l_ctrl.Location.X {type=MovieSceneScriptingFloatChannel, default=0, const [0,22]}
       middle_01_l_ctrl.Location.Y {type=MovieSceneScriptingFloatChannel, default=0, const [0,22]}
       middle_01_l_ctrl.Location.Z {type=MovieSceneScriptingFloatChannel, default=0, const [0,22]}
       middle_01_l_ctrl.Rotation.X {
        type=MovieSceneScriptingFloatChannel
        default=0.1691
        thinned_from=22
        keys [[0,0.1691],[1,0.5991],[2,1.1758],[3,1.7841],[4,2.3094],[5,2.8106],[7,3.9387],[8,4.4262],[9,4.7697],[10,4.8997],[21,4.8669]]
       }
       middle_01_l_ctrl.Rotation.Y {
        type=MovieSceneScriptingFloatChannel
        default=0.113
        thinned_from=22
        keys [[0,0.113],[1,0.402],[2,0.789],[3,1.1972],[4,1.5496],[5,1.8857],[7,2.6421],[8,2.969],[9,3.1995],[10,3.2868],[21,3.2658]]
       }
       middle_01_l_ctrl.Rotation.Z {
        type=MovieSceneScriptingFloatChannel
        default=-15.2692
        thinned_from=22
        keys [[0,-15.2692],[1,-12.3321],[2,-8.2998],[3,-3.8787],[4,0.2252],[5,3.8887],[6,7.3419],[7,10.408],[8,12.9104],[9,14.6723],[10,15.5169],[11,15.6379],[12,15.3946],[13,14.8585],[14,14.1014],[15,13.1949],[17,11.2204],[18,10.2956],[19,9.5082],[20,8.9297],[21,8.6317]]
       }
       middle_01_l_ctrl.Scale.X {type=MovieSceneScriptingFloatChannel, default=1, const [1,22]}
       middle_01_l_ctrl.Scale.Y {type=MovieSceneScriptingFloatChannel, default=1, const [1,22]}
       middle_01_l_ctrl.Scale.Z {type=MovieSceneScriptingFloatChannel, default=1, const [1,22]}
       middle_02_l_ctrl.Location.X {type=MovieSceneScriptingFloatChannel, default=-0.0452, thinned_from=22, keys [[0,-0.0452],[21,-0.0452]]}
       middle_02_l_ctrl.Location.Y {type=MovieSceneScriptingFloatChannel, default=0.0507, thinned_from=22, keys [[0,0.0507],[21,0.0507]]}
       middle_02_l_ctrl.Location.Z {type=MovieSceneScriptingFloatChannel, default=0.1146, thinned_from=22, keys [[0,0.1146],[21,0.1146]]}
       middle_02_l_ctrl.Rotation.X {type=MovieSceneScriptingFloatChannel, default=0, thinned_from=22, keys [[0,0],[21,0]]}
       middle_02_l_ctrl.Rotation.Y {type=MovieSceneScriptingFloatChannel, default=0, thinned_from=22, keys [[0,0],[21,0]]}
       middle_02_l_ctrl.Rotation.Z {
        type=MovieSceneScriptingFloatChannel
        default=-15.7907
        thinned_from=22
        keys [[0,-15.7907],[1,-14.142],[2,-11.7516],[3,-8.9118],[4,-5.9152],[5,-3.0541],[6,-0.5668],[7,1.4958],[8,3.1214],[9,4.2976],[10,5.0124],[11,5.2533],[12,5.2232],[13,5.1413],[14,5.0206],[15,4.8741],[17,4.5551],[18,4.4086],[19,4.2879],[20,4.206],[21,4.1758]]
       }
       middle_02_l_ctrl.Scale.X {type=MovieSceneScriptingFloatChannel, default=1, const [1,22]}
       middle_02_l_ctrl.Scale.Y {type=MovieSceneScriptingFloatChannel, default=1, const [1,22]}
       middle_02_l_ctrl.Scale.Z {type=MovieSceneScriptingFloatChannel, default=1, const [1,22]}
       middle_03_l_ctrl.Location.X {type=MovieSceneScriptingFloatChannel, default=0.0206, thinned_from=22, keys [[0,0.0206],[21,0.0206]]}
       middle_03_l_ctrl.Location.Y {type=MovieSceneScriptingFloatChannel, default=-0.0583, thinned_from=22, keys [[0,-0.0583],[21,-0.0583]]}
       middle_03_l_ctrl.Location.Z {type=MovieSceneScriptingFloatChannel, default=0.1324, const [0.1324,22]}
       middle_03_l_ctrl.Rotation.X {type=MovieSceneScriptingFloatChannel, default=0, thinned_from=22, keys [[0,0],[21,0]]}
       middle_03_l_ctrl.Rotation.Y {type=MovieSceneScriptingFloatChannel, default=0, thinned_from=22, keys [[0,0],[21,0]]}
       middle_03_l_ctrl.Rotation.Z {
        type=MovieSceneScriptingFloatChannel
        default=-16.034
        keys [[0,-16.034],[1,-15.0132],[2,-13.4823],[3,-11.581],[4,-9.4487],[5,-7.225],[6,-5.0496],[7,-3.1426],[8,-1.6198],[9,-0.4646],[10,0.3398],[11,0.8102],[12,0.9633],[13,0.9263],[14,0.8273],[15,0.6839],[16,0.514],[17,0.3351],[18,0.1651],[19,0.0218],[20,-0.0773],[21,-0.1142]]
       }
       middle_03_l_ctrl.Scale.X {type=MovieSceneScriptingFloatChannel, default=1, const [1,22]}
       middle_03_l_ctrl.Scale.Y {type=MovieSceneScriptingFloatChannel, default=1, const [1,22]}
       middle_03_l_ctrl.Scale.Z {type=MovieSceneScriptingFloatChannel, default=1, const [1,22]}
       ring_metacarpal_l_ctrl.Location.X {type=MovieSceneScriptingFloatChannel, default=0, const [0,22]}
       ring_metacarpal_l_ctrl.Location.Y {type=MovieSceneScriptingFloatChannel, default=0, const [0,22]}
       ring_metacarpal_l_ctrl.Location.Z {type=MovieSceneScriptingFloatChannel, default=0, const [0,22]}
       ring_metacarpal_l_ctrl.Rotation.X {type=MovieSceneScriptingFloatChannel, default=0, const [0,22]}
       ring_metacarpal_l_ctrl.Rotation.Y {type=MovieSceneScriptingFloatChannel, default=0, const [0,22]}
       ring_metacarpal_l_ctrl.Rotation.Z {type=MovieSceneScriptingFloatChannel, default=0, const [0,22]}
       ring_metacarpal_l_ctrl.Scale.X {type=MovieSceneScriptingFloatChannel, default=1, const [1,22]}
       ring_metacarpal_l_ctrl.Scale.Y {type=MovieSceneScriptingFloatChannel, default=1, const [1,22]}
       ring_metacarpal_l_ctrl.Scale.Z {type=MovieSceneScriptingFloatChannel, default=1, const [1,22]}
       ring_01_l_ctrl.Location.X {type=MovieSceneScriptingFloatChannel, default=0, const [0,22]}
       ring_01_l_ctrl.Location.Y {type=MovieSceneScriptingFloatChannel, default=0, const [0,22]}
       ring_01_l_ctrl.Location.Z {type=MovieSceneScriptingFloatChannel, default=0, thinned_from=22, keys [[0,0],[21,0]]}
       ring_01_l_ctrl.Rotation.X {
        type=MovieSceneScriptingFloatChannel
        default=0.2765
        thinned_from=22
        keys [[0,0.2765],[1,0.9436],[3,2.5773],[4,3.3015],[5,3.9974],[6,4.7315],[7,5.3709],[8,5.8163],[9,5.9679],[13,5.8552],[21,5.7554]]
       }
       ring_01_l_ctrl.Rotation.Y {
        type=MovieSceneScriptingFloatChannel
        default=-0.0861
        thinned_from=22
        keys [[0,-0.0861],[1,-0.2779],[2,-0.4779],[3,-0.6845],[4,-0.9653],[5,-1.3494],[6,-1.8002],[7,-2.2128],[8,-2.532],[9,-2.7287],[10,-2.8312],[11,-2.9139],[12,-2.9794],[13,-3.0302],[14,-3.0686],[15,-3.0969],[17,-3.1319],[21,-3.1752]]
       }
       ring_01_l_ctrl.Rotation.Z {
        type=MovieSceneScriptingFloatChannel
        default=-19.912
        keys [[0,-19.912],[1,-19.1645],[2,-17.9104],[3,-16.1569],[4,-13.9216],[5,-10.6746],[6,-6.405],[7,-1.8991],[8,2.0497],[9,4.6449],[10,6.0979],[11,7.1804],[12,7.9414],[13,8.4297],[14,8.6943],[15,8.7842],[16,8.7484],[17,8.6359],[18,8.4956],[19,8.3765],[20,8.3278],[21,8.3982]]
       }
       ring_01_l_ctrl.Scale.X {type=MovieSceneScriptingFloatChannel, default=1, const [1,22]}
       ring_01_l_ctrl.Scale.Y {type=MovieSceneScriptingFloatChannel, default=1, const [1,22]}
       ring_01_l_ctrl.Scale.Z {type=MovieSceneScriptingFloatChannel, default=1, const [1,22]}
       ring_02_l_ctrl.Location.X {type=MovieSceneScriptingFloatChannel, default=0, thinned_from=22, keys [[0,0],[21,0]]}
       ring_02_l_ctrl.Location.Y {type=MovieSceneScriptingFloatChannel, default=0, thinned_from=22, keys [[0,0],[21,0]]}
       ring_02_l_ctrl.Location.Z {type=MovieSceneScriptingFloatChannel, default=0, thinned_from=22, keys [[0,0],[21,0]]}
       ring_02_l_ctrl.Rotation.X {type=MovieSceneScriptingFloatChannel, default=0, thinned_from=22, keys [[0,0],[21,0]]}
       ring_02_l_ctrl.Rotation.Y {type=MovieSceneScriptingFloatChannel, default=0, thinned_from=22, keys [[0,0],[21,0]]}
       ring_02_l_ctrl.Rotation.Z {
        type=MovieSceneScriptingFloatChannel
        default=-20.2821
        thinned_from=22
        keys [[0,-20.2821],[2,-20.7744],[3,-20.8201],[4,-20.516],[5,-19.7],[6,-18.035],[7,-15.579],[8,-12.7631],[9,-10.0176],[10,-7.7735],[11,-6.1511],[12,-4.9466],[13,-4.1045],[14,-3.5693],[15,-3.2855],[16,-3.1976],[17,-3.2501],[18,-3.3875],[19,-3.5543],[20,-3.6952],[21,-3.7544]]
       }
       ring_02_l_ctrl.Scale.X {type=MovieSceneScriptingFloatChannel, default=1, const [1,22]}
       ring_02_l_ctrl.Scale.Y {type=MovieSceneScriptingFloatChannel, default=1, const [1,22]}
       ring_02_l_ctrl.Scale.Z {type=MovieSceneScriptingFloatChannel, default=1, const [1,22]}
       ring_03_l_ctrl.Location.X {type=MovieSceneScriptingFloatChannel, default=0, thinned_from=22, keys [[0,0],[21,0]]}
       ring_03_l_ctrl.Location.Y {type=MovieSceneScriptingFloatChannel, default=0, thinned_from=22, keys [[0,0],[21,0]]}
       ring_03_l_ctrl.Location.Z {type=MovieSceneScriptingFloatChannel, default=0, const [0,22]}
       ring_03_l_ctrl.Rotation.X {type=MovieSceneScriptingFloatChannel, default=0, thinned_from=22, keys [[0,0],[21,0]]}
       ring_03_l_ctrl.Rotation.Y {type=MovieSceneScriptingFloatChannel, default=0, const [0,22]}
       ring_03_l_ctrl.Rotation.Z {
        type=MovieSceneScriptingFloatChannel
        default=-20.2126
        thinned_from=22
        keys [[0,-20.2126],[2,-20.3895],[3,-20.3504],[4,-20.1172],[5,-19.6084],[6,-18.7422],[7,-17.1422],[8,-14.7707],[9,-12.0544],[10,-9.4202],[11,-7.2946],[12,-5.7957],[13,-4.7161],[14,-3.9941],[15,-3.5682],[16,-3.3767],[17,-3.3582],[18,-3.451],[19,-3.5935],[20,-3.7242],[21,-3.7814]]
       }
       ring_03_l_ctrl.Scale.X {type=MovieSceneScriptingFloatChannel, default=1, const [1,22]}
       ring_03_l_ctrl.Scale.Y {type=MovieSceneScriptingFloatChannel, default=1, const [1,22]}
       ring_03_l_ctrl.Scale.Z {type=MovieSceneScriptingFloatChannel, default=1, const [1,22]}
       pinky_metacarpal_l_ctrl.Location.X {type=MovieSceneScriptingFloatChannel, default=0, const [0,22]}
       pinky_metacarpal_l_ctrl.Location.Y {type=MovieSceneScriptingFloatChannel, default=0, const [0,22]}
       pinky_metacarpal_l_ctrl.Location.Z {type=MovieSceneScriptingFloatChannel, default=0, const [0,22]}
       pinky_metacarpal_l_ctrl.Rotation.X {type=MovieSceneScriptingFloatChannel, default=0, const [0,22]}
       pinky_metacarpal_l_ctrl.Rotation.Y {type=MovieSceneScriptingFloatChannel, default=0, const [0,22]}
       pinky_metacarpal_l_ctrl.Rotation.Z {type=MovieSceneScriptingFloatChannel, default=0, const [0,22]}
       pinky_metacarpal_l_ctrl.Scale.X {type=MovieSceneScriptingFloatChannel, default=1, const [1,22]}
       pinky_metacarpal_l_ctrl.Scale.Y {type=MovieSceneScriptingFloatChannel, default=1, const [1,22]}
       pinky_metacarpal_l_ctrl.Scale.Z {type=MovieSceneScriptingFloatChannel, default=1, const [1,22]}
       pinky_01_l_ctrl.Location.X {type=MovieSceneScriptingFloatChannel, default=0, thinned_from=22, keys [[0,0],[21,0]]}
       pinky_01_l_ctrl.Location.Y {type=MovieSceneScriptingFloatChannel, default=0, const [0,22]}
       pinky_01_l_ctrl.Location.Z {type=MovieSceneScriptingFloatChannel, default=0, const [0,22]}
       pinky_01_l_ctrl.Rotation.X {type=MovieSceneScriptingFloatChannel, default=0.0003, thinned_from=22, keys [[0,0.0003],[1,0],[5,0.0002],[8,0.0007],[21,0]]}
       pinky_01_l_ctrl.Rotation.Y {
        type=MovieSceneScriptingFloatChannel
        default=-0.0003
        thinned_from=22
        keys [[0,-0.0003],[1,0],[5,-0.0002],[8,-0.0009],[12,-0.0008],[16,-0.0002],[21,-0.0001]]
       }
       pinky_01_l_ctrl.Rotation.Z {
        type=MovieSceneScriptingFloatChannel
        default=-22.56
        thinned_from=22
        keys [[0,-22.56],[2,-22.6976],[3,-22.6555],[4,-22.4483],[5,-22.0077],[6,-21.2652],[7,-19.9947],[8,-18.1841],[9,-16.0868],[10,-13.9557],[11,-12.044],[12,-10.605],[13,-9.5261],[14,-8.5308],[15,-7.6262],[16,-6.8191],[17,-6.1165],[18,-5.5254],[19,-5.0528],[20,-4.7056],[21,-4.4907]]
       }
       pinky_01_l_ctrl.Scale.X {type=MovieSceneScriptingFloatChannel, default=1, const [1,22]}
       pinky_01_l_ctrl.Scale.Y {type=MovieSceneScriptingFloatChannel, default=1, const [1,22]}
       pinky_01_l_ctrl.Scale.Z {type=MovieSceneScriptingFloatChannel, default=1, const [1,22]}
       pinky_02_l_ctrl.Location.X {type=MovieSceneScriptingFloatChannel, default=0, thinned_from=22, keys [[0,0],[21,0]]}
       pinky_02_l_ctrl.Location.Y {type=MovieSceneScriptingFloatChannel, default=0, thinned_from=22, keys [[0,0],[21,0]]}
       pinky_02_l_ctrl.Location.Z {type=MovieSceneScriptingFloatChannel, default=0, thinned_from=22, keys [[0,0],[21,0]]}
       pinky_02_l_ctrl.Rotation.X {type=MovieSceneScriptingFloatChannel, default=0, thinned_from=22, keys [[0,0],[21,0]]}
       pinky_02_l_ctrl.Rotation.Y {type=MovieSceneScriptingFloatChannel, default=0, thinned_from=22, keys [[0,0],[21,0]]}
       pinky_02_l_ctrl.Rotation.Z {
        type=MovieSceneScriptingFloatChannel
        default=-22.6015
        thinned_from=22
        keys [[0,-22.6015],[2,-22.9267],[3,-22.9267],[4,-22.6835],[5,-22.0574],[6,-20.8799],[7,-19.2092],[9,-15.3262],[10,-13.5835],[11,-12.2861],[12,-11.3058],[13,-10.3683],[14,-9.4874],[15,-8.6777],[16,-7.9532],[17,-7.3284],[18,-6.8176],[19,-6.4351],[20,-6.195],[21,-6.1119]]
       }
       pinky_02_l_ctrl.Scale.X {type=MovieSceneScriptingFloatChannel, default=1, const [1,22]}
       pinky_02_l_ctrl.Scale.Y {type=MovieSceneScriptingFloatChannel, default=1, const [1,22]}
       pinky_02_l_ctrl.Scale.Z {type=MovieSceneScriptingFloatChannel, default=1, const [1,22]}
       pinky_03_l_ctrl.Location.X {type=MovieSceneScriptingFloatChannel, default=0, thinned_from=22, keys [[0,0],[21,0]]}
       pinky_03_l_ctrl.Location.Y {type=MovieSceneScriptingFloatChannel, default=0, thinned_from=22, keys [[0,0],[21,0]]}
       pinky_03_l_ctrl.Location.Z {type=MovieSceneScriptingFloatChannel, default=0, thinned_from=22, keys [[0,0],[21,0]]}
       pinky_03_l_ctrl.Rotation.X {type=MovieSceneScriptingFloatChannel, default=0, thinned_from=22, keys [[0,0],[21,0]]}
       pinky_03_l_ctrl.Rotation.Y {type=MovieSceneScriptingFloatChannel, default=0, thinned_from=22, keys [[0,0],[21,0]]}
       pinky_03_l_ctrl.Rotation.Z {
        type=MovieSceneScriptingFloatChannel
        default=-22.6182
        thinned_from=22
        keys [[0,-22.6182],[1,-22.8018],[2,-22.8889],[3,-22.7005],[4,-22.0573],[5,-20.8177],[6,-19.1135],[8,-15.2383],[9,-13.5309],[10,-12.286],[11,-11.369],[12,-10.4966],[13,-9.6779],[14,-8.9218],[15,-8.2374],[16,-7.6336],[17,-7.1194],[18,-6.7039],[19,-6.396],[20,-6.2046],[21,-6.1388]]
       }
       pinky_03_l_ctrl.Scale.X {type=MovieSceneScriptingFloatChannel, default=1, const [1,22]}
       pinky_03_l_ctrl.Scale.Y {type=MovieSceneScriptingFloatChannel, default=1, const [1,22]}
       pinky_03_l_ctrl.Scale.Z {type=MovieSceneScriptingFloatChannel, default=1, const [1,22]}
       hand_l_ik_ctrl.Location.X {
        type=MovieSceneScriptingFloatChannel
        default=-3.2069
        keys [[0,-3.2069],[1,-8.784],[2,-12.8158],[3,-13.9635],[4,-14.0362],[5,-13.445],[6,-12.6751],[7,-11.9504],[8,-11.2925],[9,-10.7223],[10,-10.2559],[11,-9.9243],[12,-9.7475],[13,-9.8725],[14,-10.3167],[15,-10.8925],[16,-11.4162],[17,-11.7918],[18,-12.0782],[19,-12.2723],[20,-12.3726],[21,-12.3899]]
       }
       hand_l_ik_ctrl.Location.Y {
        type=MovieSceneScriptingFloatChannel
        default=18.6623
        thinned_from=22
        keys [[0,18.6623],[1,24.334],[2,27.6584],[3,27.4804],[4,26.4326],[5,25.1821],[6,24.1549],[7,23.3256],[8,22.686],[9,22.2285],[10,21.9583],[11,21.8375],[12,21.8637],[13,22.0827],[15,22.7575],[16,23.027],[17,23.1996],[21,23.4205]]
       }
       hand_l_ik_ctrl.Location.Z {
        type=MovieSceneScriptingFloatChannel
        default=27.9202
        thinned_from=22
        keys [[0,27.9202],[1,35.4307],[2,42.2589],[3,46.7501],[4,50.2287],[5,52.5065],[6,53.8291],[7,54.6269],[8,54.9882],[9,55.0021],[10,54.7501],[12,53.8468],[15,51.8566],[16,51.4267],[17,51.2607],[21,51.4599]]
       }
       hand_l_ik_ctrl.Rotation.X {
        type=MovieSceneScriptingFloatChannel
        default=68.6943
        thinned_from=22
        keys [[0,68.6943],[1,87.9244],[2,100.4169],[3,106.3444],[4,110.0581],[5,112.1913],[6,113.3398],[7,113.966],[8,114.1436],[9,113.9386],[11,112.6893],[14,110.2506],[15,109.6495],[16,109.428],[21,110.6746]]
       }
       hand_l_ik_ctrl.Rotation.Y {
        type=MovieSceneScriptingFloatChannel
        default=46.5459
        thinned_from=22
        keys [[0,46.5459],[1,40.8668],[2,33.8733],[3,30.1089],[4,28.0897],[5,27.6427],[6,28.1953],[7,29.2574],[8,30.6607],[10,33.7764],[11,35.1473],[12,36.1666],[13,36.7842],[14,37.1519],[15,37.3252],[16,37.3733],[18,37.2508],[20,36.9377],[21,36.7188]]
       }
       hand_l_ik_ctrl.Rotation.Z {
        type=MovieSceneScriptingFloatChannel
        default=58.1215
        thinned_from=22
        keys [[0,58.1215],[1,77.5776],[2,90.9022],[3,97.1201],[4,100.8159],[5,102.5712],[6,103.1135],[7,103.0229],[8,102.4239],[9,101.4408],[13,96.3315],[14,95.2501],[15,94.4754],[16,94.1767],[17,94.3111],[21,95.8694]]
       }
       hand_l_ik_ctrl.Scale.X {type=MovieSceneScriptingFloatChannel, default=1, const [1,22]}
       hand_l_ik_ctrl.Scale.Y {type=MovieSceneScriptingFloatChannel, default=1, const [1,22]}
       hand_l_ik_ctrl.Scale.Z {type=MovieSceneScriptingFloatChannel, default=1, const [1,22]}
       Softness {type=MovieSceneScriptingFloatChannel, default=0, const [0,22]}
       hand_l_ikorient_ctrl.Rotation.Roll {type=MovieSceneScriptingFloatChannel, default=0, const [0,22]}
       hand_l_ikorient_ctrl.Rotation.Pitch {type=MovieSceneScriptingFloatChannel, default=0, const [0,22]}
       hand_l_ikorient_ctrl.Rotation.Yaw {type=MovieSceneScriptingFloatChannel, default=0, const [0,22]}
       arm_l_pv_ik_ctrl.Location.X {
        type=MovieSceneScriptingFloatChannel
        default=15.276
        thinned_from=22
        keys [[0,15.276],[1,13.2123],[2,8.8694],[3,7.9104],[4,7.9299],[5,8.4203],[6,8.983],[7,9.5084],[8,10.0063],[9,10.4671],[10,10.8607],[11,11.1752],[12,11.3722],[13,11.4397],[14,11.4324],[15,11.3462],[20,10.6343],[21,10.5411]]
       }
       arm_l_pv_ik_ctrl.Location.Y {
        type=MovieSceneScriptingFloatChannel
        default=31.5614
        thinned_from=22
        keys [[0,31.5614],[1,48.6817],[2,60.3293],[3,63.8561],[4,65.2639],[5,65.8533],[6,66.1935],[7,66.3567],[9,66.2664],[15,65.2935],[17,64.6862],[19,63.8651],[20,63.584],[21,63.5104]]
       }
       arm_l_pv_ik_ctrl.Location.Z {
        type=MovieSceneScriptingFloatChannel
        default=-1.3725
        thinned_from=22
        keys [[0,-1.3725],[1,2.6235],[2,7.9266],[3,11.0126],[4,12.8342],[5,14.0311],[6,14.9708],[7,15.6749],[8,16.1605],[9,16.4583],[10,16.5925],[11,16.6344],[12,16.6209],[13,16.517],[14,16.3323],[15,16.0811],[16,15.7927],[17,15.3517],[19,14.2764],[20,13.8888],[21,13.7711]]
       }
       arm_l_pv_ik_ctrl.Rotation.X {
        type=MovieSceneScriptingFloatChannel
        default=4.0716
        keys [[0,4.0716],[1,3.0377],[2,2.6218],[3,2.3088],[4,1.3588],[5,1.1039],[6,1.7626],[7,2.6753],[8,3.727],[9,4.8119],[10,5.8359],[11,6.7353],[12,7.4479],[13,7.8283],[14,7.8339],[15,7.5202],[16,6.9497],[17,6.0119],[18,4.8151],[19,3.658],[20,2.8318],[21,2.6204]]
       }
       arm_l_pv_ik_ctrl.Rotation.Y {
        type=MovieSceneScriptingFloatChannel
        default=2.9952
        thinned_from=22
        keys [[0,2.9952],[1,14.7136],[2,26.5934],[3,32.9152],[4,37.0289],[5,39.5901],[6,41.2099],[7,42.2431],[8,42.7676],[9,42.8744],[10,42.646],[11,42.2149],[15,39.8701],[18,38.5629],[19,38.1929],[20,37.9146],[21,37.788]]
       }
       arm_l_pv_ik_ctrl.Rotation.Z {
        type=MovieSceneScriptingFloatChannel
        default=55.4794
        thinned_from=22
        keys [[0,55.4794],[1,73.2042],[2,82.4634],[3,82.2529],[4,79.9617],[5,78.1539],[6,77.518],[7,77.3353],[8,77.4449],[10,78.1755],[13,79.9395],[14,80.3985],[15,80.5996],[16,80.3909],[17,79.555],[19,76.9566],[20,76.0142],[21,75.8312]]
       }
       arm_l_pv_ik_ctrl.Scale.X {type=MovieSceneScriptingFloatChannel, default=1, const [1,22]}
       arm_l_pv_ik_ctrl.Scale.Y {type=MovieSceneScriptingFloatChannel, default=1, const [1,22]}
       arm_l_pv_ik_ctrl.Scale.Z {type=MovieSceneScriptingFloatChannel, default=1, const [1,22]}
       upperarm_r_fk_ctrl.Location.X {type=MovieSceneScriptingFloatChannel, default=0, thinned_from=22, keys [[0,0],[21,0]]}
       upperarm_r_fk_ctrl.Location.Y {type=MovieSceneScriptingFloatChannel, default=0, thinned_from=22, keys [[0,0],[21,0.0001]]}
       upperarm_r_fk_ctrl.Location.Z {type=MovieSceneScriptingFloatChannel, default=-0.0001, thinned_from=22, keys [[0,-0.0001],[21,-0.0001]]}
       upperarm_r_fk_ctrl.Rotation.X {
        type=MovieSceneScriptingFloatChannel
        default=-10.4703
        thinned_from=22
        keys [[0,-10.4703],[1,-15.8297],[2,-22.23],[3,-28.4678],[5,-42.6289],[6,-47.6979],[7,-50.8805],[8,-52.3535],[9,-52.5093],[10,-51.7483],[11,-50.4073],[13,-47.1085],[14,-45.6174],[15,-44.2479],[16,-43.0173],[17,-42.1413],[18,-41.7773],[19,-41.7621],[21,-42.1429]]
       }
       upperarm_r_fk_ctrl.Rotation.Y {
        type=MovieSceneScriptingFloatChannel
        default=-4.6068
        thinned_from=22
        keys [[0,-4.6068],[1,-1.1065],[2,5.1725],[3,10.0386],[4,13.2826],[5,14.8973],[6,15.987],[7,16.8349],[8,17.4361],[9,17.7843],[10,17.8597],[11,17.7049],[12,17.3567],[13,16.6155],[15,14.3749],[16,13.5395],[17,12.9467],[18,12.4028],[19,11.9286],[20,11.56],[21,11.3271]]
       }
       upperarm_r_fk_ctrl.Rotation.Z {
        type=MovieSceneScriptingFloatChannel
        default=-30.7338
        thinned_from=22
        keys [[0,-30.7338],[1,-41.5008],[2,-47.8924],[3,-50.3553],[4,-51.1701],[5,-50.5945],[6,-49.7037],[7,-48.9178],[8,-48.4073],[9,-48.1814],[10,-48.1895],[11,-48.3864],[12,-48.7107],[13,-49.2933],[15,-50.9847],[16,-51.5335],[18,-51.8616],[21,-52.0241]]
       }
       upperarm_r_fk_ctrl.Scale.X {type=MovieSceneScriptingFloatChannel, default=1, const [1,22]}
       upperarm_r_fk_ctrl.Scale.Y {type=MovieSceneScriptingFloatChannel, default=1, const [1,22]}
       upperarm_r_fk_ctrl.Scale.Z {type=MovieSceneScriptingFloatChannel, default=1, const [1,22]}
       lowerarm_r_fk_ctrl.Location.X {type=MovieSceneScriptingFloatChannel, default=0, thinned_from=22, keys [[0,0],[21,0]]}
       lowerarm_r_fk_ctrl.Location.Y {type=MovieSceneScriptingFloatChannel, default=0, thinned_from=22, keys [[0,0],[21,0]]}
       lowerarm_r_fk_ctrl.Location.Z {type=MovieSceneScriptingFloatChannel, default=0, thinned_from=22, keys [[0,0],[21,0]]}
       lowerarm_r_fk_ctrl.Rotation.X {
        type=MovieSceneScriptingFloatChannel
        default=-3.4535
        thinned_from=22
        keys [[0,-3.4535],[1,-3.4001],[3,-3.4399],[4,-3.4399],[8,-3.3861],[11,-3.362],[14,-3.3572],[19,-3.3925],[21,-3.3918]]
       }
       lowerarm_r_fk_ctrl.Rotation.Y {
        type=MovieSceneScriptingFloatChannel
        default=1.825
        thinned_from=22
        keys [[0,1.825],[1,1.7541],[3,1.808],[4,1.808],[7,1.7499],[10,1.7061],[11,1.6954],[13,1.6876],[14,1.6873],[15,1.6931],[17,1.7244],[18,1.7362],[19,1.743],[20,1.7448],[21,1.742]]
       }
       lowerarm_r_fk_ctrl.Rotation.Z {
        type=MovieSceneScriptingFloatChannel
        default=37.1228
        thinned_from=22
        keys [[0,37.1228],[1,41.6582],[3,38.238],[4,38.238],[5,39.3642],[6,40.6961],[7,41.9201],[8,42.9331],[10,44.5951],[11,45.2322],[12,45.5511],[13,45.6966],[14,45.7152],[15,45.3726],[16,44.5088],[17,43.4874],[18,42.7628],[19,42.347],[20,42.2332],[21,42.4065]]
       }
       lowerarm_r_fk_ctrl.Scale.X {type=MovieSceneScriptingFloatChannel, default=1, const [1,22]}
       lowerarm_r_fk_ctrl.Scale.Y {type=MovieSceneScriptingFloatChannel, default=1, const [1,22]}
       lowerarm_r_fk_ctrl.Scale.Z {type=MovieSceneScriptingFloatChannel, default=1, const [1,22]}
       hand_r_fk_ctrl.Location.X {type=MovieSceneScriptingFloatChannel, default=0, thinned_from=22, keys [[0,0],[21,0]]}
       hand_r_fk_ctrl.Location.Y {type=MovieSceneScriptingFloatChannel, default=0, thinned_from=22, keys [[0,0],[21,0]]}
       hand_r_fk_ctrl.Location.Z {type=MovieSceneScriptingFloatChannel, default=0, thinned_from=22, keys [[0,0],[21,0]]}
       hand_r_fk_ctrl.Rotation.X {
        type=MovieSceneScriptingFloatChannel
        default=11.0641
        keys [[0,11.0641],[1,11.5713],[2,7.4636],[3,2.4101],[4,-0.8731],[5,-0.8483],[6,-0.0031],[7,0.809],[8,1.5872],[9,2.2231],[10,2.5074],[11,2.4403],[12,2.023],[13,1.26],[14,0.2633],[15,-0.8257],[16,-1.8879],[17,-2.6412],[18,-2.9595],[19,-2.9762],[20,-2.8266],[21,-2.6495]]
       }
       hand_r_fk_ctrl.Rotation.Y {
        type=MovieSceneScriptingFloatChannel
        default=-6.6974
        keys [[0,-6.6974],[1,2.1049],[2,10.849],[3,13.7025],[4,12.2099],[5,8.8945],[6,5.8432],[7,3.3805],[8,1.5069],[9,0.0941],[10,-0.9687],[11,-1.6771],[12,-2.0211],[13,-1.8477],[14,-1.1553],[15,-0.2023],[16,0.6833],[17,1.2356],[18,1.5084],[19,1.5992],[20,1.6131],[21,1.6537]]
       }
       hand_r_fk_ctrl.Rotation.Z {
        type=MovieSceneScriptingFloatChannel
        default=-9.9769
        thinned_from=22
        keys [[0,-9.9769],[1,-6.9508],[2,-5.1745],[3,-6.4444],[4,-8.8432],[5,-10.5318],[6,-12.636],[7,-15.3189],[8,-17.7453],[9,-19.7165],[10,-21.6002],[11,-23.3131],[12,-24.8076],[13,-25.8214],[14,-26.3091],[16,-26.8328],[17,-27.0268],[18,-26.9479],[19,-26.6949],[21,-26.084]]
       }
       hand_r_fk_ctrl.Scale.X {type=MovieSceneScriptingFloatChannel, default=1, const [1,22]}
       hand_r_fk_ctrl.Scale.Y {type=MovieSceneScriptingFloatChannel, default=1, const [1,22]}
       hand_r_fk_ctrl.Scale.Z {type=MovieSceneScriptingFloatChannel, default=1, const [1,22]}
       thumb_01_r_ctrl.Location.X {type=MovieSceneScriptingFloatChannel, default=0, const [0,22]}
       thumb_01_r_ctrl.Location.Y {type=MovieSceneScriptingFloatChannel, default=0, const [0,22]}
       thumb_01_r_ctrl.Location.Z {type=MovieSceneScriptingFloatChannel, default=0, thinned_from=22, keys [[0,0],[21,0]]}
       thumb_01_r_ctrl.Rotation.X {type=MovieSceneScriptingFloatChannel, default=15.1661, const [15.1661,22]}
       thumb_01_r_ctrl.Rotation.Y {type=MovieSceneScriptingFloatChannel, default=-25.3344, const [-25.3344,22]}
       thumb_01_r_ctrl.Rotation.Z {type=MovieSceneScriptingFloatChannel, default=-2.1592, const [-2.1592,22]}
       thumb_01_r_ctrl.Scale.X {type=MovieSceneScriptingFloatChannel, default=1, const [1,22]}
       thumb_01_r_ctrl.Scale.Y {type=MovieSceneScriptingFloatChannel, default=1, const [1,22]}
       thumb_01_r_ctrl.Scale.Z {type=MovieSceneScriptingFloatChannel, default=1, const [1,22]}
       thumb_02_r_ctrl.Location.X {type=MovieSceneScriptingFloatChannel, default=0, const [0,22]}
       thumb_02_r_ctrl.Location.Y {type=MovieSceneScriptingFloatChannel, default=0, const [0,22]}
       thumb_02_r_ctrl.Location.Z {type=MovieSceneScriptingFloatChannel, default=0, const [0,22]}
       thumb_02_r_ctrl.Rotation.X {type=MovieSceneScriptingFloatChannel, default=0, const [0,22]}
       thumb_02_r_ctrl.Rotation.Y {type=MovieSceneScriptingFloatChannel, default=0, const [0,22]}
       thumb_02_r_ctrl.Rotation.Z {type=MovieSceneScriptingFloatChannel, default=0, const [0,22]}
       thumb_02_r_ctrl.Scale.X {type=MovieSceneScriptingFloatChannel, default=1, const [1,22]}
       thumb_02_r_ctrl.Scale.Y {type=MovieSceneScriptingFloatChannel, default=1, const [1,22]}
       thumb_02_r_ctrl.Scale.Z {type=MovieSceneScriptingFloatChannel, default=1, const [1,22]}
       thumb_03_r_ctrl.Location.X {type=MovieSceneScriptingFloatChannel, default=0, const [0,22]}
       thumb_03_r_ctrl.Location.Y {type=MovieSceneScriptingFloatChannel, default=0, thinned_from=22, keys [[0,0],[21,0]]}
       thumb_03_r_ctrl.Location.Z {type=MovieSceneScriptingFloatChannel, default=0, thinned_from=22, keys [[0,0],[21,0]]}
       thumb_03_r_ctrl.Rotation.X {type=MovieSceneScriptingFloatChannel, default=0, const [0,22]}
       thumb_03_r_ctrl.Rotation.Y {type=MovieSceneScriptingFloatChannel, default=0, thinned_from=22, keys [[0,0],[21,0]]}
       thumb_03_r_ctrl.Rotation.Z {type=MovieSceneScriptingFloatChannel, default=-24.6415, const [-24.6415,22]}
       thumb_03_r_ctrl.Scale.X {type=MovieSceneScriptingFloatChannel, default=1, const [1,22]}
       thumb_03_r_ctrl.Scale.Y {type=MovieSceneScriptingFloatChannel, default=1, const [1,22]}
       thumb_03_r_ctrl.Scale.Z {type=MovieSceneScriptingFloatChannel, default=1, const [1,22]}
       index_metacarpal_r_ctrl.Location.X {type=MovieSceneScriptingFloatChannel, default=0, const [0,22]}
       index_metacarpal_r_ctrl.Location.Y {type=MovieSceneScriptingFloatChannel, default=0, const [0,22]}
       index_metacarpal_r_ctrl.Location.Z {type=MovieSceneScriptingFloatChannel, default=0, const [0,22]}
       index_metacarpal_r_ctrl.Rotation.X {type=MovieSceneScriptingFloatChannel, default=0, const [0,22]}
       index_metacarpal_r_ctrl.Rotation.Y {type=MovieSceneScriptingFloatChannel, default=0, const [0,22]}
       index_metacarpal_r_ctrl.Rotation.Z {type=MovieSceneScriptingFloatChannel, default=0, const [0,22]}
       index_metacarpal_r_ctrl.Scale.X {type=MovieSceneScriptingFloatChannel, default=1, const [1,22]}
       index_metacarpal_r_ctrl.Scale.Y {type=MovieSceneScriptingFloatChannel, default=1, const [1,22]}
       index_metacarpal_r_ctrl.Scale.Z {type=MovieSceneScriptingFloatChannel, default=1, const [1,22]}
       index_01_r_ctrl.Location.X {type=MovieSceneScriptingFloatChannel, default=0, const [0,22]}
       index_01_r_ctrl.Location.Y {type=MovieSceneScriptingFloatChannel, default=0, thinned_from=22, keys [[0,0],[21,0]]}
       index_01_r_ctrl.Location.Z {type=MovieSceneScriptingFloatChannel, default=0, thinned_from=22, keys [[0,0],[21,0]]}
       index_01_r_ctrl.Rotation.X {type=MovieSceneScriptingFloatChannel, default=-1.6158, const [-1.6158,22]}
       index_01_r_ctrl.Rotation.Y {type=MovieSceneScriptingFloatChannel, default=-0.9096, const [-0.9096,22]}
       index_01_r_ctrl.Rotation.Z {type=MovieSceneScriptingFloatChannel, default=-14.4209, const [-14.4209,22]}
       index_01_r_ctrl.Scale.X {type=MovieSceneScriptingFloatChannel, default=1, const [1,22]}
       index_01_r_ctrl.Scale.Y {type=MovieSceneScriptingFloatChannel, default=1, const [1,22]}
       index_01_r_ctrl.Scale.Z {type=MovieSceneScriptingFloatChannel, default=1, const [1,22]}
       index_02_r_ctrl.Location.X {type=MovieSceneScriptingFloatChannel, default=0, const [0,22]}
       index_02_r_ctrl.Location.Y {type=MovieSceneScriptingFloatChannel, default=0, const [0,22]}
       index_02_r_ctrl.Location.Z {type=MovieSceneScriptingFloatChannel, default=0, thinned_from=22, keys [[0,0],[21,0]]}
       index_02_r_ctrl.Rotation.X {type=MovieSceneScriptingFloatChannel, default=0, thinned_from=22, keys [[0,0],[21,0]]}
       index_02_r_ctrl.Rotation.Y {type=MovieSceneScriptingFloatChannel, default=0, const [0,22]}
       index_02_r_ctrl.Rotation.Z {type=MovieSceneScriptingFloatChannel, default=-4.4334, const [-4.4334,22]}
       index_02_r_ctrl.Scale.X {type=MovieSceneScriptingFloatChannel, default=1, const [1,22]}
       index_02_r_ctrl.Scale.Y {type=MovieSceneScriptingFloatChannel, default=1, const [1,22]}
       index_02_r_ctrl.Scale.Z {type=MovieSceneScriptingFloatChannel, default=1, const [1,22]}
       index_03_r_ctrl.Location.X {type=MovieSceneScriptingFloatChannel, default=0, thinned_from=22, keys [[0,0],[21,0]]}
       index_03_r_ctrl.Location.Y {type=MovieSceneScriptingFloatChannel, default=0, const [0,22]}
       index_03_r_ctrl.Location.Z {type=MovieSceneScriptingFloatChannel, default=0, thinned_from=22, keys [[0,0],[21,0]]}
       index_03_r_ctrl.Rotation.X {type=MovieSceneScriptingFloatChannel, default=0, const [0,22]}
       index_03_r_ctrl.Rotation.Y {type=MovieSceneScriptingFloatChannel, default=0, const [0,22]}
       index_03_r_ctrl.Rotation.Z {type=MovieSceneScriptingFloatChannel, default=-7.9391, const [-7.9391,22]}
       index_03_r_ctrl.Scale.X {type=MovieSceneScriptingFloatChannel, default=1, const [1,22]}
       index_03_r_ctrl.Scale.Y {type=MovieSceneScriptingFloatChannel, default=1, const [1,22]}
       index_03_r_ctrl.Scale.Z {type=MovieSceneScriptingFloatChannel, default=1, const [1,22]}
       middle_metacarpal_r_ctrl.Location.X {type=MovieSceneScriptingFloatChannel, default=0, const [0,22]}
       middle_metacarpal_r_ctrl.Location.Y {type=MovieSceneScriptingFloatChannel, default=0, const [0,22]}
       middle_metacarpal_r_ctrl.Location.Z {type=MovieSceneScriptingFloatChannel, default=0, const [0,22]}
       middle_metacarpal_r_ctrl.Rotation.X {type=MovieSceneScriptingFloatChannel, default=0, const [0,22]}
       middle_metacarpal_r_ctrl.Rotation.Y {type=MovieSceneScriptingFloatChannel, default=0, const [0,22]}
       middle_metacarpal_r_ctrl.Rotation.Z {type=MovieSceneScriptingFloatChannel, default=0, const [0,22]}
       middle_metacarpal_r_ctrl.Scale.X {type=MovieSceneScriptingFloatChannel, default=1, const [1,22]}
       middle_metacarpal_r_ctrl.Scale.Y {type=MovieSceneScriptingFloatChannel, default=1, const [1,22]}
       middle_metacarpal_r_ctrl.Scale.Z {type=MovieSceneScriptingFloatChannel, default=1, const [1,22]}
       middle_01_r_ctrl.Location.X {type=MovieSceneScriptingFloatChannel, default=0, const [0,22]}
       middle_01_r_ctrl.Location.Y {type=MovieSceneScriptingFloatChannel, default=0, thinned_from=22, keys [[0,0],[21,0]]}
       middle_01_r_ctrl.Location.Z {type=MovieSceneScriptingFloatChannel, default=0, thinned_from=22, keys [[0,0],[21,0]]}
       middle_01_r_ctrl.Rotation.X {type=MovieSceneScriptingFloatChannel, default=-2.5986, const [-2.5986,22]}
       middle_01_r_ctrl.Rotation.Y {type=MovieSceneScriptingFloatChannel, default=-2.1686, const [-2.1686,22]}
       middle_01_r_ctrl.Rotation.Z {type=MovieSceneScriptingFloatChannel, default=-19.3508, const [-19.3508,22]}
       middle_01_r_ctrl.Scale.X {type=MovieSceneScriptingFloatChannel, default=1, const [1,22]}
       middle_01_r_ctrl.Scale.Y {type=MovieSceneScriptingFloatChannel, default=1, const [1,22]}
       middle_01_r_ctrl.Scale.Z {type=MovieSceneScriptingFloatChannel, default=1, const [1,22]}
       middle_02_r_ctrl.Location.X {type=MovieSceneScriptingFloatChannel, default=0, thinned_from=22, keys [[0,0],[21,0]]}
       middle_02_r_ctrl.Location.Y {type=MovieSceneScriptingFloatChannel, default=0, thinned_from=22, keys [[0,0],[21,0]]}
       middle_02_r_ctrl.Location.Z {type=MovieSceneScriptingFloatChannel, default=0, thinned_from=22, keys [[0,0],[21,0]]}
       middle_02_r_ctrl.Rotation.X {type=MovieSceneScriptingFloatChannel, default=-0.0039, const [-0.0039,22]}
       middle_02_r_ctrl.Rotation.Y {type=MovieSceneScriptingFloatChannel, default=-0.013, const [-0.013,22]}
       middle_02_r_ctrl.Rotation.Z {type=MovieSceneScriptingFloatChannel, default=-12.5582, const [-12.5582,22]}
       middle_02_r_ctrl.Scale.X {type=MovieSceneScriptingFloatChannel, default=1, const [1,22]}
       middle_02_r_ctrl.Scale.Y {type=MovieSceneScriptingFloatChannel, default=1, const [1,22]}
       middle_02_r_ctrl.Scale.Z {type=MovieSceneScriptingFloatChannel, default=1, const [1,22]}
       middle_03_r_ctrl.Location.X {type=MovieSceneScriptingFloatChannel, default=0, thinned_from=22, keys [[0,0],[21,0]]}
       middle_03_r_ctrl.Location.Y {type=MovieSceneScriptingFloatChannel, default=0, thinned_from=22, keys [[0,0],[21,0]]}
       middle_03_r_ctrl.Location.Z {type=MovieSceneScriptingFloatChannel, default=0, thinned_from=22, keys [[0,0],[21,0]]}
       middle_03_r_ctrl.Rotation.X {type=MovieSceneScriptingFloatChannel, default=-0.011, const [-0.011,22]}
       middle_03_r_ctrl.Rotation.Y {type=MovieSceneScriptingFloatChannel, default=0.0014, const [0.0014,22]}
       middle_03_r_ctrl.Rotation.Z {type=MovieSceneScriptingFloatChannel, default=-6.4754, const [-6.4754,22]}
       middle_03_r_ctrl.Scale.X {type=MovieSceneScriptingFloatChannel, default=1, const [1,22]}
       middle_03_r_ctrl.Scale.Y {type=MovieSceneScriptingFloatChannel, default=1, const [1,22]}
       middle_03_r_ctrl.Scale.Z {type=MovieSceneScriptingFloatChannel, default=1, const [1,22]}
       ring_metacarpal_r_ctrl.Location.X {type=MovieSceneScriptingFloatChannel, default=0, const [0,22]}
       ring_metacarpal_r_ctrl.Location.Y {type=MovieSceneScriptingFloatChannel, default=0, const [0,22]}
       ring_metacarpal_r_ctrl.Location.Z {type=MovieSceneScriptingFloatChannel, default=0, const [0,22]}
       ring_metacarpal_r_ctrl.Rotation.X {type=MovieSceneScriptingFloatChannel, default=0, const [0,22]}
       ring_metacarpal_r_ctrl.Rotation.Y {type=MovieSceneScriptingFloatChannel, default=0, const [0,22]}
       ring_metacarpal_r_ctrl.Rotation.Z {type=MovieSceneScriptingFloatChannel, default=0, const [0,22]}
       ring_metacarpal_r_ctrl.Scale.X {type=MovieSceneScriptingFloatChannel, default=1, const [1,22]}
       ring_metacarpal_r_ctrl.Scale.Y {type=MovieSceneScriptingFloatChannel, default=1, const [1,22]}
       ring_metacarpal_r_ctrl.Scale.Z {type=MovieSceneScriptingFloatChannel, default=1, const [1,22]}
       ring_01_r_ctrl.Location.X {type=MovieSceneScriptingFloatChannel, default=0, thinned_from=22, keys [[0,0],[21,0]]}
       ring_01_r_ctrl.Location.Y {type=MovieSceneScriptingFloatChannel, default=0, thinned_from=22, keys [[0,0],[21,0]]}
       ring_01_r_ctrl.Location.Z {type=MovieSceneScriptingFloatChannel, default=0, thinned_from=22, keys [[0,0],[21,0]]}
       ring_01_r_ctrl.Rotation.X {type=MovieSceneScriptingFloatChannel, default=-1.6667, const [-1.6667,22]}
       ring_01_r_ctrl.Rotation.Y {type=MovieSceneScriptingFloatChannel, default=-1.5698, const [-1.5698,22]}
       ring_01_r_ctrl.Rotation.Z {type=MovieSceneScriptingFloatChannel, default=-13.1407, const [-13.1407,22]}
       ring_01_r_ctrl.Scale.X {type=MovieSceneScriptingFloatChannel, default=1, const [1,22]}
       ring_01_r_ctrl.Scale.Y {type=MovieSceneScriptingFloatChannel, default=1, const [1,22]}
       ring_01_r_ctrl.Scale.Z {type=MovieSceneScriptingFloatChannel, default=1, const [1,22]}
       ring_02_r_ctrl.Location.X {type=MovieSceneScriptingFloatChannel, default=0, const [0,22]}
       ring_02_r_ctrl.Location.Y {type=MovieSceneScriptingFloatChannel, default=0, thinned_from=22, keys [[0,0],[21,0]]}
       ring_02_r_ctrl.Location.Z {type=MovieSceneScriptingFloatChannel, default=0, const [0,22]}
       ring_02_r_ctrl.Rotation.X {type=MovieSceneScriptingFloatChannel, default=0, const [0,22]}
       ring_02_r_ctrl.Rotation.Y {type=MovieSceneScriptingFloatChannel, default=0, thinned_from=22, keys [[0,0],[21,0]]}
       ring_02_r_ctrl.Rotation.Z {type=MovieSceneScriptingFloatChannel, default=-14.3916, const [-14.3916,22]}
       ring_02_r_ctrl.Scale.X {type=MovieSceneScriptingFloatChannel, default=1, const [1,22]}
       ring_02_r_ctrl.Scale.Y {type=MovieSceneScriptingFloatChannel, default=1, const [1,22]}
       ring_02_r_ctrl.Scale.Z {type=MovieSceneScriptingFloatChannel, default=1, const [1,22]}
       ring_03_r_ctrl.Location.X {type=MovieSceneScriptingFloatChannel, default=0, thinned_from=22, keys [[0,0],[21,0]]}
       ring_03_r_ctrl.Location.Y {type=MovieSceneScriptingFloatChannel, default=0, thinned_from=22, keys [[0,0],[21,0]]}
       ring_03_r_ctrl.Location.Z {type=MovieSceneScriptingFloatChannel, default=0, thinned_from=22, keys [[0,0],[21,0]]}
       ring_03_r_ctrl.Rotation.X {type=MovieSceneScriptingFloatChannel, default=0, thinned_from=22, keys [[0,0],[21,0]]}
       ring_03_r_ctrl.Rotation.Y {type=MovieSceneScriptingFloatChannel, default=0, const [0,22]}
       ring_03_r_ctrl.Rotation.Z {type=MovieSceneScriptingFloatChannel, default=-20.52, const [-20.52,22]}
       ring_03_r_ctrl.Scale.X {type=MovieSceneScriptingFloatChannel, default=1, const [1,22]}
       ring_03_r_ctrl.Scale.Y {type=MovieSceneScriptingFloatChannel, default=1, const [1,22]}
       ring_03_r_ctrl.Scale.Z {type=MovieSceneScriptingFloatChannel, default=1, const [1,22]}
       pinky_metacarpal_r_ctrl.Location.X {type=MovieSceneScriptingFloatChannel, default=0, const [0,22]}
       pinky_metacarpal_r_ctrl.Location.Y {type=MovieSceneScriptingFloatChannel, default=0, const [0,22]}
       pinky_metacarpal_r_ctrl.Location.Z {type=MovieSceneScriptingFloatChannel, default=0, const [0,22]}
       pinky_metacarpal_r_ctrl.Rotation.X {type=MovieSceneScriptingFloatChannel, default=0, const [0,22]}
       pinky_metacarpal_r_ctrl.Rotation.Y {type=MovieSceneScriptingFloatChannel, default=0, const [0,22]}
       pinky_metacarpal_r_ctrl.Rotation.Z {type=MovieSceneScriptingFloatChannel, default=0, const [0,22]}
       pinky_metacarpal_r_ctrl.Scale.X {type=MovieSceneScriptingFloatChannel, default=1, const [1,22]}
       pinky_metacarpal_r_ctrl.Scale.Y {type=MovieSceneScriptingFloatChannel, default=1, const [1,22]}
       pinky_metacarpal_r_ctrl.Scale.Z {type=MovieSceneScriptingFloatChannel, default=1, const [1,22]}
       pinky_01_r_ctrl.Location.X {type=MovieSceneScriptingFloatChannel, default=0, const [0,22]}
       pinky_01_r_ctrl.Location.Y {type=MovieSceneScriptingFloatChannel, default=0, thinned_from=22, keys [[0,0],[21,0]]}
       pinky_01_r_ctrl.Location.Z {type=MovieSceneScriptingFloatChannel, default=0, thinned_from=22, keys [[0,0],[21,0]]}
       pinky_01_r_ctrl.Rotation.X {type=MovieSceneScriptingFloatChannel, default=3.2666, const [3.2666,22]}
       pinky_01_r_ctrl.Rotation.Y {type=MovieSceneScriptingFloatChannel, default=-5.7299, const [-5.7299,22]}
       pinky_01_r_ctrl.Rotation.Z {type=MovieSceneScriptingFloatChannel, default=-4.1286, const [-4.1286,22]}
       pinky_01_r_ctrl.Scale.X {type=MovieSceneScriptingFloatChannel, default=1, const [1,22]}
       pinky_01_r_ctrl.Scale.Y {type=MovieSceneScriptingFloatChannel, default=1, const [1,22]}
       pinky_01_r_ctrl.Scale.Z {type=MovieSceneScriptingFloatChannel, default=1, const [1,22]}
       pinky_02_r_ctrl.Location.X {type=MovieSceneScriptingFloatChannel, default=0, thinned_from=22, keys [[0,0],[21,0]]}
       pinky_02_r_ctrl.Location.Y {type=MovieSceneScriptingFloatChannel, default=0, thinned_from=22, keys [[0,0],[21,0]]}
       pinky_02_r_ctrl.Location.Z {type=MovieSceneScriptingFloatChannel, default=0, thinned_from=22, keys [[0,0],[21,0]]}
       pinky_02_r_ctrl.Rotation.X {type=MovieSceneScriptingFloatChannel, default=0, thinned_from=22, keys [[0,0],[21,0]]}
       pinky_02_r_ctrl.Rotation.Y {type=MovieSceneScriptingFloatChannel, default=0, thinned_from=22, keys [[0,0],[21,0]]}
       pinky_02_r_ctrl.Rotation.Z {type=MovieSceneScriptingFloatChannel, default=-13.151, const [-13.151,22]}
       pinky_02_r_ctrl.Scale.X {type=MovieSceneScriptingFloatChannel, default=1, const [1,22]}
       pinky_02_r_ctrl.Scale.Y {type=MovieSceneScriptingFloatChannel, default=1, const [1,22]}
       pinky_02_r_ctrl.Scale.Z {type=MovieSceneScriptingFloatChannel, default=1, const [1,22]}
       pinky_03_r_ctrl.Location.X {type=MovieSceneScriptingFloatChannel, default=0, thinned_from=22, keys [[0,0],[21,0]]}
       pinky_03_r_ctrl.Location.Y {type=MovieSceneScriptingFloatChannel, default=0, thinned_from=22, keys [[0,0],[21,0]]}
       pinky_03_r_ctrl.Location.Z {type=MovieSceneScriptingFloatChannel, default=0, thinned_from=22, keys [[0,0],[21,0]]}
       pinky_03_r_ctrl.Rotation.X {type=MovieSceneScriptingFloatChannel, default=0, const [0,22]}
       pinky_03_r_ctrl.Rotation.Y {type=MovieSceneScriptingFloatChannel, default=0, thinned_from=22, keys [[0,0],[21,0]]}
       pinky_03_r_ctrl.Rotation.Z {type=MovieSceneScriptingFloatChannel, default=-22.5109, const [-22.5109,22]}
       pinky_03_r_ctrl.Scale.X {type=MovieSceneScriptingFloatChannel, default=1, const [1,22]}
       pinky_03_r_ctrl.Scale.Y {type=MovieSceneScriptingFloatChannel, default=1, const [1,22]}
       pinky_03_r_ctrl.Scale.Z {type=MovieSceneScriptingFloatChannel, default=1, const [1,22]}
       hand_r_ik_ctrl.Location.X {
        type=MovieSceneScriptingFloatChannel
        default=-14.8243
        thinned_from=22
        keys [[0,-14.8243],[1,-22.8963],[2,-30.6662],[3,-36.1972],[4,-40.5535],[5,-43.5935],[6,-45.5049],[7,-46.657],[8,-47.176],[9,-47.1988],[10,-46.8512],[14,-44.3277],[15,-43.76],[16,-43.3063],[17,-42.9732],[18,-42.8004],[20,-42.753],[21,-42.8091]]
       }
       hand_r_ik_ctrl.Location.Y {
        type=MovieSceneScriptingFloatChannel
        default=6.9648
        thinned_from=22
        keys [[0,6.9648],[1,8.9857],[2,10.3344],[3,10.4751],[4,10.1269],[5,9.5226],[6,8.8835],[7,8.2996],[8,7.801],[9,7.3767],[10,7.0336],[11,6.8224],[12,6.797],[13,6.9556],[14,7.2377],[15,7.6002],[17,8.3824],[18,8.6979],[19,8.9341],[20,9.0764],[21,9.1201]]
       }
       hand_r_ik_ctrl.Location.Z {
        type=MovieSceneScriptingFloatChannel
        default=-21.3988
        thinned_from=22
        keys [[0,-21.3988],[1,-29.746],[2,-36.6972],[3,-39.8259],[4,-40.9994],[5,-40.9648],[6,-40.4638],[8,-39.2363],[10,-38.2001],[11,-37.8234],[12,-37.6253],[13,-37.6261],[14,-37.792],[15,-38.0991],[17,-38.9552],[18,-39.2919],[19,-39.5294],[20,-39.6632],[21,-39.691]]
       }
       hand_r_ik_ctrl.Rotation.X {
        type=MovieSceneScriptingFloatChannel
        default=-14.5331
        keys [[0,-14.5331],[1,-12.7155],[2,-10.8581],[3,-9.2848],[4,-6.4571],[5,-2.5323],[6,1.2667],[7,4.5269],[8,6.9137],[9,8.4381],[10,9.2741],[11,9.6019],[12,9.5906],[13,9.1399],[14,8.3181],[15,7.4439],[16,6.8287],[17,6.5199],[18,6.4186],[19,6.4231],[20,6.4525],[21,6.428]]
       }
       hand_r_ik_ctrl.Rotation.Y {
        type=MovieSceneScriptingFloatChannel
        default=-18.9272
        thinned_from=22
        keys [[0,-18.9272],[1,-19.6862],[2,-21.4034],[3,-24.2549],[4,-27.8605],[5,-30.7401],[6,-32.3202],[7,-33.062],[8,-33.4017],[9,-33.5928],[11,-33.704],[13,-33.9455],[15,-34.0423],[16,-34.3399],[18,-35.4091],[19,-35.8725],[20,-36.1496],[21,-36.1449]]
       }
       hand_r_ik_ctrl.Rotation.Z {
        type=MovieSceneScriptingFloatChannel
        default=-22.2752
        thinned_from=22
        keys [[0,-22.2752],[1,-29.3153],[2,-37.6952],[3,-47.409],[4,-58.8797],[5,-69.4568],[6,-78.0747],[7,-85.1966],[8,-89.9547],[9,-92.5501],[10,-93.8818],[11,-94.2966],[12,-94.1169],[13,-93.2514],[15,-90.1187],[16,-88.798],[17,-87.8918],[18,-87.2359],[20,-86.4721],[21,-86.2896]]
       }
       hand_r_ik_ctrl.Scale.X {type=MovieSceneScriptingFloatChannel, default=1, const [1,22]}
       hand_r_ik_ctrl.Scale.Y {type=MovieSceneScriptingFloatChannel, default=1, const [1,22]}
       hand_r_ik_ctrl.Scale.Z {type=MovieSceneScriptingFloatChannel, default=1, const [1,22]}
       Softness_2 {type=MovieSceneScriptingFloatChannel, default=0, const [0,22]}
       hand_r_ikorient_ctrl.Rotation.Roll {type=MovieSceneScriptingFloatChannel, default=0, const [0,22]}
       hand_r_ikorient_ctrl.Rotation.Pitch {type=MovieSceneScriptingFloatChannel, default=0, const [0,22]}
       hand_r_ikorient_ctrl.Rotation.Yaw {type=MovieSceneScriptingFloatChannel, default=0, const [0,22]}
       arm_r_pv_ik_ctrl.Location.X {
        type=MovieSceneScriptingFloatChannel
        default=-29.8763
        thinned_from=22
        keys [[0,-29.8763],[1,-38.2037],[2,-43.9394],[3,-47.665],[4,-50.9276],[5,-53.2508],[6,-54.3671],[7,-54.8144],[8,-54.8771],[10,-54.4789],[17,-52.7924],[18,-52.7223],[21,-52.9638]]
       }
       arm_r_pv_ik_ctrl.Location.Y {
        type=MovieSceneScriptingFloatChannel
        default=-21.5151
        thinned_from=22
        keys [[0,-21.5151],[1,-25.2911],[2,-27.6507],[3,-28.3926],[4,-28.0928],[5,-27.1045],[6,-26.2865],[7,-25.7198],[8,-25.3986],[9,-25.2681],[10,-25.2534],[13,-25.473],[15,-25.4726],[17,-25.7666],[18,-25.865],[20,-25.8888],[21,-25.8395]]
       }
       arm_r_pv_ik_ctrl.Location.Z {
        type=MovieSceneScriptingFloatChannel
        default=11.2597
        thinned_from=22
        keys [[0,11.2597],[1,16.6039],[2,19.8633],[3,20.692],[4,19.8552],[5,17.8105],[6,16.0861],[7,14.8896],[8,14.3277],[9,14.2968],[10,14.6335],[11,15.2178],[12,15.9288],[13,16.7028],[14,17.5169],[15,18.2753],[16,18.8849],[17,19.2249],[18,19.3601],[19,19.3628],[21,19.2543]]
       }
       arm_r_pv_ik_ctrl.Rotation.X {
        type=MovieSceneScriptingFloatChannel
        default=28.1399
        keys [[0,28.1399],[1,34.2573],[2,35.6759],[3,33.0249],[4,27.2044],[5,19.3209],[6,12.8963],[7,8.4053],[8,6.0755],[9,5.5443],[10,6.335],[11,8.0113],[12,10.2003],[13,12.7341],[14,15.4361],[15,17.9939],[16,20.1014],[17,21.5064],[18,22.2732],[19,22.592],[20,22.6288],[21,22.5471]]
       }
       arm_r_pv_ik_ctrl.Rotation.Y {
        type=MovieSceneScriptingFloatChannel
        default=5.0882
        thinned_from=22
        keys [[0,5.0882],[1,13.3721],[2,23.0682],[3,30.527],[4,36.4279],[5,39.9399],[6,41.4803],[7,41.9955],[8,42.0637],[10,41.8109],[12,41.3498],[17,39.6596],[18,39.4431],[19,39.3411],[21,39.3898]]
       }
       arm_r_pv_ik_ctrl.Rotation.Z {
        type=MovieSceneScriptingFloatChannel
        default=-40.8577
        thinned_from=22
        keys [[0,-40.8577],[1,-50.5683],[2,-56.3268],[3,-60.9754],[4,-67.7374],[5,-76.0232],[6,-82.3728],[7,-86.5715],[8,-88.5705],[9,-88.8184],[10,-87.8624],[11,-86.1574],[13,-82.0219],[15,-78.313],[16,-76.648],[17,-75.4063],[18,-74.7868],[19,-74.6194],[21,-75.0054]]
       }
       arm_r_pv_ik_ctrl.Scale.X {type=MovieSceneScriptingFloatChannel, default=1, const [1,22]}
       arm_r_pv_ik_ctrl.Scale.Y {type=MovieSceneScriptingFloatChannel, default=1, const [1,22]}
       arm_r_pv_ik_ctrl.Scale.Z {type=MovieSceneScriptingFloatChannel, default=1, const [1,22]}
       body_ctrl.Location.X {
        type=MovieSceneScriptingFloatChannel
        default=2.3592
        thinned_from=22
        keys [[0,2.3592],[1,2.4988],[2,2.6648],[3,2.8199],[4,2.9456],[5,3.0076],[6,3.0153],[7,3.006],[8,2.9849],[11,2.8969],[12,2.876],[14,2.8513],[16,2.8428],[19,2.8619],[20,2.8639],[21,2.8561]]
       }
       body_ctrl.Location.Y {
        type=MovieSceneScriptingFloatChannel
        default=-1.6892
        keys [[0,-1.6892],[1,-1.4622],[2,-1.1837],[3,-0.9244],[4,-0.717],[5,-0.541],[6,-0.3981],[7,-0.2953],[8,-0.2248],[9,-0.1795],[10,-0.1527],[11,-0.1377],[12,-0.1278],[13,-0.143],[14,-0.1955],[15,-0.2662],[16,-0.3354],[17,-0.3932],[18,-0.4425],[19,-0.4834],[20,-0.5164],[21,-0.5418]]
       }
       body_ctrl.Location.Z {
        type=MovieSceneScriptingFloatChannel
        default=-4.2875
        thinned_from=22
        keys [[0,-4.2875],[1,-4.7547],[2,-5.1408],[3,-5.3395],[4,-5.4725],[5,-5.5356],[6,-5.5507],[7,-5.5456],[9,-5.4949],[12,-5.3912],[15,-5.3291],[21,-5.3046]]
       }
       body_ctrl.Rotation.X {
        type=MovieSceneScriptingFloatChannel
        default=3.966
        thinned_from=22
        keys [[0,3.966],[1,3.2369],[2,2.9182],[3,2.7683],[4,2.6631],[5,2.7659],[6,2.9476],[7,2.9805],[8,2.9169],[9,2.8016],[10,2.674],[11,2.5692],[12,2.5201],[13,2.5361],[14,2.6009],[15,2.7133],[16,2.8823],[18,3.3014],[19,3.4891],[20,3.6287],[21,3.6931]]
       }
       body_ctrl.Rotation.Y {
        type=MovieSceneScriptingFloatChannel
        default=0.1669
        thinned_from=22
        keys [[0,0.1669],[1,1.5189],[2,2.5137],[3,2.6996],[4,2.7048],[5,2.791],[6,3.0025],[7,3.1824],[8,3.3156],[9,3.3953],[10,3.4182],[11,3.3809],[12,3.2783],[13,3.1273],[14,2.9653],[15,2.8147],[16,2.6952],[17,2.6168],[18,2.5676],[20,2.5071],[21,2.4702]]
       }
       body_ctrl.Rotation.Z {
        type=MovieSceneScriptingFloatChannel
        default=19.6568
        thinned_from=22
        keys [[0,19.6568],[1,25.0226],[2,29.9267],[3,33.2222],[4,35.8494],[5,37.6427],[6,38.7507],[7,39.4777],[8,39.8871],[9,40.0405],[10,39.9984],[12,39.5651],[13,39.1449],[16,37.2357],[18,36.2951],[19,35.9449],[20,35.714],[21,35.6319]]
       }
       body_ctrl.Scale.X {type=MovieSceneScriptingFloatChannel, default=1, const [1,22]}
       body_ctrl.Scale.Y {type=MovieSceneScriptingFloatChannel, default=1, const [1,22]}
       body_ctrl.Scale.Z {type=MovieSceneScriptingFloatChannel, default=1, const [1,22]}
       hips_ctrl.Location.X {type=MovieSceneScriptingFloatChannel, default=0, thinned_from=22, keys [[0,0],[21,0]]}
       hips_ctrl.Location.Y {type=MovieSceneScriptingFloatChannel, default=0, thinned_from=22, keys [[0,0],[21,0]]}
       hips_ctrl.Location.Z {type=MovieSceneScriptingFloatChannel, default=0, thinned_from=22, keys [[0,0],[21,0]]}
       hips_ctrl.Rotation.X {
        type=MovieSceneScriptingFloatChannel
        default=-6.3698
        thinned_from=22
        keys [[0,-6.3698],[1,-5.3347],[2,-4.7993],[3,-4.7237],[4,-4.8392],[5,-5.1638],[6,-5.5312],[7,-5.7691],[8,-5.9128],[9,-5.9886],[10,-6.016],[11,-6.0107],[14,-5.9345],[15,-5.937],[16,-5.9914],[19,-6.3062],[20,-6.3914],[21,-6.4387]]
       }
       hips_ctrl.Rotation.Y {
        type=MovieSceneScriptingFloatChannel
        default=0.7116
        keys [[0,0.7116],[1,1.3183],[2,1.6161],[3,1.6161],[4,1.4662],[5,1.0841],[6,0.5958],[7,0.2061],[8,-0.0792],[9,-0.2626],[10,-0.3492],[11,-0.3434],[12,-0.2465],[13,-0.0539],[14,0.2036],[15,0.4749],[16,0.7092],[17,0.8796],[18,1.0007],[19,1.0835],[20,1.1405],[21,1.1853]]
       }
       hips_ctrl.Rotation.Z {
        type=MovieSceneScriptingFloatChannel
        default=-2.2462
        thinned_from=22
        keys [[0,-2.2462],[1,-3.5517],[2,-4.8115],[3,-5.7193],[4,-6.3953],[5,-6.7828],[6,-6.9427],[7,-7.0001],[8,-6.9809],[9,-6.9068],[10,-6.7972],[12,-6.5415],[14,-6.2415],[15,-6.1207],[16,-6.0839],[17,-6.1452],[18,-6.2606],[19,-6.3917],[20,-6.5004],[21,-6.5486]]
       }
       hips_ctrl.Scale.X {type=MovieSceneScriptingFloatChannel, default=1, const [1,22]}
       hips_ctrl.Scale.Y {type=MovieSceneScriptingFloatChannel, default=1, const [1,22]}
       hips_ctrl.Scale.Z {type=MovieSceneScriptingFloatChannel, default=1, const [1,22]}
       hips_tan_ctrl.Location.X {
        type=MovieSceneScriptingFloatChannel
        default=-0.0012
        thinned_from=22
        keys [[0,-0.0012],[1,-0.0036],[3,-0.0072],[4,-0.0093],[7,-0.0168],[8,-0.0189],[9,-0.0205],[10,-0.0216],[11,-0.0221],[12,-0.0221],[13,-0.0214],[16,-0.0177],[18,-0.0161],[20,-0.0151],[21,-0.0149]]
       }
       hips_tan_ctrl.Location.Y {
        type=MovieSceneScriptingFloatChannel
        default=-0.1343
        thinned_from=22
        keys [[0,-0.1343],[1,-0.0911],[2,-0.0548],[3,-0.0385],[4,-0.0311],[5,-0.0289],[6,-0.0298],[7,-0.0336],[8,-0.0394],[10,-0.0528],[11,-0.0582],[12,-0.0616],[13,-0.063],[14,-0.0633],[15,-0.0631],[16,-0.0626],[17,-0.0618],[19,-0.0596],[20,-0.0587],[21,-0.0583]]
       }
       hips_tan_ctrl.Location.Z {
        type=MovieSceneScriptingFloatChannel
        default=-0.0171
        thinned_from=22
        keys [[0,-0.0171],[1,-0.0142],[2,-0.0116],[3,-0.0105],[4,-0.0101],[5,-0.0104],[9,-0.0129],[11,-0.0136],[13,-0.0135],[17,-0.0131],[21,-0.0132]]
       }
       hips_tan_ctrl.Rotation.X {type=MovieSceneScriptingFloatChannel, default=0, const [0,22]}
       hips_tan_ctrl.Rotation.Y {type=MovieSceneScriptingFloatChannel, default=0, const [0,22]}
       hips_tan_ctrl.Rotation.Z {type=MovieSceneScriptingFloatChannel, default=0, const [0,22]}
       hips_tan_ctrl.Scale.X {type=MovieSceneScriptingFloatChannel, default=1, const [1,22]}
       hips_tan_ctrl.Scale.Y {type=MovieSceneScriptingFloatChannel, default=1, const [1,22]}
       hips_tan_ctrl.Scale.Z {type=MovieSceneScriptingFloatChannel, default=1, const [1,22]}
       spine_01_ctrl.Location.X {
        type=MovieSceneScriptingFloatChannel
        default=-0.0046
        thinned_from=22
        keys [[0,-0.0046],[2,-0.0099],[3,-0.0119],[4,-0.0133],[5,-0.0142],[6,-0.0147],[8,-0.0149],[12,-0.014],[15,-0.0129],[16,-0.0128],[21,-0.0136]]
       }
       spine_01_ctrl.Location.Y {type=MovieSceneScriptingFloatChannel, default=0, thinned_from=22, keys [[0,0],[21,-0.0003]]}
       spine_01_ctrl.Location.Z {
        type=MovieSceneScriptingFloatChannel
        default=-0.0042
        thinned_from=22
        keys [[0,-0.0042],[2,-0.0021],[3,-0.0016],[6,-0.0008],[14,-0.0018],[21,-0.0015]]
       }
       spine_01_ctrl.Rotation.X {
        type=MovieSceneScriptingFloatChannel
        default=2.7821
        thinned_from=22
        keys [[0,2.7821],[1,2.0637],[2,1.3735],[3,1.0292],[4,0.8326],[5,0.6673],[6,0.5448],[7,0.5389],[8,0.6155],[9,0.7419],[10,0.887],[11,1.021],[12,1.1144],[13,1.1658],[14,1.1949],[15,1.2026],[16,1.1875],[17,1.152],[19,1.0604],[20,1.023],[21,1.0049]]
       }
       spine_01_ctrl.Rotation.Y {
        type=MovieSceneScriptingFloatChannel
        default=0.3068
        keys [[0,0.3068],[1,0.5293],[2,0.697],[3,0.7482],[4,0.7257],[5,0.6056],[6,0.4296],[7,0.2626],[8,0.116],[9,-0.0018],[10,-0.0841],[11,-0.1248],[12,-0.1176],[13,-0.0596],[14,0.0336],[15,0.1414],[16,0.2445],[17,0.3326],[18,0.4075],[19,0.4681],[20,0.513],[21,0.5413]]
       }
       spine_01_ctrl.Rotation.Z {
        type=MovieSceneScriptingFloatChannel
        default=2.1406
        thinned_from=22
        keys [[0,2.1406],[2,4.6513],[3,5.5742],[4,6.2769],[5,6.7106],[6,6.9333],[7,7.048],[8,7.073],[9,7.0274],[10,6.93],[11,6.7994],[12,6.6539],[13,6.4775],[14,6.2795],[15,6.1156],[16,6.0422],[17,6.0767],[18,6.1725],[19,6.2898],[20,6.3887],[21,6.4292]]
       }
       spine_01_ctrl.Scale.X {type=MovieSceneScriptingFloatChannel, default=1, const [1,22]}
       spine_01_ctrl.Scale.Y {type=MovieSceneScriptingFloatChannel, default=1, const [1,22]}
       spine_01_ctrl.Scale.Z {type=MovieSceneScriptingFloatChannel, default=1, const [1,22]}
       spine_02_ctrl.Location.X {
        type=MovieSceneScriptingFloatChannel
        default=-0.0031
        thinned_from=22
        keys [[0,-0.0031],[2,-0.0067],[3,-0.0081],[4,-0.0091],[5,-0.0097],[7,-0.0102],[12,-0.0097],[15,-0.0089],[17,-0.0088],[21,-0.0093]]
       }
       spine_02_ctrl.Location.Y {type=MovieSceneScriptingFloatChannel, default=0, thinned_from=22, keys [[0,0],[21,-0.0005]]}
       spine_02_ctrl.Location.Z {
        type=MovieSceneScriptingFloatChannel
        default=-0.0024
        thinned_from=22
        keys [[0,-0.0024],[3,-0.0011],[6,-0.0001],[8,0.0001],[16,-0.0007],[21,-0.0004]]
       }
       spine_02_ctrl.Rotation.X {
        type=MovieSceneScriptingFloatChannel
        default=1.6187
        keys [[0,1.6187],[1,1.3613],[2,0.9955],[3,0.7673],[4,0.5861],[5,0.3309],[6,0.0733],[7,-0.0462],[8,-0.0617],[9,-0.0055],[10,0.0919],[11,0.2015],[12,0.2951],[13,0.3673],[14,0.4261],[15,0.4621],[16,0.4626],[17,0.4302],[18,0.3841],[19,0.3351],[20,0.2949],[21,0.2758]]
       }
       spine_02_ctrl.Rotation.Y {
        type=MovieSceneScriptingFloatChannel
        default=0.5198
        keys [[0,0.5198],[1,0.8959],[2,1.1575],[3,1.2431],[4,1.2299],[5,1.0877],[6,0.8685],[7,0.6559],[8,0.4657],[9,0.3105],[10,0.201],[11,0.1471],[12,0.1593],[13,0.2421],[14,0.3742],[15,0.5288],[16,0.6805],[17,0.8133],[18,0.9266],[19,1.0188],[20,1.0879],[21,1.1316]]
       }
       spine_02_ctrl.Rotation.Z {
        type=MovieSceneScriptingFloatChannel
        default=2.1206
        thinned_from=22
        keys [[0,2.1206],[2,4.621],[3,5.5458],[4,6.2511],[5,6.6869],[6,6.914],[7,7.0357],[8,7.0684],[9,7.0293],[10,6.9361],[11,6.8068],[12,6.6593],[13,6.4773],[14,6.2714],[15,6.0993],[16,6.0184],[17,6.0467],[18,6.1377],[19,6.2513],[20,6.3474],[21,6.3859]]
       }
       spine_02_ctrl.Scale.X {type=MovieSceneScriptingFloatChannel, default=1, const [1,22]}
       spine_02_ctrl.Scale.Y {type=MovieSceneScriptingFloatChannel, default=1, const [1,22]}
       spine_02_ctrl.Scale.Z {type=MovieSceneScriptingFloatChannel, default=1, const [1,22]}
       spine_03_ctrl.Location.X {
        type=MovieSceneScriptingFloatChannel
        default=0.0017
        thinned_from=22
        keys [[0,0.0017],[2,0.0037],[4,0.0049],[6,0.0053],[16,0.0046],[21,0.005]]
       }
       spine_03_ctrl.Location.Y {type=MovieSceneScriptingFloatChannel, default=0, thinned_from=22, keys [[0,0],[21,0.0004]]}
       spine_03_ctrl.Location.Z {type=MovieSceneScriptingFloatChannel, default=0.0002, thinned_from=22, keys [[0,0.0002],[1,0.0004],[4,0.0001],[7,-0.0008],[21,-0.0006]]}
       spine_03_ctrl.Rotation.X {
        type=MovieSceneScriptingFloatChannel
        default=0.2795
        keys [[0,0.2795],[1,0.493],[2,0.4276],[3,0.2887],[4,0.0891],[5,-0.3233],[6,-0.7942],[7,-1.0762],[8,-1.2079],[9,-1.2265],[10,-1.1669],[11,-1.0626],[12,-0.946],[13,-0.8276],[14,-0.712],[15,-0.6245],[16,-0.5939],[17,-0.6179],[18,-0.664],[19,-0.7173],[20,-0.7619],[21,-0.7816]]
       }
       spine_03_ctrl.Rotation.Y {
        type=MovieSceneScriptingFloatChannel
        default=0.8131
        thinned_from=22
        keys [[0,0.8131],[1,1.3473],[2,1.6617],[3,1.7348],[4,1.6913],[5,1.5002],[6,1.2177],[7,0.9387],[8,0.6857],[9,0.4772],[10,0.3298],[11,0.2595],[12,0.2832],[13,0.41],[14,0.6105],[16,1.0805],[17,1.2871],[18,1.4623],[19,1.6046],[20,1.7113],[21,1.7796]]
       }
       spine_03_ctrl.Rotation.Z {
        type=MovieSceneScriptingFloatChannel
        default=2.0944
        keys [[0,2.0944],[1,3.3316],[2,4.5768],[3,5.4966],[4,6.196],[5,6.6257],[6,6.8508],[7,6.9759],[8,7.0148],[9,6.9827],[10,6.8957],[11,6.7703],[12,6.6235],[13,6.4388],[14,6.228],[15,6.0494],[16,5.9607],[17,5.9809],[18,6.064],[19,6.1705],[20,6.2611],[21,6.2964]]
       }
       spine_03_ctrl.Scale.X {type=MovieSceneScriptingFloatChannel, default=1, const [1,22]}
       spine_03_ctrl.Scale.Y {type=MovieSceneScriptingFloatChannel, default=1, const [1,22]}
       spine_03_ctrl.Scale.Z {type=MovieSceneScriptingFloatChannel, default=1, const [1,22]}
       spine_04_ctrl.Location.X {type=MovieSceneScriptingFloatChannel, default=0}
       spine_04_ctrl.Location.Y {type=MovieSceneScriptingFloatChannel, default=0}
       spine_04_ctrl.Location.Z {type=MovieSceneScriptingFloatChannel, default=0}
       spine_04_ctrl.Rotation.X {type=MovieSceneScriptingFloatChannel, default=0}
       spine_04_ctrl.Rotation.Y {type=MovieSceneScriptingFloatChannel, default=0}
       spine_04_ctrl.Rotation.Z {type=MovieSceneScriptingFloatChannel, default=0}
       spine_04_ctrl.Scale.X {type=MovieSceneScriptingFloatChannel, default=1}
       spine_04_ctrl.Scale.Y {type=MovieSceneScriptingFloatChannel, default=1}
       spine_04_ctrl.Scale.Z {type=MovieSceneScriptingFloatChannel, default=1}
       head_ctrl.Location.X {type=MovieSceneScriptingFloatChannel, default=0, thinned_from=22, keys [[0,0],[21,0]]}
       head_ctrl.Location.Y {type=MovieSceneScriptingFloatChannel, default=0, thinned_from=22, keys [[0,0],[21,0]]}
       head_ctrl.Location.Z {type=MovieSceneScriptingFloatChannel, default=0, thinned_from=22, keys [[0,0],[21,0]]}
       head_ctrl.Rotation.X {
        type=MovieSceneScriptingFloatChannel
        default=3.3063
        thinned_from=22
        keys [[0,3.3063],[1,4.604],[2,5.4049],[3,5.5524],[4,5.4867],[5,5.3553],[6,5.3151],[7,5.3867],[8,5.5131],[9,5.6513],[10,5.7668],[11,5.8295],[12,5.8105],[13,5.702],[14,5.5334],[15,5.3235],[17,4.8336],[18,4.6175],[19,4.4395],[20,4.3099],[21,4.2402]]
       }
       head_ctrl.Rotation.Y {
        type=MovieSceneScriptingFloatChannel
        default=-1.2804
        thinned_from=22
        keys [[0,-1.2804],[1,-3.8055],[2,-5.8597],[3,-6.717],[4,-7.2259],[5,-7.6411],[6,-8.034],[7,-8.2909],[8,-8.4276],[9,-8.4666],[10,-8.4307],[11,-8.3401],[12,-8.211],[14,-7.9281],[15,-7.8231],[16,-7.777],[20,-7.7423],[21,-7.7167]]
       }
       head_ctrl.Rotation.Z {
        type=MovieSceneScriptingFloatChannel
        default=-18.4649
        thinned_from=22
        keys [[0,-18.4649],[1,-21.2738],[2,-23.939],[3,-25.6593],[4,-26.9564],[5,-28.0169],[6,-28.8583],[7,-29.4088],[8,-29.7153],[9,-29.8232],[10,-29.7771],[11,-29.6214],[12,-29.4005],[13,-29.0154],[14,-28.4237],[16,-27.1358],[17,-26.6534],[18,-26.2982],[19,-26.0585],[20,-25.9231],[21,-25.8805]]
       }
       head_ctrl.Scale.X {type=MovieSceneScriptingFloatChannel, default=1, const [1,22]}
       head_ctrl.Scale.Y {type=MovieSceneScriptingFloatChannel, default=1, const [1,22]}
       head_ctrl.Scale.Z {type=MovieSceneScriptingFloatChannel, default=1, const [1,22]}
       clavicle_l_ctrl.Location.X {
        type=MovieSceneScriptingFloatChannel
        default=0.0104
        keys [[0,0.0104],[1,0.0006],[2,0.0058],[3,0.0179],[4,0.0316],[5,0.049],[6,0.0677],[7,0.0837],[8,0.0963],[9,0.1052],[10,0.1101],[11,0.1105],[12,0.1063],[13,0.0965],[14,0.0827],[15,0.0677],[16,0.0542],[17,0.0438],[18,0.0361],[19,0.0304],[20,0.0263],[21,0.0233]]
       }
       clavicle_l_ctrl.Location.Y {
        type=MovieSceneScriptingFloatChannel
        default=-0.2747
        thinned_from=22
        keys [[0,-0.2747],[1,-0.1902],[2,-0.1256],[3,-0.0893],[4,-0.0698],[5,-0.0744],[6,-0.0911],[7,-0.1033],[8,-0.1116],[9,-0.1171],[10,-0.1204],[11,-0.1222],[12,-0.1231],[13,-0.1245],[15,-0.1287],[16,-0.1305],[17,-0.131],[18,-0.1301],[20,-0.1275],[21,-0.1272]]
       }
       clavicle_l_ctrl.Location.Z {
        type=MovieSceneScriptingFloatChannel
        default=-0.0337
        keys [[0,-0.0337],[1,-0.0176],[2,-0.0122],[3,-0.0139],[4,-0.0183],[5,-0.0269],[6,-0.0375],[7,-0.0464],[8,-0.0535],[9,-0.0584],[10,-0.0611],[11,-0.0616],[12,-0.0596],[13,-0.055],[14,-0.0486],[15,-0.0416],[16,-0.0354],[17,-0.0305],[18,-0.0268],[19,-0.024],[20,-0.022],[21,-0.0205]]
       }
       clavicle_l_ctrl.Rotation.X {
        type=MovieSceneScriptingFloatChannel
        default=-4.6567
        thinned_from=22
        keys [[0,-4.6567],[1,-4.0804],[2,-3.7695],[3,-3.732],[4,-3.8442],[5,-4.1495],[6,-4.5083],[7,-4.7512],[8,-4.897],[9,-4.9644],[10,-4.9709],[11,-4.9345],[12,-4.871],[14,-4.6994],[15,-4.6269],[16,-4.5941],[17,-4.5977],[20,-4.656],[21,-4.6636]]
       }
       clavicle_l_ctrl.Rotation.Y {
        type=MovieSceneScriptingFloatChannel
        default=0.1835
        thinned_from=22
        keys [[0,0.1835],[1,2.0493],[2,3.4823],[3,4.2034],[4,4.6479],[5,4.7931],[6,4.7568],[7,4.6899],[10,4.4623],[11,4.4366],[12,4.4681],[13,4.5368],[14,4.6552],[15,4.7983],[16,4.9618],[17,5.0647],[18,5.1368],[19,5.1736],[20,5.1847],[21,5.1795]]
       }
       clavicle_l_ctrl.Rotation.Z {
        type=MovieSceneScriptingFloatChannel
        default=3.1676
        thinned_from=22
        keys [[0,3.1676],[1,5.2902],[2,7.0441],[3,7.5877],[5,7.443],[6,7.5332],[7,7.5555],[8,7.5225],[9,7.4475],[10,7.353],[11,7.2271],[13,6.9456],[14,6.7731],[15,6.6283],[16,6.5544],[17,6.6135],[18,6.7548],[19,6.9267],[20,7.0714],[21,7.1304]]
       }
       clavicle_l_ctrl.Scale.X {type=MovieSceneScriptingFloatChannel, default=1, const [1,22]}
       clavicle_l_ctrl.Scale.Y {type=MovieSceneScriptingFloatChannel, default=1, const [1,22]}
       clavicle_l_ctrl.Scale.Z {type=MovieSceneScriptingFloatChannel, default=1, const [1,22]}
       clavicle_r_ctrl.Location.X {
        type=MovieSceneScriptingFloatChannel
        default=0.0297
        thinned_from=22
        keys [[0,0.0297],[1,0.0301],[2,0.0161],[5,-0.0387],[6,-0.056],[7,-0.0707],[8,-0.0822],[9,-0.0901],[10,-0.0941],[11,-0.0938],[12,-0.0888],[13,-0.0783],[14,-0.0637],[15,-0.048],[16,-0.0342],[17,-0.024],[18,-0.0168],[19,-0.0119],[20,-0.0084],[21,-0.0057]]
       }
       clavicle_r_ctrl.Location.Y {
        type=MovieSceneScriptingFloatChannel
        default=-0.4476
        thinned_from=22
        keys [[0,-0.4476],[1,-0.464],[2,-0.501],[3,-0.5399],[4,-0.5774],[5,-0.6173],[6,-0.6527],[7,-0.675],[8,-0.6863],[9,-0.6891],[10,-0.6853],[11,-0.6769],[12,-0.666],[13,-0.6524],[14,-0.6374],[15,-0.625],[16,-0.6193],[17,-0.6212],[18,-0.6268],[20,-0.6398],[21,-0.6423]]
       }
       clavicle_r_ctrl.Location.Z {
        type=MovieSceneScriptingFloatChannel
        default=-0.0716
        thinned_from=22
        keys [[0,-0.0716],[1,-0.0749],[2,-0.0755],[3,-0.0749],[4,-0.0744],[5,-0.0749],[6,-0.0752],[7,-0.0738],[8,-0.0714],[10,-0.0659],[11,-0.064],[12,-0.0636],[13,-0.0651],[14,-0.0682],[15,-0.0724],[17,-0.0825],[18,-0.0871],[19,-0.091],[20,-0.094],[21,-0.0959]]
       }
       clavicle_r_ctrl.Rotation.X {
        type=MovieSceneScriptingFloatChannel
        default=-4.8824
        thinned_from=22
        keys [[0,-4.8824],[1,-4.4664],[2,-4.3665],[3,-4.4772],[4,-4.6864],[5,-5.057],[6,-5.4676],[7,-5.7494],[8,-5.9211],[9,-6.0017],[10,-6.0087],[11,-5.961],[12,-5.8749],[13,-5.7538],[14,-5.6156],[15,-5.4936],[16,-5.4235],[17,-5.406],[18,-5.4148],[20,-5.4559],[21,-5.4617]]
       }
       clavicle_r_ctrl.Rotation.Y {
        type=MovieSceneScriptingFloatChannel
        default=-1.4433
        keys [[0,-1.4433],[1,-0.6528],[2,0.2138],[3,0.8548],[4,1.3901],[5,1.7409],[6,1.9641],[7,2.1644],[8,2.3344],[9,2.467],[10,2.5412],[11,2.577],[12,2.5617],[13,2.462],[14,2.3149],[15,2.1447],[16,1.9917],[17,1.8085],[18,1.6324],[19,1.4744],[20,1.3608],[21,1.3143]]
       }
       clavicle_r_ctrl.Rotation.Z {
        type=MovieSceneScriptingFloatChannel
        default=-2.9252
        thinned_from=22
        keys [[0,-2.9252],[1,-5.8904],[2,-8.478],[3,-9.9176],[4,-10.7944],[5,-11.2664],[6,-11.5243],[7,-11.6677],[8,-11.7152],[9,-11.6856],[10,-11.5927],[12,-11.3282],[14,-10.9379],[15,-10.7677],[16,-10.6921],[17,-10.7057],[20,-10.9161],[21,-10.9533]]
       }
       clavicle_r_ctrl.Scale.X {type=MovieSceneScriptingFloatChannel, default=1, const [1,22]}
       clavicle_r_ctrl.Scale.Y {type=MovieSceneScriptingFloatChannel, default=1, const [1,22]}
       clavicle_r_ctrl.Scale.Z {type=MovieSceneScriptingFloatChannel, default=1, const [1,22]}
       neck_01_ctrl.Location.X {
        type=MovieSceneScriptingFloatChannel
        default=-0.0307
        keys [[0,-0.0307],[1,-0.0475],[2,-0.0269],[3,0.0089],[4,0.0474],[5,0.0913],[6,0.1367],[7,0.1754],[8,0.2061],[9,0.2275],[10,0.239],[11,0.2395],[12,0.2282],[13,0.2029],[14,0.1674],[15,0.1289],[16,0.0945],[17,0.0683],[18,0.0492],[19,0.0355],[20,0.0256],[21,0.0182]]
       }
       neck_01_ctrl.Location.Y {
        type=MovieSceneScriptingFloatChannel
        default=-0.9441
        thinned_from=22
        keys [[0,-0.9441],[1,-0.8552],[2,-0.8194],[3,-0.8231],[4,-0.8471],[5,-0.9054],[6,-0.9739],[7,-1.019],[8,-1.0449],[9,-1.0556],[10,-1.0549],[11,-1.0463],[12,-1.0332],[14,-1],[15,-0.9865],[16,-0.9813],[17,-0.9844],[20,-1.0044],[21,-1.0072]]
       }
       neck_01_ctrl.Location.Z {
        type=MovieSceneScriptingFloatChannel
        default=-0.1002
        thinned_from=22
        keys [[0,-0.1002],[1,-0.0871],[2,-0.082],[3,-0.0824],[4,-0.0859],[5,-0.0947],[6,-0.1055],[7,-0.1131],[8,-0.1178],[9,-0.12],[10,-0.1201],[11,-0.1187],[12,-0.1164],[14,-0.11],[15,-0.1074],[16,-0.1063],[17,-0.1065],[18,-0.1074],[20,-0.1095],[21,-0.1099]]
       }
       neck_01_ctrl.Rotation.X {
        type=MovieSceneScriptingFloatChannel
        default=-7.0422
        keys [[0,-7.0422],[1,-5.3041],[2,-3.4618],[3,-2.6853],[4,-2.4381],[5,-2.3877],[6,-2.4476],[7,-2.7142],[8,-3.0935],[9,-3.5116],[10,-3.9081],[11,-4.2295],[12,-4.425],[13,-4.4884],[14,-4.4688],[15,-4.3904],[16,-4.2633],[17,-4.102],[18,-3.9321],[19,-3.773],[20,-3.6462],[21,-3.5738]]
       }
       neck_01_ctrl.Rotation.Y {
        type=MovieSceneScriptingFloatChannel
        default=-3.0923
        keys [[0,-3.0923],[1,-5.0018],[2,-5.8347],[3,-5.7263],[4,-5.3227],[5,-4.6251],[6,-3.8708],[7,-3.3358],[8,-3.0005],[9,-2.8335],[10,-2.8005],[11,-2.8694],[12,-3.0135],[13,-3.2546],[14,-3.5897],[15,-3.9571],[16,-4.2919],[17,-4.562],[18,-4.7797],[19,-4.944],[20,-5.055],[21,-5.1142]]
       }
       neck_01_ctrl.Rotation.Z {
        type=MovieSceneScriptingFloatChannel
        default=-16.1353
        thinned_from=22
        keys [[0,-16.1353],[1,-25.3715],[2,-34.284],[3,-40.6182],[4,-45.599],[5,-48.8458],[6,-50.6903],[7,-51.8059],[8,-52.3193],[9,-52.3524],[10,-52.0243],[11,-51.4523],[12,-50.7527],[13,-49.7916],[15,-47.3764],[16,-46.5425],[17,-46.137],[18,-45.9704],[21,-46.0485]]
       }
       neck_01_ctrl.Scale.X {type=MovieSceneScriptingFloatChannel, default=1, const [1,22]}
       neck_01_ctrl.Scale.Y {type=MovieSceneScriptingFloatChannel, default=1, const [1,22]}
       neck_01_ctrl.Scale.Z {type=MovieSceneScriptingFloatChannel, default=1, const [1,22]}
       neck_02_ctrl.Location.X {type=MovieSceneScriptingFloatChannel, default=0, const [0,22]}
       neck_02_ctrl.Location.Y {type=MovieSceneScriptingFloatChannel, default=0, const [0,22]}
       neck_02_ctrl.Location.Z {type=MovieSceneScriptingFloatChannel, default=0, const [0,22]}
       neck_02_ctrl.Rotation.X {type=MovieSceneScriptingFloatChannel, default=0, const [0,22]}
       neck_02_ctrl.Rotation.Y {type=MovieSceneScriptingFloatChannel, default=0, const [0,22]}
       neck_02_ctrl.Rotation.Z {type=MovieSceneScriptingFloatChannel, default=0, const [0,22]}
       neck_02_ctrl.Scale.X {type=MovieSceneScriptingFloatChannel, default=1, const [1,22]}
       neck_02_ctrl.Scale.Y {type=MovieSceneScriptingFloatChannel, default=1, const [1,22]}
       neck_02_ctrl.Scale.Z {type=MovieSceneScriptingFloatChannel, default=1, const [1,22]}
       head_ik_ctrl.Location.X {
        type=MovieSceneScriptingFloatChannel
        default=1.1264
        thinned_from=22
        keys [[0,1.1264],[1,1.8374],[2,2.4366],[3,2.7641],[4,2.9663],[5,3.0579],[6,3.0824],[7,3.0771],[8,3.051],[9,3.0115],[12,2.8742],[14,2.7934],[15,2.7666],[16,2.7633],[17,2.7856],[19,2.8587],[20,2.8892],[21,2.9024]]
       }
       head_ik_ctrl.Location.Y {
        type=MovieSceneScriptingFloatChannel
        default=-2.212
        thinned_from=22
        keys [[0,-2.212],[1,-1.9806],[2,-1.7782],[3,-1.7078],[4,-1.7027],[6,-1.7281],[7,-1.7624],[8,-1.811],[10,-1.9238],[11,-1.9742],[12,-2.0117],[13,-2.038],[14,-2.0579],[15,-2.07],[16,-2.0735],[17,-2.0667],[20,-2.0209],[21,-2.0147]]
       }
       head_ik_ctrl.Location.Z {
        type=MovieSceneScriptingFloatChannel
        default=0.3495
        thinned_from=22
        keys [[0,0.3495],[1,0.3007],[2,0.2385],[3,0.2105],[4,0.1889],[5,0.1369],[6,0.0704],[7,0.0298],[8,0.0114],[9,0.0108],[10,0.0234],[11,0.0441],[12,0.0675],[13,0.0924],[14,0.1177],[15,0.139],[16,0.1516],[17,0.1553],[18,0.1546],[20,0.1488],[21,0.1485]]
       }
       head_ik_ctrl.Rotation.X {
        type=MovieSceneScriptingFloatChannel
        default=3.3063
        thinned_from=22
        keys [[0,3.3063],[1,4.604],[2,5.4049],[3,5.5524],[4,5.4867],[5,5.3553],[6,5.3151],[7,5.3867],[8,5.5131],[9,5.6513],[10,5.7668],[11,5.8295],[12,5.8105],[13,5.702],[14,5.5334],[15,5.3236],[17,4.8336],[18,4.6175],[19,4.4395],[20,4.3099],[21,4.2402]]
       }
       head_ik_ctrl.Rotation.Y {
        type=MovieSceneScriptingFloatChannel
        default=-1.2804
        thinned_from=22
        keys [[0,-1.2804],[1,-3.8055],[2,-5.8597],[3,-6.717],[4,-7.2259],[5,-7.6411],[6,-8.034],[7,-8.2909],[8,-8.4276],[9,-8.4666],[10,-8.4307],[11,-8.3401],[12,-8.211],[14,-7.9281],[15,-7.8231],[16,-7.777],[20,-7.7423],[21,-7.7167]]
       }
       head_ik_ctrl.Rotation.Z {
        type=MovieSceneScriptingFloatChannel
        default=-18.4649
        thinned_from=22
        keys [[0,-18.4649],[1,-21.2738],[2,-23.939],[3,-25.6593],[4,-26.9564],[5,-28.0169],[6,-28.8583],[7,-29.4088],[8,-29.7153],[9,-29.8231],[10,-29.7771],[11,-29.6214],[12,-29.4005],[13,-29.0154],[14,-28.4237],[16,-27.1358],[17,-26.6534],[18,-26.2982],[19,-26.0585],[20,-25.923],[21,-25.8805]]
       }
       head_ik_ctrl.Scale.X {type=MovieSceneScriptingFloatChannel, default=1, const [1,22]}
       head_ik_ctrl.Scale.Y {type=MovieSceneScriptingFloatChannel, default=1, const [1,22]}
       head_ik_ctrl.Scale.Z {type=MovieSceneScriptingFloatChannel, default=1, const [1,22]}
       chest_ctrl.Location.X {
        type=MovieSceneScriptingFloatChannel
        default=-0.1133
        keys [[0,-0.1133],[1,-0.1847],[2,-0.213],[3,-0.1982],[4,-0.1594],[5,-0.084],[6,0.0085],[7,0.0864],[8,0.1472],[9,0.1898],[10,0.2141],[11,0.2198],[12,0.2062],[13,0.1718],[14,0.1224],[15,0.0687],[16,0.0208],[17,-0.0161],[18,-0.0446],[19,-0.0658],[20,-0.0809],[21,-0.0914]]
       }
       chest_ctrl.Location.Y {
        type=MovieSceneScriptingFloatChannel
        default=0.9053
        thinned_from=22
        keys [[0,0.9053],[1,0.6808],[2,0.4472],[3,0.3243],[4,0.2501],[5,0.182],[6,0.1295],[7,0.1243],[8,0.1519],[9,0.199],[10,0.2532],[11,0.3029],[12,0.3368],[13,0.3539],[14,0.3614],[15,0.3601],[16,0.3491],[17,0.3298],[19,0.2846],[20,0.2666],[21,0.2574]]
       }
       chest_ctrl.Location.Z {
        type=MovieSceneScriptingFloatChannel
        default=0.0534
        thinned_from=22
        keys [[0,0.0534],[1,0.045],[2,0.0326],[3,0.0249],[5,0.015],[6,0.0105],[7,0.0095],[8,0.0111],[9,0.0143],[11,0.0217],[12,0.0245],[13,0.0263],[14,0.0274],[15,0.0279],[16,0.0273],[17,0.026],[19,0.0225],[20,0.021],[21,0.0202]]
       }
       chest_ctrl.Rotation.X {
        type=MovieSceneScriptingFloatChannel
        default=4.7145
        keys [[0,4.7145],[1,4.0183],[2,2.9831],[3,2.3282],[4,1.7765],[5,0.9213],[6,0.0141],[7,-0.4556],[8,-0.587],[9,-0.4748],[10,-0.2103],[11,0.118],[12,0.423],[13,0.6875],[14,0.9261],[15,1.0966],[16,1.1499],[17,1.0919],[18,0.9851],[19,0.8633],[20,0.7622],[21,0.7181]]
       }
       chest_ctrl.Rotation.Y {
        type=MovieSceneScriptingFloatChannel
        default=1.372
        keys [[0,1.372],[1,2.4444],[2,3.2026],[3,3.4346],[4,3.3791],[5,2.9777],[6,2.3586],[7,1.7196],[8,1.1182],[9,0.6047],[10,0.2244],[11,0.0208],[12,0.0368],[13,0.2906],[14,0.7103],[15,1.208],[16,1.7013],[17,2.1356],[18,2.5068],[19,2.8097],[20,3.0371],[21,3.1807]]
       }
       chest_ctrl.Rotation.Z {
        type=MovieSceneScriptingFloatChannel
        default=6.4352
        keys [[0,6.4352],[1,10.2075],[2,13.9484],[3,16.7011],[4,18.7936],[5,20.0707],[6,20.725],[7,21.0763],[8,21.1682],[9,21.0485],[10,20.7686],[11,20.382],[12,19.9433],[13,19.4057],[14,18.8002],[15,18.2971],[16,18.065],[17,18.1565],[18,18.4325],[19,18.774],[20,19.0622],[21,19.1784]]
       }
       chest_ctrl.Scale.X {type=MovieSceneScriptingFloatChannel, default=1, const [1,22]}
       chest_ctrl.Scale.Y {type=MovieSceneScriptingFloatChannel, default=1, const [1,22]}
       chest_ctrl.Scale.Z {type=MovieSceneScriptingFloatChannel, default=1, const [1,22]}
       chest_tan_ctrl.Location.X {
        type=MovieSceneScriptingFloatChannel
        default=-0.0859
        keys [[0,-0.0859],[1,-0.1446],[2,-0.173],[3,-0.1718],[4,-0.157],[5,-0.125],[6,-0.0826],[7,-0.042],[8,-0.0059],[9,0.023],[10,0.0428],[11,0.0512],[12,0.0461],[13,0.0261],[14,-0.0045],[15,-0.0397],[16,-0.0735],[17,-0.1021],[18,-0.1256],[19,-0.1441],[20,-0.1579],[21,-0.167]]
       }
       chest_tan_ctrl.Location.Y {
        type=MovieSceneScriptingFloatChannel
        default=0.0073
        thinned_from=22
        keys [[0,0.0073],[1,0.0326],[2,0.0197],[3,-0.0017],[4,-0.0306],[5,-0.0885],[6,-0.1543],[7,-0.1934],[8,-0.2114],[9,-0.2135],[10,-0.2046],[11,-0.1896],[13,-0.1561],[14,-0.1398],[15,-0.1274],[16,-0.1231],[17,-0.1264],[18,-0.133],[19,-0.1405],[20,-0.1469],[21,-0.1496]]
       }
       chest_tan_ctrl.Location.Z {
        type=MovieSceneScriptingFloatChannel
        default=0.2645
        thinned_from=22
        keys [[0,0.2645],[1,0.268],[2,0.268],[3,0.2664],[4,0.2637],[5,0.258],[6,0.2518],[7,0.2481],[8,0.2463],[9,0.2458],[10,0.2463],[15,0.2525],[16,0.2531],[21,0.2522]]
       }
       chest_tan_ctrl.Rotation.X {
        type=MovieSceneScriptingFloatChannel
        default=0
        thinned_from=22
        keys [[0,0],[1,-0.414],[2,-0.2938],[3,-0.1213],[4,0.094],[5,0.4943],[6,0.934],[7,1.1823],[8,1.2827],[9,1.2754],[10,1.198],[12,0.9724],[13,0.8683],[14,0.774],[15,0.7103],[16,0.703],[17,0.7488],[18,0.815],[19,0.8859],[20,0.9444],[21,0.9724]]
       }
       chest_tan_ctrl.Rotation.Y {
        type=MovieSceneScriptingFloatChannel
        default=0
        keys [[0,0],[1,-1.3736],[2,-1.6905],[3,-1.7545],[4,-1.691],[5,-1.4529],[6,-1.1141],[7,-0.8009],[8,-0.5329],[9,-0.3245],[10,-0.1872],[11,-0.0346],[12,-0.0854],[13,-0.3146],[14,-0.5296],[15,-0.7755],[16,-1.0129],[17,-1.2156],[18,-1.3839],[19,-1.518],[20,-1.6178],[21,-1.683]]
       }
       chest_tan_ctrl.Rotation.Z {
        type=MovieSceneScriptingFloatChannel
        default=0
        keys [[0,0],[1,-3.3208],[2,-4.5663],[3,-5.4904],[4,-6.1961],[5,-6.6362],[6,-6.8683],[7,-6.993],[8,-7.028],[9,-6.9914],[10,-6.901],[11,-6.7742],[12,-6.6279],[13,-6.4441],[14,-6.2353],[15,-6.0588],[16,-5.9725],[17,-5.9958],[18,-6.0823],[19,-6.1923],[20,-6.2858],[21,-6.3228]]
       }
       chest_tan_ctrl.Scale.X {type=MovieSceneScriptingFloatChannel, default=1, const [1,22]}
       chest_tan_ctrl.Scale.Y {type=MovieSceneScriptingFloatChannel, default=1, const [1,22]}
       chest_tan_ctrl.Scale.Z {type=MovieSceneScriptingFloatChannel, default=1, const [1,22]}
      ]
     }
     {
      type=MovieSceneControlRigParameterSection
      end_frame=30
      start_frame=0
      channels [
       global_ctrl.Location.X {type=MovieSceneScriptingFloatChannel, default=0}
       global_ctrl.Location.Y {type=MovieSceneScriptingFloatChannel, default=0}
       global_ctrl.Location.Z {type=MovieSceneScriptingFloatChannel, default=0}
       global_ctrl.Rotation.X {type=MovieSceneScriptingFloatChannel, default=0}
       global_ctrl.Rotation.Y {type=MovieSceneScriptingFloatChannel, default=0}
       global_ctrl.Rotation.Z {type=MovieSceneScriptingFloatChannel, default=0}
       root_ctrl.Location.X {type=MovieSceneScriptingFloatChannel, default=0}
       root_ctrl.Location.Y {type=MovieSceneScriptingFloatChannel, default=0}
       root_ctrl.Location.Z {type=MovieSceneScriptingFloatChannel, default=0}
       root_ctrl.Rotation.X {type=MovieSceneScriptingFloatChannel, default=0}
       root_ctrl.Rotation.Y {type=MovieSceneScriptingFloatChannel, default=0}
       root_ctrl.Rotation.Z {type=MovieSceneScriptingFloatChannel, default=0}
       root_ctrl.Scale.X {type=MovieSceneScriptingFloatChannel, default=0}
       root_ctrl.Scale.Y {type=MovieSceneScriptingFloatChannel, default=0}
       root_ctrl.Scale.Z {type=MovieSceneScriptingFloatChannel, default=0}
       body_offset_ctrl.Location.X {type=MovieSceneScriptingFloatChannel, default=0}
       body_offset_ctrl.Location.Y {type=MovieSceneScriptingFloatChannel, default=0}
       body_offset_ctrl.Location.Z {type=MovieSceneScriptingFloatChannel, default=0}
       body_offset_ctrl.Rotation.X {type=MovieSceneScriptingFloatChannel, default=0}
       body_offset_ctrl.Rotation.Y {type=MovieSceneScriptingFloatChannel, default=0}
       body_offset_ctrl.Rotation.Z {type=MovieSceneScriptingFloatChannel, default=0}
       body_offset_ctrl.Scale.X {type=MovieSceneScriptingFloatChannel, default=0}
       body_offset_ctrl.Scale.Y {type=MovieSceneScriptingFloatChannel, default=0}
       body_offset_ctrl.Scale.Z {type=MovieSceneScriptingFloatChannel, default=0}
       thigh_l_fk_ctrl.Location.X {type=MovieSceneScriptingFloatChannel, default=0}
       thigh_l_fk_ctrl.Location.Y {type=MovieSceneScriptingFloatChannel, default=0}
       thigh_l_fk_ctrl.Location.Z {type=MovieSceneScriptingFloatChannel, default=0}
       thigh_l_fk_ctrl.Rotation.X {type=MovieSceneScriptingFloatChannel, default=0}
       thigh_l_fk_ctrl.Rotation.Y {type=MovieSceneScriptingFloatChannel, default=0}
       thigh_l_fk_ctrl.Rotation.Z {type=MovieSceneScriptingFloatChannel, default=0}
       thigh_l_fk_ctrl.Scale.X {type=MovieSceneScriptingFloatChannel, default=0}
       thigh_l_fk_ctrl.Scale.Y {type=MovieSceneScriptingFloatChannel, default=0}
       thigh_l_fk_ctrl.Scale.Z {type=MovieSceneScriptingFloatChannel, default=0}
       calf_l_fk_ctrl.Location.X {type=MovieSceneScriptingFloatChannel, default=0}
       calf_l_fk_ctrl.Location.Y {type=MovieSceneScriptingFloatChannel, default=0}
       calf_l_fk_ctrl.Location.Z {type=MovieSceneScriptingFloatChannel, default=0}
       calf_l_fk_ctrl.Rotation.X {type=MovieSceneScriptingFloatChannel, default=0}
       calf_l_fk_ctrl.Rotation.Y {type=MovieSceneScriptingFloatChannel, default=0}
       calf_l_fk_ctrl.Rotation.Z {type=MovieSceneScriptingFloatChannel, default=0}
       calf_l_fk_ctrl.Scale.X {type=MovieSceneScriptingFloatChannel, default=0}
       calf_l_fk_ctrl.Scale.Y {type=MovieSceneScriptingFloatChannel, default=0}
       calf_l_fk_ctrl.Scale.Z {type=MovieSceneScriptingFloatChannel, default=0}
       foot_l_fk_ctrl.Location.X {type=MovieSceneScriptingFloatChannel, default=0}
       foot_l_fk_ctrl.Location.Y {type=MovieSceneScriptingFloatChannel, default=0}
       foot_l_fk_ctrl.Location.Z {type=MovieSceneScriptingFloatChannel, default=0}
       foot_l_fk_ctrl.Rotation.X {type=MovieSceneScriptingFloatChannel, default=0}
       foot_l_fk_ctrl.Rotation.Y {type=MovieSceneScriptingFloatChannel, default=0}
       foot_l_fk_ctrl.Rotation.Z {type=MovieSceneScriptingFloatChannel, default=0}
       foot_l_fk_ctrl.Scale.X {type=MovieSceneScriptingFloatChannel, default=0}
       foot_l_fk_ctrl.Scale.Y {type=MovieSceneScriptingFloatChannel, default=0}
       foot_l_fk_ctrl.Scale.Z {type=MovieSceneScriptingFloatChannel, default=0}
       ball_l_fk_ctrl.Location.X {type=MovieSceneScriptingFloatChannel, default=0}
       ball_l_fk_ctrl.Location.Y {type=MovieSceneScriptingFloatChannel, default=0}
       ball_l_fk_ctrl.Location.Z {type=MovieSceneScriptingFloatChannel, default=0}
       ball_l_fk_ctrl.Rotation.X {type=MovieSceneScriptingFloatChannel, default=0}
       ball_l_fk_ctrl.Rotation.Y {type=MovieSceneScriptingFloatChannel, default=0}
       ball_l_fk_ctrl.Rotation.Z {type=MovieSceneScriptingFloatChannel, default=0}
       ball_l_fk_ctrl.Scale.X {type=MovieSceneScriptingFloatChannel, default=0}
       ball_l_fk_ctrl.Scale.Y {type=MovieSceneScriptingFloatChannel, default=0}
       ball_l_fk_ctrl.Scale.Z {type=MovieSceneScriptingFloatChannel, default=0}
       foot_l_ik_ctrl.Location.X {type=MovieSceneScriptingFloatChannel, default=0}
       foot_l_ik_ctrl.Location.Y {type=MovieSceneScriptingFloatChannel, default=0}
       foot_l_ik_ctrl.Location.Z {type=MovieSceneScriptingFloatChannel, default=0}
       foot_l_ik_ctrl.Rotation.X {type=MovieSceneScriptingFloatChannel, default=0}
       foot_l_ik_ctrl.Rotation.Y {type=MovieSceneScriptingFloatChannel, default=0}
       foot_l_ik_ctrl.Rotation.Z {type=MovieSceneScriptingFloatChannel, default=0}
       foot_l_ik_ctrl.Scale.X {type=MovieSceneScriptingFloatChannel, default=0}
       foot_l_ik_ctrl.Scale.Y {type=MovieSceneScriptingFloatChannel, default=0}
       foot_l_ik_ctrl.Scale.Z {type=MovieSceneScriptingFloatChannel, default=0}
       ball_l_ik_ctrl.Location.X {type=MovieSceneScriptingFloatChannel, default=0}
       ball_l_ik_ctrl.Location.Y {type=MovieSceneScriptingFloatChannel, default=0}
       ball_l_ik_ctrl.Location.Z {type=MovieSceneScriptingFloatChannel, default=0}
       ball_l_ik_ctrl.Rotation.X {type=MovieSceneScriptingFloatChannel, default=0}
       ball_l_ik_ctrl.Rotation.Y {type=MovieSceneScriptingFloatChannel, default=0}
       ball_l_ik_ctrl.Rotation.Z {type=MovieSceneScriptingFloatChannel, default=0}
       ball_l_ik_ctrl.Scale.X {type=MovieSceneScriptingFloatChannel, default=0}
       ball_l_ik_ctrl.Scale.Y {type=MovieSceneScriptingFloatChannel, default=0}
       ball_l_ik_ctrl.Scale.Z {type=MovieSceneScriptingFloatChannel, default=0}
       foot_bk1_l_ctrl.Location.X {type=MovieSceneScriptingFloatChannel, default=0}
       foot_bk1_l_ctrl.Location.Y {type=MovieSceneScriptingFloatChannel, default=0}
       foot_bk1_l_ctrl.Location.Z {type=MovieSceneScriptingFloatChannel, default=0}
       foot_bk1_l_ctrl.Rotation.X {type=MovieSceneScriptingFloatChannel, default=0}
       foot_bk1_l_ctrl.Rotation.Y {type=MovieSceneScriptingFloatChannel, default=0}
       foot_bk1_l_ctrl.Rotation.Z {type=MovieSceneScriptingFloatChannel, default=0}
       foot_bk1_l_ctrl.Scale.X {type=MovieSceneScriptingFloatChannel, default=0}
       foot_bk1_l_ctrl.Scale.Y {type=MovieSceneScriptingFloatChannel, default=0}
       foot_bk1_l_ctrl.Scale.Z {type=MovieSceneScriptingFloatChannel, default=0}
       foot_roll_l_ctrl.Location.X {type=MovieSceneScriptingFloatChannel, default=0}
       foot_roll_l_ctrl.Location.Y {type=MovieSceneScriptingFloatChannel, default=0}
       foot_roll_l_ctrl.Location.Z {type=MovieSceneScriptingFloatChannel, default=0}
       foot_roll_l_ctrl.Rotation.X {type=MovieSceneScriptingFloatChannel, default=0}
       foot_roll_l_ctrl.Rotation.Y {type=MovieSceneScriptingFloatChannel, default=0}
       foot_roll_l_ctrl.Rotation.Z {type=MovieSceneScriptingFloatChannel, default=0}
       foot_roll_l_ctrl.Scale.X {type=MovieSceneScriptingFloatChannel, default=0}
       foot_roll_l_ctrl.Scale.Y {type=MovieSceneScriptingFloatChannel, default=0}
       foot_roll_l_ctrl.Scale.Z {type=MovieSceneScriptingFloatChannel, default=0}
       roll_blend_l_ctrl {type=MovieSceneScriptingFloatChannel, default=0}
       tip_l_ctrl.Location.X {type=MovieSceneScriptingFloatChannel, default=0}
       tip_l_ctrl.Location.Y {type=MovieSceneScriptingFloatChannel, default=0}
       tip_l_ctrl.Location.Z {type=MovieSceneScriptingFloatChannel, default=0}
       tip_l_ctrl.Rotation.X {type=MovieSceneScriptingFloatChannel, default=0}
       tip_l_ctrl.Rotation.Y {type=MovieSceneScriptingFloatChannel, default=0}
       tip_l_ctrl.Rotation.Z {type=MovieSceneScriptingFloatChannel, default=0}
       tip_l_ctrl.Scale.X {type=MovieSceneScriptingFloatChannel, default=0}
       tip_l_ctrl.Scale.Y {type=MovieSceneScriptingFloatChannel, default=0}
       tip_l_ctrl.Scale.Z {type=MovieSceneScriptingFloatChannel, default=0}
       heel_l_ctrl.Location.X {type=MovieSceneScriptingFloatChannel, default=0}
       heel_l_ctrl.Location.Y {type=MovieSceneScriptingFloatChannel, default=0}
       heel_l_ctrl.Location.Z {type=MovieSceneScriptingFloatChannel, default=0}
       heel_l_ctrl.Rotation.X {type=MovieSceneScriptingFloatChannel, default=0}
       heel_l_ctrl.Rotation.Y {type=MovieSceneScriptingFloatChannel, default=0}
       heel_l_ctrl.Rotation.Z {type=MovieSceneScriptingFloatChannel, default=0}
       heel_l_ctrl.Scale.X {type=MovieSceneScriptingFloatChannel, default=0}
       heel_l_ctrl.Scale.Y {type=MovieSceneScriptingFloatChannel, default=0}
       heel_l_ctrl.Scale.Z {type=MovieSceneScriptingFloatChannel, default=0}
       leg_l_softness {type=MovieSceneScriptingFloatChannel, default=0}
       leg_l_pv_ik_ctrl.Location.X {type=MovieSceneScriptingFloatChannel, default=0}
       leg_l_pv_ik_ctrl.Location.Y {type=MovieSceneScriptingFloatChannel, default=0}
       leg_l_pv_ik_ctrl.Location.Z {type=MovieSceneScriptingFloatChannel, default=0}
       leg_l_pv_ik_ctrl.Rotation.X {type=MovieSceneScriptingFloatChannel, default=0}
       leg_l_pv_ik_ctrl.Rotation.Y {type=MovieSceneScriptingFloatChannel, default=0}
       leg_l_pv_ik_ctrl.Rotation.Z {type=MovieSceneScriptingFloatChannel, default=0}
       leg_l_pv_ik_ctrl.Scale.X {type=MovieSceneScriptingFloatChannel, default=0}
       leg_l_pv_ik_ctrl.Scale.Y {type=MovieSceneScriptingFloatChannel, default=0}
       leg_l_pv_ik_ctrl.Scale.Z {type=MovieSceneScriptingFloatChannel, default=0}
       thigh_r_fk_ctrl.Location.X {type=MovieSceneScriptingFloatChannel, default=0}
       thigh_r_fk_ctrl.Location.Y {type=MovieSceneScriptingFloatChannel, default=0}
       thigh_r_fk_ctrl.Location.Z {type=MovieSceneScriptingFloatChannel, default=0}
       thigh_r_fk_ctrl.Rotation.X {type=MovieSceneScriptingFloatChannel, default=0}
       thigh_r_fk_ctrl.Rotation.Y {type=MovieSceneScriptingFloatChannel, default=0}
       thigh_r_fk_ctrl.Rotation.Z {type=MovieSceneScriptingFloatChannel, default=0}
       thigh_r_fk_ctrl.Scale.X {type=MovieSceneScriptingFloatChannel, default=0}
       thigh_r_fk_ctrl.Scale.Y {type=MovieSceneScriptingFloatChannel, default=0}
       thigh_r_fk_ctrl.Scale.Z {type=MovieSceneScriptingFloatChannel, default=0}
       calf_r_fk_ctrl.Location.X {type=MovieSceneScriptingFloatChannel, default=0}
       calf_r_fk_ctrl.Location.Y {type=MovieSceneScriptingFloatChannel, default=0}
       calf_r_fk_ctrl.Location.Z {type=MovieSceneScriptingFloatChannel, default=0}
       calf_r_fk_ctrl.Rotation.X {type=MovieSceneScriptingFloatChannel, default=0}
       calf_r_fk_ctrl.Rotation.Y {type=MovieSceneScriptingFloatChannel, default=0}
       calf_r_fk_ctrl.Rotation.Z {type=MovieSceneScriptingFloatChannel, default=0}
       calf_r_fk_ctrl.Scale.X {type=MovieSceneScriptingFloatChannel, default=0}
       calf_r_fk_ctrl.Scale.Y {type=MovieSceneScriptingFloatChannel, default=0}
       calf_r_fk_ctrl.Scale.Z {type=MovieSceneScriptingFloatChannel, default=0}
       foot_r_fk_ctrl.Location.X {type=MovieSceneScriptingFloatChannel, default=0}
       foot_r_fk_ctrl.Location.Y {type=MovieSceneScriptingFloatChannel, default=0}
       foot_r_fk_ctrl.Location.Z {type=MovieSceneScriptingFloatChannel, default=0}
       foot_r_fk_ctrl.Rotation.X {type=MovieSceneScriptingFloatChannel, default=0}
       foot_r_fk_ctrl.Rotation.Y {type=MovieSceneScriptingFloatChannel, default=0}
       foot_r_fk_ctrl.Rotation.Z {type=MovieSceneScriptingFloatChannel, default=0}
       foot_r_fk_ctrl.Scale.X {type=MovieSceneScriptingFloatChannel, default=0}
       foot_r_fk_ctrl.Scale.Y {type=MovieSceneScriptingFloatChannel, default=0}
       foot_r_fk_ctrl.Scale.Z {type=MovieSceneScriptingFloatChannel, default=0}
       ball_r_fk_ctrl.Location.X {type=MovieSceneScriptingFloatChannel, default=0}
       ball_r_fk_ctrl.Location.Y {type=MovieSceneScriptingFloatChannel, default=0}
       ball_r_fk_ctrl.Location.Z {type=MovieSceneScriptingFloatChannel, default=0}
       ball_r_fk_ctrl.Rotation.X {type=MovieSceneScriptingFloatChannel, default=0}
       ball_r_fk_ctrl.Rotation.Y {type=MovieSceneScriptingFloatChannel, default=0}
       ball_r_fk_ctrl.Rotation.Z {type=MovieSceneScriptingFloatChannel, default=0}
       ball_r_fk_ctrl.Scale.X {type=MovieSceneScriptingFloatChannel, default=0}
       ball_r_fk_ctrl.Scale.Y {type=MovieSceneScriptingFloatChannel, default=0}
       ball_r_fk_ctrl.Scale.Z {type=MovieSceneScriptingFloatChannel, default=0}
       foot_r_ik_ctrl.Location.X {type=MovieSceneScriptingFloatChannel, default=0}
       foot_r_ik_ctrl.Location.Y {type=MovieSceneScriptingFloatChannel, default=0}
       foot_r_ik_ctrl.Location.Z {type=MovieSceneScriptingFloatChannel, default=0}
       foot_r_ik_ctrl.Rotation.X {type=MovieSceneScriptingFloatChannel, default=0}
       foot_r_ik_ctrl.Rotation.Y {type=MovieSceneScriptingFloatChannel, default=0}
       foot_r_ik_ctrl.Rotation.Z {type=MovieSceneScriptingFloatChannel, default=0}
       foot_r_ik_ctrl.Scale.X {type=MovieSceneScriptingFloatChannel, default=0}
       foot_r_ik_ctrl.Scale.Y {type=MovieSceneScriptingFloatChannel, default=0}
       foot_r_ik_ctrl.Scale.Z {type=MovieSceneScriptingFloatChannel, default=0}
       ball_r_ik_ctrl.Location.X {type=MovieSceneScriptingFloatChannel, default=0}
       ball_r_ik_ctrl.Location.Y {type=MovieSceneScriptingFloatChannel, default=0}
       ball_r_ik_ctrl.Location.Z {type=MovieSceneScriptingFloatChannel, default=0}
       ball_r_ik_ctrl.Rotation.X {type=MovieSceneScriptingFloatChannel, default=0}
       ball_r_ik_ctrl.Rotation.Y {type=MovieSceneScriptingFloatChannel, default=0}
       ball_r_ik_ctrl.Rotation.Z {type=MovieSceneScriptingFloatChannel, default=0}
       ball_r_ik_ctrl.Scale.X {type=MovieSceneScriptingFloatChannel, default=0}
       ball_r_ik_ctrl.Scale.Y {type=MovieSceneScriptingFloatChannel, default=0}
       ball_r_ik_ctrl.Scale.Z {type=MovieSceneScriptingFloatChannel, default=0}
       foot_bk1_r_ctrl.Location.X {type=MovieSceneScriptingFloatChannel, default=0}
       foot_bk1_r_ctrl.Location.Y {type=MovieSceneScriptingFloatChannel, default=0}
       foot_bk1_r_ctrl.Location.Z {type=MovieSceneScriptingFloatChannel, default=0}
       foot_bk1_r_ctrl.Rotation.X {type=MovieSceneScriptingFloatChannel, default=0}
       foot_bk1_r_ctrl.Rotation.Y {type=MovieSceneScriptingFloatChannel, default=0}
       foot_bk1_r_ctrl.Rotation.Z {type=MovieSceneScriptingFloatChannel, default=0}
       foot_bk1_r_ctrl.Scale.X {type=MovieSceneScriptingFloatChannel, default=0}
       foot_bk1_r_ctrl.Scale.Y {type=MovieSceneScriptingFloatChannel, default=0}
       foot_bk1_r_ctrl.Scale.Z {type=MovieSceneScriptingFloatChannel, default=0}
       foot_roll_r_ctrl.Location.X {type=MovieSceneScriptingFloatChannel, default=0}
       foot_roll_r_ctrl.Location.Y {type=MovieSceneScriptingFloatChannel, default=0}
       foot_roll_r_ctrl.Location.Z {type=MovieSceneScriptingFloatChannel, default=0}
       foot_roll_r_ctrl.Rotation.X {type=MovieSceneScriptingFloatChannel, default=0}
       foot_roll_r_ctrl.Rotation.Y {type=MovieSceneScriptingFloatChannel, default=0}
       foot_roll_r_ctrl.Rotation.Z {type=MovieSceneScriptingFloatChannel, default=0}
       foot_roll_r_ctrl.Scale.X {type=MovieSceneScriptingFloatChannel, default=0}
       foot_roll_r_ctrl.Scale.Y {type=MovieSceneScriptingFloatChannel, default=0}
       foot_roll_r_ctrl.Scale.Z {type=MovieSceneScriptingFloatChannel, default=0}
       roll_blend_r_ctrl {type=MovieSceneScriptingFloatChannel, default=0}
       tip_r_ctrl.Location.X {type=MovieSceneScriptingFloatChannel, default=0}
       tip_r_ctrl.Location.Y {type=MovieSceneScriptingFloatChannel, default=0}
       tip_r_ctrl.Location.Z {type=MovieSceneScriptingFloatChannel, default=0}
       tip_r_ctrl.Rotation.X {type=MovieSceneScriptingFloatChannel, default=0}
       tip_r_ctrl.Rotation.Y {type=MovieSceneScriptingFloatChannel, default=0}
       tip_r_ctrl.Rotation.Z {type=MovieSceneScriptingFloatChannel, default=0}
       tip_r_ctrl.Scale.X {type=MovieSceneScriptingFloatChannel, default=0}
       tip_r_ctrl.Scale.Y {type=MovieSceneScriptingFloatChannel, default=0}
       tip_r_ctrl.Scale.Z {type=MovieSceneScriptingFloatChannel, default=0}
       heel_r_ctrl.Location.X {type=MovieSceneScriptingFloatChannel, default=0}
       heel_r_ctrl.Location.Y {type=MovieSceneScriptingFloatChannel, default=0}
       heel_r_ctrl.Location.Z {type=MovieSceneScriptingFloatChannel, default=0}
       heel_r_ctrl.Rotation.X {type=MovieSceneScriptingFloatChannel, default=0}
       heel_r_ctrl.Rotation.Y {type=MovieSceneScriptingFloatChannel, default=0}
       heel_r_ctrl.Rotation.Z {type=MovieSceneScriptingFloatChannel, default=0}
       heel_r_ctrl.Scale.X {type=MovieSceneScriptingFloatChannel, default=0}
       heel_r_ctrl.Scale.Y {type=MovieSceneScriptingFloatChannel, default=0}
       heel_r_ctrl.Scale.Z {type=MovieSceneScriptingFloatChannel, default=0}
       leg_r_softness {type=MovieSceneScriptingFloatChannel, default=0}
       leg_r_pv_ik_ctrl.Location.X {type=MovieSceneScriptingFloatChannel, default=0}
       leg_r_pv_ik_ctrl.Location.Y {type=MovieSceneScriptingFloatChannel, default=0}
       leg_r_pv_ik_ctrl.Location.Z {type=MovieSceneScriptingFloatChannel, default=0}
       leg_r_pv_ik_ctrl.Rotation.X {type=MovieSceneScriptingFloatChannel, default=0}
       leg_r_pv_ik_ctrl.Rotation.Y {type=MovieSceneScriptingFloatChannel, default=0}
       leg_r_pv_ik_ctrl.Rotation.Z {type=MovieSceneScriptingFloatChannel, default=0}
       leg_r_pv_ik_ctrl.Scale.X {type=MovieSceneScriptingFloatChannel, default=0}
       leg_r_pv_ik_ctrl.Scale.Y {type=MovieSceneScriptingFloatChannel, default=0}
       leg_r_pv_ik_ctrl.Scale.Z {type=MovieSceneScriptingFloatChannel, default=0}
       upperarm_l_fk_ctrl.Location.X {type=MovieSceneScriptingFloatChannel, default=0, const [0,2]}
       upperarm_l_fk_ctrl.Location.Y {type=MovieSceneScriptingFloatChannel, default=0, const [0,2]}
       upperarm_l_fk_ctrl.Location.Z {type=MovieSceneScriptingFloatChannel, default=0, const [0,2]}
       upperarm_l_fk_ctrl.Rotation.X {type=MovieSceneScriptingFloatChannel, default=0, keys [[0,-12.3745],[21,0.9246]]}
       upperarm_l_fk_ctrl.Rotation.Y {type=MovieSceneScriptingFloatChannel, default=0, keys [[0,23.9499],[21,34.8229]]}
       upperarm_l_fk_ctrl.Rotation.Z {type=MovieSceneScriptingFloatChannel, default=0, keys [[0,-45.8332],[8,7.3467],[21,32.8927]]}
       upperarm_l_fk_ctrl.Scale.X {type=MovieSceneScriptingFloatChannel, default=0, const [0,2]}
       upperarm_l_fk_ctrl.Scale.Y {type=MovieSceneScriptingFloatChannel, default=0, const [0,2]}
       upperarm_l_fk_ctrl.Scale.Z {type=MovieSceneScriptingFloatChannel, default=0, const [0,2]}
       lowerarm_l_fk_ctrl.Location.X {type=MovieSceneScriptingFloatChannel, default=0, const [0,2]}
       lowerarm_l_fk_ctrl.Location.Y {type=MovieSceneScriptingFloatChannel, default=0, const [0,2]}
       lowerarm_l_fk_ctrl.Location.Z {type=MovieSceneScriptingFloatChannel, default=0, const [0,2]}
       lowerarm_l_fk_ctrl.Rotation.X {type=MovieSceneScriptingFloatChannel, default=0, const [0,2]}
       lowerarm_l_fk_ctrl.Rotation.Y {type=MovieSceneScriptingFloatChannel, default=0, const [0,2]}
       lowerarm_l_fk_ctrl.Rotation.Z {type=MovieSceneScriptingFloatChannel, default=0, const [0,2]}
       lowerarm_l_fk_ctrl.Scale.X {type=MovieSceneScriptingFloatChannel, default=0, const [0,2]}
       lowerarm_l_fk_ctrl.Scale.Y {type=MovieSceneScriptingFloatChannel, default=0, const [0,2]}
       lowerarm_l_fk_ctrl.Scale.Z {type=MovieSceneScriptingFloatChannel, default=0, const [0,2]}
       hand_l_fk_ctrl.Location.X {type=MovieSceneScriptingFloatChannel, default=0, const [0,2]}
       hand_l_fk_ctrl.Location.Y {type=MovieSceneScriptingFloatChannel, default=0, const [0,2]}
       hand_l_fk_ctrl.Location.Z {type=MovieSceneScriptingFloatChannel, default=0, const [0,2]}
       hand_l_fk_ctrl.Rotation.X {type=MovieSceneScriptingFloatChannel, default=0, const [0,2]}
       hand_l_fk_ctrl.Rotation.Y {type=MovieSceneScriptingFloatChannel, default=0, const [0,2]}
       hand_l_fk_ctrl.Rotation.Z {type=MovieSceneScriptingFloatChannel, default=0, const [0,2]}
       hand_l_fk_ctrl.Scale.X {type=MovieSceneScriptingFloatChannel, default=0, const [0,2]}
       hand_l_fk_ctrl.Scale.Y {type=MovieSceneScriptingFloatChannel, default=0, const [0,2]}
       hand_l_fk_ctrl.Scale.Z {type=MovieSceneScriptingFloatChannel, default=0, const [0,2]}
       thumb_01_l_ctrl.Location.X {type=MovieSceneScriptingFloatChannel, default=0}
       thumb_01_l_ctrl.Location.Y {type=MovieSceneScriptingFloatChannel, default=0}
       thumb_01_l_ctrl.Location.Z {type=MovieSceneScriptingFloatChannel, default=0}
       thumb_01_l_ctrl.Rotation.X {type=MovieSceneScriptingFloatChannel, default=0}
       thumb_01_l_ctrl.Rotation.Y {type=MovieSceneScriptingFloatChannel, default=0}
       thumb_01_l_ctrl.Rotation.Z {type=MovieSceneScriptingFloatChannel, default=0}
       thumb_01_l_ctrl.Scale.X {type=MovieSceneScriptingFloatChannel, default=0}
       thumb_01_l_ctrl.Scale.Y {type=MovieSceneScriptingFloatChannel, default=0}
       thumb_01_l_ctrl.Scale.Z {type=MovieSceneScriptingFloatChannel, default=0}
       thumb_02_l_ctrl.Location.X {type=MovieSceneScriptingFloatChannel, default=0}
       thumb_02_l_ctrl.Location.Y {type=MovieSceneScriptingFloatChannel, default=0}
       thumb_02_l_ctrl.Location.Z {type=MovieSceneScriptingFloatChannel, default=0}
       thumb_02_l_ctrl.Rotation.X {type=MovieSceneScriptingFloatChannel, default=0}
       thumb_02_l_ctrl.Rotation.Y {type=MovieSceneScriptingFloatChannel, default=0}
       thumb_02_l_ctrl.Rotation.Z {type=MovieSceneScriptingFloatChannel, default=0}
       thumb_02_l_ctrl.Scale.X {type=MovieSceneScriptingFloatChannel, default=0}
       thumb_02_l_ctrl.Scale.Y {type=MovieSceneScriptingFloatChannel, default=0}
       thumb_02_l_ctrl.Scale.Z {type=MovieSceneScriptingFloatChannel, default=0}
       thumb_03_l_ctrl.Location.X {type=MovieSceneScriptingFloatChannel, default=0}
       thumb_03_l_ctrl.Location.Y {type=MovieSceneScriptingFloatChannel, default=0}
       thumb_03_l_ctrl.Location.Z {type=MovieSceneScriptingFloatChannel, default=0}
       thumb_03_l_ctrl.Rotation.X {type=MovieSceneScriptingFloatChannel, default=0}
       thumb_03_l_ctrl.Rotation.Y {type=MovieSceneScriptingFloatChannel, default=0}
       thumb_03_l_ctrl.Rotation.Z {type=MovieSceneScriptingFloatChannel, default=0}
       thumb_03_l_ctrl.Scale.X {type=MovieSceneScriptingFloatChannel, default=0}
       thumb_03_l_ctrl.Scale.Y {type=MovieSceneScriptingFloatChannel, default=0}
       thumb_03_l_ctrl.Scale.Z {type=MovieSceneScriptingFloatChannel, default=0}
       index_metacarpal_l_ctrl.Location.X {type=MovieSceneScriptingFloatChannel, default=0}
       index_metacarpal_l_ctrl.Location.Y {type=MovieSceneScriptingFloatChannel, default=0}
       index_metacarpal_l_ctrl.Location.Z {type=MovieSceneScriptingFloatChannel, default=0}
       index_metacarpal_l_ctrl.Rotation.X {type=MovieSceneScriptingFloatChannel, default=0}
       index_metacarpal_l_ctrl.Rotation.Y {type=MovieSceneScriptingFloatChannel, default=0}
       index_metacarpal_l_ctrl.Rotation.Z {type=MovieSceneScriptingFloatChannel, default=0}
       index_metacarpal_l_ctrl.Scale.X {type=MovieSceneScriptingFloatChannel, default=0}
       index_metacarpal_l_ctrl.Scale.Y {type=MovieSceneScriptingFloatChannel, default=0}
       index_metacarpal_l_ctrl.Scale.Z {type=MovieSceneScriptingFloatChannel, default=0}
       index_01_l_ctrl.Location.X {type=MovieSceneScriptingFloatChannel, default=0}
       index_01_l_ctrl.Location.Y {type=MovieSceneScriptingFloatChannel, default=0}
       index_01_l_ctrl.Location.Z {type=MovieSceneScriptingFloatChannel, default=0}
       index_01_l_ctrl.Rotation.X {type=MovieSceneScriptingFloatChannel, default=0}
       index_01_l_ctrl.Rotation.Y {type=MovieSceneScriptingFloatChannel, default=0}
       index_01_l_ctrl.Rotation.Z {type=MovieSceneScriptingFloatChannel, default=0}
       index_01_l_ctrl.Scale.X {type=MovieSceneScriptingFloatChannel, default=0}
       index_01_l_ctrl.Scale.Y {type=MovieSceneScriptingFloatChannel, default=0}
       index_01_l_ctrl.Scale.Z {type=MovieSceneScriptingFloatChannel, default=0}
       index_02_l_ctrl.Location.X {type=MovieSceneScriptingFloatChannel, default=0}
       index_02_l_ctrl.Location.Y {type=MovieSceneScriptingFloatChannel, default=0}
       index_02_l_ctrl.Location.Z {type=MovieSceneScriptingFloatChannel, default=0}
       index_02_l_ctrl.Rotation.X {type=MovieSceneScriptingFloatChannel, default=0}
       index_02_l_ctrl.Rotation.Y {type=MovieSceneScriptingFloatChannel, default=0}
       index_02_l_ctrl.Rotation.Z {type=MovieSceneScriptingFloatChannel, default=0}
       index_02_l_ctrl.Scale.X {type=MovieSceneScriptingFloatChannel, default=0}
       index_02_l_ctrl.Scale.Y {type=MovieSceneScriptingFloatChannel, default=0}
       index_02_l_ctrl.Scale.Z {type=MovieSceneScriptingFloatChannel, default=0}
       index_03_l_ctrl.Location.X {type=MovieSceneScriptingFloatChannel, default=0}
       index_03_l_ctrl.Location.Y {type=MovieSceneScriptingFloatChannel, default=0}
       index_03_l_ctrl.Location.Z {type=MovieSceneScriptingFloatChannel, default=0}
       index_03_l_ctrl.Rotation.X {type=MovieSceneScriptingFloatChannel, default=0}
       index_03_l_ctrl.Rotation.Y {type=MovieSceneScriptingFloatChannel, default=0}
       index_03_l_ctrl.Rotation.Z {type=MovieSceneScriptingFloatChannel, default=0}
       index_03_l_ctrl.Scale.X {type=MovieSceneScriptingFloatChannel, default=0}
       index_03_l_ctrl.Scale.Y {type=MovieSceneScriptingFloatChannel, default=0}
       index_03_l_ctrl.Scale.Z {type=MovieSceneScriptingFloatChannel, default=0}
       middle_metacarpal_l_ctrl.Location.X {type=MovieSceneScriptingFloatChannel, default=0}
       middle_metacarpal_l_ctrl.Location.Y {type=MovieSceneScriptingFloatChannel, default=0}
       middle_metacarpal_l_ctrl.Location.Z {type=MovieSceneScriptingFloatChannel, default=0}
       middle_metacarpal_l_ctrl.Rotation.X {type=MovieSceneScriptingFloatChannel, default=0}
       middle_metacarpal_l_ctrl.Rotation.Y {type=MovieSceneScriptingFloatChannel, default=0}
       middle_metacarpal_l_ctrl.Rotation.Z {type=MovieSceneScriptingFloatChannel, default=0}
       middle_metacarpal_l_ctrl.Scale.X {type=MovieSceneScriptingFloatChannel, default=0}
       middle_metacarpal_l_ctrl.Scale.Y {type=MovieSceneScriptingFloatChannel, default=0}
       middle_metacarpal_l_ctrl.Scale.Z {type=MovieSceneScriptingFloatChannel, default=0}
       middle_01_l_ctrl.Location.X {type=MovieSceneScriptingFloatChannel, default=0}
       middle_01_l_ctrl.Location.Y {type=MovieSceneScriptingFloatChannel, default=0}
       middle_01_l_ctrl.Location.Z {type=MovieSceneScriptingFloatChannel, default=0}
       middle_01_l_ctrl.Rotation.X {type=MovieSceneScriptingFloatChannel, default=0}
       middle_01_l_ctrl.Rotation.Y {type=MovieSceneScriptingFloatChannel, default=0}
       middle_01_l_ctrl.Rotation.Z {type=MovieSceneScriptingFloatChannel, default=0}
       middle_01_l_ctrl.Scale.X {type=MovieSceneScriptingFloatChannel, default=0}
       middle_01_l_ctrl.Scale.Y {type=MovieSceneScriptingFloatChannel, default=0}
       middle_01_l_ctrl.Scale.Z {type=MovieSceneScriptingFloatChannel, default=0}
       middle_02_l_ctrl.Location.X {type=MovieSceneScriptingFloatChannel, default=0}
       middle_02_l_ctrl.Location.Y {type=MovieSceneScriptingFloatChannel, default=0}
       middle_02_l_ctrl.Location.Z {type=MovieSceneScriptingFloatChannel, default=0}
       middle_02_l_ctrl.Rotation.X {type=MovieSceneScriptingFloatChannel, default=0}
       middle_02_l_ctrl.Rotation.Y {type=MovieSceneScriptingFloatChannel, default=0}
       middle_02_l_ctrl.Rotation.Z {type=MovieSceneScriptingFloatChannel, default=0}
       middle_02_l_ctrl.Scale.X {type=MovieSceneScriptingFloatChannel, default=0}
       middle_02_l_ctrl.Scale.Y {type=MovieSceneScriptingFloatChannel, default=0}
       middle_02_l_ctrl.Scale.Z {type=MovieSceneScriptingFloatChannel, default=0}
       middle_03_l_ctrl.Location.X {type=MovieSceneScriptingFloatChannel, default=0}
       middle_03_l_ctrl.Location.Y {type=MovieSceneScriptingFloatChannel, default=0}
       middle_03_l_ctrl.Location.Z {type=MovieSceneScriptingFloatChannel, default=0}
       middle_03_l_ctrl.Rotation.X {type=MovieSceneScriptingFloatChannel, default=0}
       middle_03_l_ctrl.Rotation.Y {type=MovieSceneScriptingFloatChannel, default=0}
       middle_03_l_ctrl.Rotation.Z {type=MovieSceneScriptingFloatChannel, default=0}
       middle_03_l_ctrl.Scale.X {type=MovieSceneScriptingFloatChannel, default=0}
       middle_03_l_ctrl.Scale.Y {type=MovieSceneScriptingFloatChannel, default=0}
       middle_03_l_ctrl.Scale.Z {type=MovieSceneScriptingFloatChannel, default=0}
       ring_metacarpal_l_ctrl.Location.X {type=MovieSceneScriptingFloatChannel, default=0}
       ring_metacarpal_l_ctrl.Location.Y {type=MovieSceneScriptingFloatChannel, default=0}
       ring_metacarpal_l_ctrl.Location.Z {type=MovieSceneScriptingFloatChannel, default=0}
       ring_metacarpal_l_ctrl.Rotation.X {type=MovieSceneScriptingFloatChannel, default=0}
       ring_metacarpal_l_ctrl.Rotation.Y {type=MovieSceneScriptingFloatChannel, default=0}
       ring_metacarpal_l_ctrl.Rotation.Z {type=MovieSceneScriptingFloatChannel, default=0}
       ring_metacarpal_l_ctrl.Scale.X {type=MovieSceneScriptingFloatChannel, default=0}
       ring_metacarpal_l_ctrl.Scale.Y {type=MovieSceneScriptingFloatChannel, default=0}
       ring_metacarpal_l_ctrl.Scale.Z {type=MovieSceneScriptingFloatChannel, default=0}
       ring_01_l_ctrl.Location.X {type=MovieSceneScriptingFloatChannel, default=0}
       ring_01_l_ctrl.Location.Y {type=MovieSceneScriptingFloatChannel, default=0}
       ring_01_l_ctrl.Location.Z {type=MovieSceneScriptingFloatChannel, default=0}
       ring_01_l_ctrl.Rotation.X {type=MovieSceneScriptingFloatChannel, default=0}
       ring_01_l_ctrl.Rotation.Y {type=MovieSceneScriptingFloatChannel, default=0}
       ring_01_l_ctrl.Rotation.Z {type=MovieSceneScriptingFloatChannel, default=0}
       ring_01_l_ctrl.Scale.X {type=MovieSceneScriptingFloatChannel, default=0}
       ring_01_l_ctrl.Scale.Y {type=MovieSceneScriptingFloatChannel, default=0}
       ring_01_l_ctrl.Scale.Z {type=MovieSceneScriptingFloatChannel, default=0}
       ring_02_l_ctrl.Location.X {type=MovieSceneScriptingFloatChannel, default=0}
       ring_02_l_ctrl.Location.Y {type=MovieSceneScriptingFloatChannel, default=0}
       ring_02_l_ctrl.Location.Z {type=MovieSceneScriptingFloatChannel, default=0}
       ring_02_l_ctrl.Rotation.X {type=MovieSceneScriptingFloatChannel, default=0}
       ring_02_l_ctrl.Rotation.Y {type=MovieSceneScriptingFloatChannel, default=0}
       ring_02_l_ctrl.Rotation.Z {type=MovieSceneScriptingFloatChannel, default=0}
       ring_02_l_ctrl.Scale.X {type=MovieSceneScriptingFloatChannel, default=0}
       ring_02_l_ctrl.Scale.Y {type=MovieSceneScriptingFloatChannel, default=0}
       ring_02_l_ctrl.Scale.Z {type=MovieSceneScriptingFloatChannel, default=0}
       ring_03_l_ctrl.Location.X {type=MovieSceneScriptingFloatChannel, default=0}
       ring_03_l_ctrl.Location.Y {type=MovieSceneScriptingFloatChannel, default=0}
       ring_03_l_ctrl.Location.Z {type=MovieSceneScriptingFloatChannel, default=0}
       ring_03_l_ctrl.Rotation.X {type=MovieSceneScriptingFloatChannel, default=0}
       ring_03_l_ctrl.Rotation.Y {type=MovieSceneScriptingFloatChannel, default=0}
       ring_03_l_ctrl.Rotation.Z {type=MovieSceneScriptingFloatChannel, default=0}
       ring_03_l_ctrl.Scale.X {type=MovieSceneScriptingFloatChannel, default=0}
       ring_03_l_ctrl.Scale.Y {type=MovieSceneScriptingFloatChannel, default=0}
       ring_03_l_ctrl.Scale.Z {type=MovieSceneScriptingFloatChannel, default=0}
       pinky_metacarpal_l_ctrl.Location.X {type=MovieSceneScriptingFloatChannel, default=0}
       pinky_metacarpal_l_ctrl.Location.Y {type=MovieSceneScriptingFloatChannel, default=0}
       pinky_metacarpal_l_ctrl.Location.Z {type=MovieSceneScriptingFloatChannel, default=0}
       pinky_metacarpal_l_ctrl.Rotation.X {type=MovieSceneScriptingFloatChannel, default=0}
       pinky_metacarpal_l_ctrl.Rotation.Y {type=MovieSceneScriptingFloatChannel, default=0}
       pinky_metacarpal_l_ctrl.Rotation.Z {type=MovieSceneScriptingFloatChannel, default=0}
       pinky_metacarpal_l_ctrl.Scale.X {type=MovieSceneScriptingFloatChannel, default=0}
       pinky_metacarpal_l_ctrl.Scale.Y {type=MovieSceneScriptingFloatChannel, default=0}
       pinky_metacarpal_l_ctrl.Scale.Z {type=MovieSceneScriptingFloatChannel, default=0}
       pinky_01_l_ctrl.Location.X {type=MovieSceneScriptingFloatChannel, default=0}
       pinky_01_l_ctrl.Location.Y {type=MovieSceneScriptingFloatChannel, default=0}
       pinky_01_l_ctrl.Location.Z {type=MovieSceneScriptingFloatChannel, default=0}
       pinky_01_l_ctrl.Rotation.X {type=MovieSceneScriptingFloatChannel, default=0}
       pinky_01_l_ctrl.Rotation.Y {type=MovieSceneScriptingFloatChannel, default=0}
       pinky_01_l_ctrl.Rotation.Z {type=MovieSceneScriptingFloatChannel, default=0}
       pinky_01_l_ctrl.Scale.X {type=MovieSceneScriptingFloatChannel, default=0}
       pinky_01_l_ctrl.Scale.Y {type=MovieSceneScriptingFloatChannel, default=0}
       pinky_01_l_ctrl.Scale.Z {type=MovieSceneScriptingFloatChannel, default=0}
       pinky_02_l_ctrl.Location.X {type=MovieSceneScriptingFloatChannel, default=0}
       pinky_02_l_ctrl.Location.Y {type=MovieSceneScriptingFloatChannel, default=0}
       pinky_02_l_ctrl.Location.Z {type=MovieSceneScriptingFloatChannel, default=0}
       pinky_02_l_ctrl.Rotation.X {type=MovieSceneScriptingFloatChannel, default=0}
       pinky_02_l_ctrl.Rotation.Y {type=MovieSceneScriptingFloatChannel, default=0}
       pinky_02_l_ctrl.Rotation.Z {type=MovieSceneScriptingFloatChannel, default=0}
       pinky_02_l_ctrl.Scale.X {type=MovieSceneScriptingFloatChannel, default=0}
       pinky_02_l_ctrl.Scale.Y {type=MovieSceneScriptingFloatChannel, default=0}
       pinky_02_l_ctrl.Scale.Z {type=MovieSceneScriptingFloatChannel, default=0}
       pinky_03_l_ctrl.Location.X {type=MovieSceneScriptingFloatChannel, default=0}
       pinky_03_l_ctrl.Location.Y {type=MovieSceneScriptingFloatChannel, default=0}
       pinky_03_l_ctrl.Location.Z {type=MovieSceneScriptingFloatChannel, default=0}
       pinky_03_l_ctrl.Rotation.X {type=MovieSceneScriptingFloatChannel, default=0}
       pinky_03_l_ctrl.Rotation.Y {type=MovieSceneScriptingFloatChannel, default=0}
       pinky_03_l_ctrl.Rotation.Z {type=MovieSceneScriptingFloatChannel, default=0}
       pinky_03_l_ctrl.Scale.X {type=MovieSceneScriptingFloatChannel, default=0}
       pinky_03_l_ctrl.Scale.Y {type=MovieSceneScriptingFloatChannel, default=0}
       pinky_03_l_ctrl.Scale.Z {type=MovieSceneScriptingFloatChannel, default=0}
       hand_l_ik_ctrl.Location.X {type=MovieSceneScriptingFloatChannel, default=0}
       hand_l_ik_ctrl.Location.Y {type=MovieSceneScriptingFloatChannel, default=0}
       hand_l_ik_ctrl.Location.Z {type=MovieSceneScriptingFloatChannel, default=0}
       hand_l_ik_ctrl.Rotation.X {type=MovieSceneScriptingFloatChannel, default=0}
       hand_l_ik_ctrl.Rotation.Y {type=MovieSceneScriptingFloatChannel, default=0}
       hand_l_ik_ctrl.Rotation.Z {type=MovieSceneScriptingFloatChannel, default=0}
       hand_l_ik_ctrl.Scale.X {type=MovieSceneScriptingFloatChannel, default=0}
       hand_l_ik_ctrl.Scale.Y {type=MovieSceneScriptingFloatChannel, default=0}
       hand_l_ik_ctrl.Scale.Z {type=MovieSceneScriptingFloatChannel, default=0}
       Softness {type=MovieSceneScriptingFloatChannel, default=0}
       hand_l_ikorient_ctrl.Rotation.Roll {type=MovieSceneScriptingFloatChannel, default=0}
       hand_l_ikorient_ctrl.Rotation.Pitch {type=MovieSceneScriptingFloatChannel, default=0}
       hand_l_ikorient_ctrl.Rotation.Yaw {type=MovieSceneScriptingFloatChannel, default=0}
       arm_l_pv_ik_ctrl.Location.X {type=MovieSceneScriptingFloatChannel, default=0}
       arm_l_pv_ik_ctrl.Location.Y {type=MovieSceneScriptingFloatChannel, default=0}
       arm_l_pv_ik_ctrl.Location.Z {type=MovieSceneScriptingFloatChannel, default=0}
       arm_l_pv_ik_ctrl.Rotation.X {type=MovieSceneScriptingFloatChannel, default=0}
       arm_l_pv_ik_ctrl.Rotation.Y {type=MovieSceneScriptingFloatChannel, default=0}
       arm_l_pv_ik_ctrl.Rotation.Z {type=MovieSceneScriptingFloatChannel, default=0}
       arm_l_pv_ik_ctrl.Scale.X {type=MovieSceneScriptingFloatChannel, default=0}
       arm_l_pv_ik_ctrl.Scale.Y {type=MovieSceneScriptingFloatChannel, default=0}
       arm_l_pv_ik_ctrl.Scale.Z {type=MovieSceneScriptingFloatChannel, default=0}
       upperarm_r_fk_ctrl.Location.X {type=MovieSceneScriptingFloatChannel, default=0, const [0,2]}
       upperarm_r_fk_ctrl.Location.Y {type=MovieSceneScriptingFloatChannel, default=0, const [0,2]}
       upperarm_r_fk_ctrl.Location.Z {type=MovieSceneScriptingFloatChannel, default=0, const [0,2]}
       upperarm_r_fk_ctrl.Rotation.X {type=MovieSceneScriptingFloatChannel, default=0, keys [[0,5.7993],[21,0]]}
       upperarm_r_fk_ctrl.Rotation.Y {type=MovieSceneScriptingFloatChannel, default=0, keys [[0,-4.8918],[21,0]]}
       upperarm_r_fk_ctrl.Rotation.Z {type=MovieSceneScriptingFloatChannel, default=0, keys [[0,90.7725],[8,31.5847],[21,0]]}
       upperarm_r_fk_ctrl.Scale.X {type=MovieSceneScriptingFloatChannel, default=0, const [0,2]}
       upperarm_r_fk_ctrl.Scale.Y {type=MovieSceneScriptingFloatChannel, default=0, const [0,2]}
       upperarm_r_fk_ctrl.Scale.Z {type=MovieSceneScriptingFloatChannel, default=0, const [0,2]}
       lowerarm_r_fk_ctrl.Location.X {type=MovieSceneScriptingFloatChannel, default=0, const [0,2]}
       lowerarm_r_fk_ctrl.Location.Y {type=MovieSceneScriptingFloatChannel, default=0, const [0,2]}
       lowerarm_r_fk_ctrl.Location.Z {type=MovieSceneScriptingFloatChannel, default=0, const [0,2]}
       lowerarm_r_fk_ctrl.Rotation.X {type=MovieSceneScriptingFloatChannel, default=0, const [0,2]}
       lowerarm_r_fk_ctrl.Rotation.Y {type=MovieSceneScriptingFloatChannel, default=0, const [0,2]}
       lowerarm_r_fk_ctrl.Rotation.Z {type=MovieSceneScriptingFloatChannel, default=0, const [0,2]}
       lowerarm_r_fk_ctrl.Scale.X {type=MovieSceneScriptingFloatChannel, default=0, const [0,2]}
       lowerarm_r_fk_ctrl.Scale.Y {type=MovieSceneScriptingFloatChannel, default=0, const [0,2]}
       lowerarm_r_fk_ctrl.Scale.Z {type=MovieSceneScriptingFloatChannel, default=0, const [0,2]}
       hand_r_fk_ctrl.Location.X {type=MovieSceneScriptingFloatChannel, default=0, const [0,2]}
       hand_r_fk_ctrl.Location.Y {type=MovieSceneScriptingFloatChannel, default=0, const [0,2]}
       hand_r_fk_ctrl.Location.Z {type=MovieSceneScriptingFloatChannel, default=0, const [0,2]}
       hand_r_fk_ctrl.Rotation.X {type=MovieSceneScriptingFloatChannel, default=0, const [0,2]}
       hand_r_fk_ctrl.Rotation.Y {type=MovieSceneScriptingFloatChannel, default=0, const [0,2]}
       hand_r_fk_ctrl.Rotation.Z {type=MovieSceneScriptingFloatChannel, default=0, const [0,2]}
       hand_r_fk_ctrl.Scale.X {type=MovieSceneScriptingFloatChannel, default=0, const [0,2]}
       hand_r_fk_ctrl.Scale.Y {type=MovieSceneScriptingFloatChannel, default=0, const [0,2]}
       hand_r_fk_ctrl.Scale.Z {type=MovieSceneScriptingFloatChannel, default=0, const [0,2]}
       thumb_01_r_ctrl.Location.X {type=MovieSceneScriptingFloatChannel, default=0}
       thumb_01_r_ctrl.Location.Y {type=MovieSceneScriptingFloatChannel, default=0}
       thumb_01_r_ctrl.Location.Z {type=MovieSceneScriptingFloatChannel, default=0}
       thumb_01_r_ctrl.Rotation.X {type=MovieSceneScriptingFloatChannel, default=0}
       thumb_01_r_ctrl.Rotation.Y {type=MovieSceneScriptingFloatChannel, default=0}
       thumb_01_r_ctrl.Rotation.Z {type=MovieSceneScriptingFloatChannel, default=0}
       thumb_01_r_ctrl.Scale.X {type=MovieSceneScriptingFloatChannel, default=0}
       thumb_01_r_ctrl.Scale.Y {type=MovieSceneScriptingFloatChannel, default=0}
       thumb_01_r_ctrl.Scale.Z {type=MovieSceneScriptingFloatChannel, default=0}
       thumb_02_r_ctrl.Location.X {type=MovieSceneScriptingFloatChannel, default=0}
       thumb_02_r_ctrl.Location.Y {type=MovieSceneScriptingFloatChannel, default=0}
       thumb_02_r_ctrl.Location.Z {type=MovieSceneScriptingFloatChannel, default=0}
       thumb_02_r_ctrl.Rotation.X {type=MovieSceneScriptingFloatChannel, default=0}
       thumb_02_r_ctrl.Rotation.Y {type=MovieSceneScriptingFloatChannel, default=0}
       thumb_02_r_ctrl.Rotation.Z {type=MovieSceneScriptingFloatChannel, default=0}
       thumb_02_r_ctrl.Scale.X {type=MovieSceneScriptingFloatChannel, default=0}
       thumb_02_r_ctrl.Scale.Y {type=MovieSceneScriptingFloatChannel, default=0}
       thumb_02_r_ctrl.Scale.Z {type=MovieSceneScriptingFloatChannel, default=0}
       thumb_03_r_ctrl.Location.X {type=MovieSceneScriptingFloatChannel, default=0}
       thumb_03_r_ctrl.Location.Y {type=MovieSceneScriptingFloatChannel, default=0}
       thumb_03_r_ctrl.Location.Z {type=MovieSceneScriptingFloatChannel, default=0}
       thumb_03_r_ctrl.Rotation.X {type=MovieSceneScriptingFloatChannel, default=0}
       thumb_03_r_ctrl.Rotation.Y {type=MovieSceneScriptingFloatChannel, default=0}
       thumb_03_r_ctrl.Rotation.Z {type=MovieSceneScriptingFloatChannel, default=0}
       thumb_03_r_ctrl.Scale.X {type=MovieSceneScriptingFloatChannel, default=0}
       thumb_03_r_ctrl.Scale.Y {type=MovieSceneScriptingFloatChannel, default=0}
       thumb_03_r_ctrl.Scale.Z {type=MovieSceneScriptingFloatChannel, default=0}
       index_metacarpal_r_ctrl.Location.X {type=MovieSceneScriptingFloatChannel, default=0}
       index_metacarpal_r_ctrl.Location.Y {type=MovieSceneScriptingFloatChannel, default=0}
       index_metacarpal_r_ctrl.Location.Z {type=MovieSceneScriptingFloatChannel, default=0}
       index_metacarpal_r_ctrl.Rotation.X {type=MovieSceneScriptingFloatChannel, default=0}
       index_metacarpal_r_ctrl.Rotation.Y {type=MovieSceneScriptingFloatChannel, default=0}
       index_metacarpal_r_ctrl.Rotation.Z {type=MovieSceneScriptingFloatChannel, default=0}
       index_metacarpal_r_ctrl.Scale.X {type=MovieSceneScriptingFloatChannel, default=0}
       index_metacarpal_r_ctrl.Scale.Y {type=MovieSceneScriptingFloatChannel, default=0}
       index_metacarpal_r_ctrl.Scale.Z {type=MovieSceneScriptingFloatChannel, default=0}
       index_01_r_ctrl.Location.X {type=MovieSceneScriptingFloatChannel, default=0}
       index_01_r_ctrl.Location.Y {type=MovieSceneScriptingFloatChannel, default=0}
       index_01_r_ctrl.Location.Z {type=MovieSceneScriptingFloatChannel, default=0}
       index_01_r_ctrl.Rotation.X {type=MovieSceneScriptingFloatChannel, default=0}
       index_01_r_ctrl.Rotation.Y {type=MovieSceneScriptingFloatChannel, default=0}
       index_01_r_ctrl.Rotation.Z {type=MovieSceneScriptingFloatChannel, default=0}
       index_01_r_ctrl.Scale.X {type=MovieSceneScriptingFloatChannel, default=0}
       index_01_r_ctrl.Scale.Y {type=MovieSceneScriptingFloatChannel, default=0}
       index_01_r_ctrl.Scale.Z {type=MovieSceneScriptingFloatChannel, default=0}
       index_02_r_ctrl.Location.X {type=MovieSceneScriptingFloatChannel, default=0}
       index_02_r_ctrl.Location.Y {type=MovieSceneScriptingFloatChannel, default=0}
       index_02_r_ctrl.Location.Z {type=MovieSceneScriptingFloatChannel, default=0}
       index_02_r_ctrl.Rotation.X {type=MovieSceneScriptingFloatChannel, default=0}
       index_02_r_ctrl.Rotation.Y {type=MovieSceneScriptingFloatChannel, default=0}
       index_02_r_ctrl.Rotation.Z {type=MovieSceneScriptingFloatChannel, default=0}
       index_02_r_ctrl.Scale.X {type=MovieSceneScriptingFloatChannel, default=0}
       index_02_r_ctrl.Scale.Y {type=MovieSceneScriptingFloatChannel, default=0}
       index_02_r_ctrl.Scale.Z {type=MovieSceneScriptingFloatChannel, default=0}
       index_03_r_ctrl.Location.X {type=MovieSceneScriptingFloatChannel, default=0}
       index_03_r_ctrl.Location.Y {type=MovieSceneScriptingFloatChannel, default=0}
       index_03_r_ctrl.Location.Z {type=MovieSceneScriptingFloatChannel, default=0}
       index_03_r_ctrl.Rotation.X {type=MovieSceneScriptingFloatChannel, default=0}
       index_03_r_ctrl.Rotation.Y {type=MovieSceneScriptingFloatChannel, default=0}
       index_03_r_ctrl.Rotation.Z {type=MovieSceneScriptingFloatChannel, default=0}
       index_03_r_ctrl.Scale.X {type=MovieSceneScriptingFloatChannel, default=0}
       index_03_r_ctrl.Scale.Y {type=MovieSceneScriptingFloatChannel, default=0}
       index_03_r_ctrl.Scale.Z {type=MovieSceneScriptingFloatChannel, default=0}
       middle_metacarpal_r_ctrl.Location.X {type=MovieSceneScriptingFloatChannel, default=0}
       middle_metacarpal_r_ctrl.Location.Y {type=MovieSceneScriptingFloatChannel, default=0}
       middle_metacarpal_r_ctrl.Location.Z {type=MovieSceneScriptingFloatChannel, default=0}
       middle_metacarpal_r_ctrl.Rotation.X {type=MovieSceneScriptingFloatChannel, default=0}
       middle_metacarpal_r_ctrl.Rotation.Y {type=MovieSceneScriptingFloatChannel, default=0}
       middle_metacarpal_r_ctrl.Rotation.Z {type=MovieSceneScriptingFloatChannel, default=0}
       middle_metacarpal_r_ctrl.Scale.X {type=MovieSceneScriptingFloatChannel, default=0}
       middle_metacarpal_r_ctrl.Scale.Y {type=MovieSceneScriptingFloatChannel, default=0}
       middle_metacarpal_r_ctrl.Scale.Z {type=MovieSceneScriptingFloatChannel, default=0}
       middle_01_r_ctrl.Location.X {type=MovieSceneScriptingFloatChannel, default=0}
       middle_01_r_ctrl.Location.Y {type=MovieSceneScriptingFloatChannel, default=0}
       middle_01_r_ctrl.Location.Z {type=MovieSceneScriptingFloatChannel, default=0}
       middle_01_r_ctrl.Rotation.X {type=MovieSceneScriptingFloatChannel, default=0}
       middle_01_r_ctrl.Rotation.Y {type=MovieSceneScriptingFloatChannel, default=0}
       middle_01_r_ctrl.Rotation.Z {type=MovieSceneScriptingFloatChannel, default=0}
       middle_01_r_ctrl.Scale.X {type=MovieSceneScriptingFloatChannel, default=0}
       middle_01_r_ctrl.Scale.Y {type=MovieSceneScriptingFloatChannel, default=0}
       middle_01_r_ctrl.Scale.Z {type=MovieSceneScriptingFloatChannel, default=0}
       middle_02_r_ctrl.Location.X {type=MovieSceneScriptingFloatChannel, default=0}
       middle_02_r_ctrl.Location.Y {type=MovieSceneScriptingFloatChannel, default=0}
       middle_02_r_ctrl.Location.Z {type=MovieSceneScriptingFloatChannel, default=0}
       middle_02_r_ctrl.Rotation.X {type=MovieSceneScriptingFloatChannel, default=0}
       middle_02_r_ctrl.Rotation.Y {type=MovieSceneScriptingFloatChannel, default=0}
       middle_02_r_ctrl.Rotation.Z {type=MovieSceneScriptingFloatChannel, default=0}
       middle_02_r_ctrl.Scale.X {type=MovieSceneScriptingFloatChannel, default=0}
       middle_02_r_ctrl.Scale.Y {type=MovieSceneScriptingFloatChannel, default=0}
       middle_02_r_ctrl.Scale.Z {type=MovieSceneScriptingFloatChannel, default=0}
       middle_03_r_ctrl.Location.X {type=MovieSceneScriptingFloatChannel, default=0}
       middle_03_r_ctrl.Location.Y {type=MovieSceneScriptingFloatChannel, default=0}
       middle_03_r_ctrl.Location.Z {type=MovieSceneScriptingFloatChannel, default=0}
       middle_03_r_ctrl.Rotation.X {type=MovieSceneScriptingFloatChannel, default=0}
       middle_03_r_ctrl.Rotation.Y {type=MovieSceneScriptingFloatChannel, default=0}
       middle_03_r_ctrl.Rotation.Z {type=MovieSceneScriptingFloatChannel, default=0}
       middle_03_r_ctrl.Scale.X {type=MovieSceneScriptingFloatChannel, default=0}
       middle_03_r_ctrl.Scale.Y {type=MovieSceneScriptingFloatChannel, default=0}
       middle_03_r_ctrl.Scale.Z {type=MovieSceneScriptingFloatChannel, default=0}
       ring_metacarpal_r_ctrl.Location.X {type=MovieSceneScriptingFloatChannel, default=0}
       ring_metacarpal_r_ctrl.Location.Y {type=MovieSceneScriptingFloatChannel, default=0}
       ring_metacarpal_r_ctrl.Location.Z {type=MovieSceneScriptingFloatChannel, default=0}
       ring_metacarpal_r_ctrl.Rotation.X {type=MovieSceneScriptingFloatChannel, default=0}
       ring_metacarpal_r_ctrl.Rotation.Y {type=MovieSceneScriptingFloatChannel, default=0}
       ring_metacarpal_r_ctrl.Rotation.Z {type=MovieSceneScriptingFloatChannel, default=0}
       ring_metacarpal_r_ctrl.Scale.X {type=MovieSceneScriptingFloatChannel, default=0}
       ring_metacarpal_r_ctrl.Scale.Y {type=MovieSceneScriptingFloatChannel, default=0}
       ring_metacarpal_r_ctrl.Scale.Z {type=MovieSceneScriptingFloatChannel, default=0}
       ring_01_r_ctrl.Location.X {type=MovieSceneScriptingFloatChannel, default=0}
       ring_01_r_ctrl.Location.Y {type=MovieSceneScriptingFloatChannel, default=0}
       ring_01_r_ctrl.Location.Z {type=MovieSceneScriptingFloatChannel, default=0}
       ring_01_r_ctrl.Rotation.X {type=MovieSceneScriptingFloatChannel, default=0}
       ring_01_r_ctrl.Rotation.Y {type=MovieSceneScriptingFloatChannel, default=0}
       ring_01_r_ctrl.Rotation.Z {type=MovieSceneScriptingFloatChannel, default=0}
       ring_01_r_ctrl.Scale.X {type=MovieSceneScriptingFloatChannel, default=0}
       ring_01_r_ctrl.Scale.Y {type=MovieSceneScriptingFloatChannel, default=0}
       ring_01_r_ctrl.Scale.Z {type=MovieSceneScriptingFloatChannel, default=0}
       ring_02_r_ctrl.Location.X {type=MovieSceneScriptingFloatChannel, default=0}
       ring_02_r_ctrl.Location.Y {type=MovieSceneScriptingFloatChannel, default=0}
       ring_02_r_ctrl.Location.Z {type=MovieSceneScriptingFloatChannel, default=0}
       ring_02_r_ctrl.Rotation.X {type=MovieSceneScriptingFloatChannel, default=0}
       ring_02_r_ctrl.Rotation.Y {type=MovieSceneScriptingFloatChannel, default=0}
       ring_02_r_ctrl.Rotation.Z {type=MovieSceneScriptingFloatChannel, default=0}
       ring_02_r_ctrl.Scale.X {type=MovieSceneScriptingFloatChannel, default=0}
       ring_02_r_ctrl.Scale.Y {type=MovieSceneScriptingFloatChannel, default=0}
       ring_02_r_ctrl.Scale.Z {type=MovieSceneScriptingFloatChannel, default=0}
       ring_03_r_ctrl.Location.X {type=MovieSceneScriptingFloatChannel, default=0}
       ring_03_r_ctrl.Location.Y {type=MovieSceneScriptingFloatChannel, default=0}
       ring_03_r_ctrl.Location.Z {type=MovieSceneScriptingFloatChannel, default=0}
       ring_03_r_ctrl.Rotation.X {type=MovieSceneScriptingFloatChannel, default=0}
       ring_03_r_ctrl.Rotation.Y {type=MovieSceneScriptingFloatChannel, default=0}
       ring_03_r_ctrl.Rotation.Z {type=MovieSceneScriptingFloatChannel, default=0}
       ring_03_r_ctrl.Scale.X {type=MovieSceneScriptingFloatChannel, default=0}
       ring_03_r_ctrl.Scale.Y {type=MovieSceneScriptingFloatChannel, default=0}
       ring_03_r_ctrl.Scale.Z {type=MovieSceneScriptingFloatChannel, default=0}
       pinky_metacarpal_r_ctrl.Location.X {type=MovieSceneScriptingFloatChannel, default=0}
       pinky_metacarpal_r_ctrl.Location.Y {type=MovieSceneScriptingFloatChannel, default=0}
       pinky_metacarpal_r_ctrl.Location.Z {type=MovieSceneScriptingFloatChannel, default=0}
       pinky_metacarpal_r_ctrl.Rotation.X {type=MovieSceneScriptingFloatChannel, default=0}
       pinky_metacarpal_r_ctrl.Rotation.Y {type=MovieSceneScriptingFloatChannel, default=0}
       pinky_metacarpal_r_ctrl.Rotation.Z {type=MovieSceneScriptingFloatChannel, default=0}
       pinky_metacarpal_r_ctrl.Scale.X {type=MovieSceneScriptingFloatChannel, default=0}
       pinky_metacarpal_r_ctrl.Scale.Y {type=MovieSceneScriptingFloatChannel, default=0}
       pinky_metacarpal_r_ctrl.Scale.Z {type=MovieSceneScriptingFloatChannel, default=0}
       pinky_01_r_ctrl.Location.X {type=MovieSceneScriptingFloatChannel, default=0}
       pinky_01_r_ctrl.Location.Y {type=MovieSceneScriptingFloatChannel, default=0}
       pinky_01_r_ctrl.Location.Z {type=MovieSceneScriptingFloatChannel, default=0}
       pinky_01_r_ctrl.Rotation.X {type=MovieSceneScriptingFloatChannel, default=0}
       pinky_01_r_ctrl.Rotation.Y {type=MovieSceneScriptingFloatChannel, default=0}
       pinky_01_r_ctrl.Rotation.Z {type=MovieSceneScriptingFloatChannel, default=0}
       pinky_01_r_ctrl.Scale.X {type=MovieSceneScriptingFloatChannel, default=0}
       pinky_01_r_ctrl.Scale.Y {type=MovieSceneScriptingFloatChannel, default=0}
       pinky_01_r_ctrl.Scale.Z {type=MovieSceneScriptingFloatChannel, default=0}
       pinky_02_r_ctrl.Location.X {type=MovieSceneScriptingFloatChannel, default=0}
       pinky_02_r_ctrl.Location.Y {type=MovieSceneScriptingFloatChannel, default=0}
       pinky_02_r_ctrl.Location.Z {type=MovieSceneScriptingFloatChannel, default=0}
       pinky_02_r_ctrl.Rotation.X {type=MovieSceneScriptingFloatChannel, default=0}
       pinky_02_r_ctrl.Rotation.Y {type=MovieSceneScriptingFloatChannel, default=0}
       pinky_02_r_ctrl.Rotation.Z {type=MovieSceneScriptingFloatChannel, default=0}
       pinky_02_r_ctrl.Scale.X {type=MovieSceneScriptingFloatChannel, default=0}
       pinky_02_r_ctrl.Scale.Y {type=MovieSceneScriptingFloatChannel, default=0}
       pinky_02_r_ctrl.Scale.Z {type=MovieSceneScriptingFloatChannel, default=0}
       pinky_03_r_ctrl.Location.X {type=MovieSceneScriptingFloatChannel, default=0}
       pinky_03_r_ctrl.Location.Y {type=MovieSceneScriptingFloatChannel, default=0}
       pinky_03_r_ctrl.Location.Z {type=MovieSceneScriptingFloatChannel, default=0}
       pinky_03_r_ctrl.Rotation.X {type=MovieSceneScriptingFloatChannel, default=0}
       pinky_03_r_ctrl.Rotation.Y {type=MovieSceneScriptingFloatChannel, default=0}
       pinky_03_r_ctrl.Rotation.Z {type=MovieSceneScriptingFloatChannel, default=0}
       pinky_03_r_ctrl.Scale.X {type=MovieSceneScriptingFloatChannel, default=0}
       pinky_03_r_ctrl.Scale.Y {type=MovieSceneScriptingFloatChannel, default=0}
       pinky_03_r_ctrl.Scale.Z {type=MovieSceneScriptingFloatChannel, default=0}
       hand_r_ik_ctrl.Location.X {type=MovieSceneScriptingFloatChannel, default=0}
       hand_r_ik_ctrl.Location.Y {type=MovieSceneScriptingFloatChannel, default=0}
       hand_r_ik_ctrl.Location.Z {type=MovieSceneScriptingFloatChannel, default=0}
       hand_r_ik_ctrl.Rotation.X {type=MovieSceneScriptingFloatChannel, default=0}
       hand_r_ik_ctrl.Rotation.Y {type=MovieSceneScriptingFloatChannel, default=0}
       hand_r_ik_ctrl.Rotation.Z {type=MovieSceneScriptingFloatChannel, default=0}
       hand_r_ik_ctrl.Scale.X {type=MovieSceneScriptingFloatChannel, default=0}
       hand_r_ik_ctrl.Scale.Y {type=MovieSceneScriptingFloatChannel, default=0}
       hand_r_ik_ctrl.Scale.Z {type=MovieSceneScriptingFloatChannel, default=0}
       Softness_2 {type=MovieSceneScriptingFloatChannel, default=0}
       hand_r_ikorient_ctrl.Rotation.Roll {type=MovieSceneScriptingFloatChannel, default=0}
       hand_r_ikorient_ctrl.Rotation.Pitch {type=MovieSceneScriptingFloatChannel, default=0}
       hand_r_ikorient_ctrl.Rotation.Yaw {type=MovieSceneScriptingFloatChannel, default=0}
       arm_r_pv_ik_ctrl.Location.X {type=MovieSceneScriptingFloatChannel, default=0}
       arm_r_pv_ik_ctrl.Location.Y {type=MovieSceneScriptingFloatChannel, default=0}
       arm_r_pv_ik_ctrl.Location.Z {type=MovieSceneScriptingFloatChannel, default=0}
       arm_r_pv_ik_ctrl.Rotation.X {type=MovieSceneScriptingFloatChannel, default=0}
       arm_r_pv_ik_ctrl.Rotation.Y {type=MovieSceneScriptingFloatChannel, default=0}
       arm_r_pv_ik_ctrl.Rotation.Z {type=MovieSceneScriptingFloatChannel, default=0}
       arm_r_pv_ik_ctrl.Scale.X {type=MovieSceneScriptingFloatChannel, default=0}
       arm_r_pv_ik_ctrl.Scale.Y {type=MovieSceneScriptingFloatChannel, default=0}
       arm_r_pv_ik_ctrl.Scale.Z {type=MovieSceneScriptingFloatChannel, default=0}
       body_ctrl.Location.X {type=MovieSceneScriptingFloatChannel, default=0}
       body_ctrl.Location.Y {type=MovieSceneScriptingFloatChannel, default=0}
       body_ctrl.Location.Z {type=MovieSceneScriptingFloatChannel, default=0}
       body_ctrl.Rotation.X {type=MovieSceneScriptingFloatChannel, default=0}
       body_ctrl.Rotation.Y {type=MovieSceneScriptingFloatChannel, default=0}
       body_ctrl.Rotation.Z {type=MovieSceneScriptingFloatChannel, default=0}
       body_ctrl.Scale.X {type=MovieSceneScriptingFloatChannel, default=0}
       body_ctrl.Scale.Y {type=MovieSceneScriptingFloatChannel, default=0}
       body_ctrl.Scale.Z {type=MovieSceneScriptingFloatChannel, default=0}
       hips_ctrl.Location.X {type=MovieSceneScriptingFloatChannel, default=0}
       hips_ctrl.Location.Y {type=MovieSceneScriptingFloatChannel, default=0}
       hips_ctrl.Location.Z {type=MovieSceneScriptingFloatChannel, default=0}
       hips_ctrl.Rotation.X {type=MovieSceneScriptingFloatChannel, default=0}
       hips_ctrl.Rotation.Y {type=MovieSceneScriptingFloatChannel, default=0}
       hips_ctrl.Rotation.Z {type=MovieSceneScriptingFloatChannel, default=0}
       hips_ctrl.Scale.X {type=MovieSceneScriptingFloatChannel, default=0}
       hips_ctrl.Scale.Y {type=MovieSceneScriptingFloatChannel, default=0}
       hips_ctrl.Scale.Z {type=MovieSceneScriptingFloatChannel, default=0}
       hips_tan_ctrl.Location.X {type=MovieSceneScriptingFloatChannel, default=0}
       hips_tan_ctrl.Location.Y {type=MovieSceneScriptingFloatChannel, default=0}
       hips_tan_ctrl.Location.Z {type=MovieSceneScriptingFloatChannel, default=0}
       hips_tan_ctrl.Rotation.X {type=MovieSceneScriptingFloatChannel, default=0}
       hips_tan_ctrl.Rotation.Y {type=MovieSceneScriptingFloatChannel, default=0}
       hips_tan_ctrl.Rotation.Z {type=MovieSceneScriptingFloatChannel, default=0}
       hips_tan_ctrl.Scale.X {type=MovieSceneScriptingFloatChannel, default=0}
       hips_tan_ctrl.Scale.Y {type=MovieSceneScriptingFloatChannel, default=0}
       hips_tan_ctrl.Scale.Z {type=MovieSceneScriptingFloatChannel, default=0}
       spine_01_ctrl.Location.X {type=MovieSceneScriptingFloatChannel, default=0}
       spine_01_ctrl.Location.Y {type=MovieSceneScriptingFloatChannel, default=0}
       spine_01_ctrl.Location.Z {type=MovieSceneScriptingFloatChannel, default=0}
       spine_01_ctrl.Rotation.X {type=MovieSceneScriptingFloatChannel, default=0}
       spine_01_ctrl.Rotation.Y {type=MovieSceneScriptingFloatChannel, default=0}
       spine_01_ctrl.Rotation.Z {type=MovieSceneScriptingFloatChannel, default=0}
       spine_01_ctrl.Scale.X {type=MovieSceneScriptingFloatChannel, default=0}
       spine_01_ctrl.Scale.Y {type=MovieSceneScriptingFloatChannel, default=0}
       spine_01_ctrl.Scale.Z {type=MovieSceneScriptingFloatChannel, default=0}
       spine_02_ctrl.Location.X {type=MovieSceneScriptingFloatChannel, default=0}
       spine_02_ctrl.Location.Y {type=MovieSceneScriptingFloatChannel, default=0}
       spine_02_ctrl.Location.Z {type=MovieSceneScriptingFloatChannel, default=0}
       spine_02_ctrl.Rotation.X {type=MovieSceneScriptingFloatChannel, default=0}
       spine_02_ctrl.Rotation.Y {type=MovieSceneScriptingFloatChannel, default=0}
       spine_02_ctrl.Rotation.Z {type=MovieSceneScriptingFloatChannel, default=0}
       spine_02_ctrl.Scale.X {type=MovieSceneScriptingFloatChannel, default=0}
       spine_02_ctrl.Scale.Y {type=MovieSceneScriptingFloatChannel, default=0}
       spine_02_ctrl.Scale.Z {type=MovieSceneScriptingFloatChannel, default=0}
       spine_03_ctrl.Location.X {type=MovieSceneScriptingFloatChannel, default=0}
       spine_03_ctrl.Location.Y {type=MovieSceneScriptingFloatChannel, default=0}
       spine_03_ctrl.Location.Z {type=MovieSceneScriptingFloatChannel, default=0}
       spine_03_ctrl.Rotation.X {type=MovieSceneScriptingFloatChannel, default=0}
       spine_03_ctrl.Rotation.Y {type=MovieSceneScriptingFloatChannel, default=0}
       spine_03_ctrl.Rotation.Z {type=MovieSceneScriptingFloatChannel, default=0}
       spine_03_ctrl.Scale.X {type=MovieSceneScriptingFloatChannel, default=0}
       spine_03_ctrl.Scale.Y {type=MovieSceneScriptingFloatChannel, default=0}
       spine_03_ctrl.Scale.Z {type=MovieSceneScriptingFloatChannel, default=0}
       spine_04_ctrl.Location.X {type=MovieSceneScriptingFloatChannel, default=0}
       spine_04_ctrl.Location.Y {type=MovieSceneScriptingFloatChannel, default=0}
       spine_04_ctrl.Location.Z {type=MovieSceneScriptingFloatChannel, default=0}
       spine_04_ctrl.Rotation.X {type=MovieSceneScriptingFloatChannel, default=0}
       spine_04_ctrl.Rotation.Y {type=MovieSceneScriptingFloatChannel, default=0}
       spine_04_ctrl.Rotation.Z {type=MovieSceneScriptingFloatChannel, default=0}
       spine_04_ctrl.Scale.X {type=MovieSceneScriptingFloatChannel, default=0}
       spine_04_ctrl.Scale.Y {type=MovieSceneScriptingFloatChannel, default=0}
       spine_04_ctrl.Scale.Z {type=MovieSceneScriptingFloatChannel, default=0}
       head_ctrl.Location.X {type=MovieSceneScriptingFloatChannel, default=0}
       head_ctrl.Location.Y {type=MovieSceneScriptingFloatChannel, default=0}
       head_ctrl.Location.Z {type=MovieSceneScriptingFloatChannel, default=0}
       head_ctrl.Rotation.X {type=MovieSceneScriptingFloatChannel, default=0}
       head_ctrl.Rotation.Y {type=MovieSceneScriptingFloatChannel, default=0}
       head_ctrl.Rotation.Z {type=MovieSceneScriptingFloatChannel, default=0}
       head_ctrl.Scale.X {type=MovieSceneScriptingFloatChannel, default=0}
       head_ctrl.Scale.Y {type=MovieSceneScriptingFloatChannel, default=0}
       head_ctrl.Scale.Z {type=MovieSceneScriptingFloatChannel, default=0}
       clavicle_l_ctrl.Location.X {type=MovieSceneScriptingFloatChannel, default=0}
       clavicle_l_ctrl.Location.Y {type=MovieSceneScriptingFloatChannel, default=0}
       clavicle_l_ctrl.Location.Z {type=MovieSceneScriptingFloatChannel, default=0}
       clavicle_l_ctrl.Rotation.X {type=MovieSceneScriptingFloatChannel, default=0}
       clavicle_l_ctrl.Rotation.Y {type=MovieSceneScriptingFloatChannel, default=0}
       clavicle_l_ctrl.Rotation.Z {type=MovieSceneScriptingFloatChannel, default=0}
       clavicle_l_ctrl.Scale.X {type=MovieSceneScriptingFloatChannel, default=0}
       clavicle_l_ctrl.Scale.Y {type=MovieSceneScriptingFloatChannel, default=0}
       clavicle_l_ctrl.Scale.Z {type=MovieSceneScriptingFloatChannel, default=0}
       clavicle_r_ctrl.Location.X {type=MovieSceneScriptingFloatChannel, default=0}
       clavicle_r_ctrl.Location.Y {type=MovieSceneScriptingFloatChannel, default=0}
       clavicle_r_ctrl.Location.Z {type=MovieSceneScriptingFloatChannel, default=0}
       clavicle_r_ctrl.Rotation.X {type=MovieSceneScriptingFloatChannel, default=0}
       clavicle_r_ctrl.Rotation.Y {type=MovieSceneScriptingFloatChannel, default=0}
       clavicle_r_ctrl.Rotation.Z {type=MovieSceneScriptingFloatChannel, default=0}
       clavicle_r_ctrl.Scale.X {type=MovieSceneScriptingFloatChannel, default=0}
       clavicle_r_ctrl.Scale.Y {type=MovieSceneScriptingFloatChannel, default=0}
       clavicle_r_ctrl.Scale.Z {type=MovieSceneScriptingFloatChannel, default=0}
       neck_01_ctrl.Location.X {type=MovieSceneScriptingFloatChannel, default=0}
       neck_01_ctrl.Location.Y {type=MovieSceneScriptingFloatChannel, default=0}
       neck_01_ctrl.Location.Z {type=MovieSceneScriptingFloatChannel, default=0}
       neck_01_ctrl.Rotation.X {type=MovieSceneScriptingFloatChannel, default=0}
       neck_01_ctrl.Rotation.Y {type=MovieSceneScriptingFloatChannel, default=0}
       neck_01_ctrl.Rotation.Z {type=MovieSceneScriptingFloatChannel, default=0}
       neck_01_ctrl.Scale.X {type=MovieSceneScriptingFloatChannel, default=0}
       neck_01_ctrl.Scale.Y {type=MovieSceneScriptingFloatChannel, default=0}
       neck_01_ctrl.Scale.Z {type=MovieSceneScriptingFloatChannel, default=0}
       neck_02_ctrl.Location.X {type=MovieSceneScriptingFloatChannel, default=0}
       neck_02_ctrl.Location.Y {type=MovieSceneScriptingFloatChannel, default=0}
       neck_02_ctrl.Location.Z {type=MovieSceneScriptingFloatChannel, default=0}
       neck_02_ctrl.Rotation.X {type=MovieSceneScriptingFloatChannel, default=0}
       neck_02_ctrl.Rotation.Y {type=MovieSceneScriptingFloatChannel, default=0}
       neck_02_ctrl.Rotation.Z {type=MovieSceneScriptingFloatChannel, default=0}
       neck_02_ctrl.Scale.X {type=MovieSceneScriptingFloatChannel, default=0}
       neck_02_ctrl.Scale.Y {type=MovieSceneScriptingFloatChannel, default=0}
       neck_02_ctrl.Scale.Z {type=MovieSceneScriptingFloatChannel, default=0}
       head_ik_ctrl.Location.X {type=MovieSceneScriptingFloatChannel, default=0}
       head_ik_ctrl.Location.Y {type=MovieSceneScriptingFloatChannel, default=0}
       head_ik_ctrl.Location.Z {type=MovieSceneScriptingFloatChannel, default=0}
       head_ik_ctrl.Rotation.X {type=MovieSceneScriptingFloatChannel, default=0}
       head_ik_ctrl.Rotation.Y {type=MovieSceneScriptingFloatChannel, default=0}
       head_ik_ctrl.Rotation.Z {type=MovieSceneScriptingFloatChannel, default=0}
       head_ik_ctrl.Scale.X {type=MovieSceneScriptingFloatChannel, default=0}
       head_ik_ctrl.Scale.Y {type=MovieSceneScriptingFloatChannel, default=0}
       head_ik_ctrl.Scale.Z {type=MovieSceneScriptingFloatChannel, default=0}
       chest_ctrl.Location.X {type=MovieSceneScriptingFloatChannel, default=0}
       chest_ctrl.Location.Y {type=MovieSceneScriptingFloatChannel, default=0}
       chest_ctrl.Location.Z {type=MovieSceneScriptingFloatChannel, default=0}
       chest_ctrl.Rotation.X {type=MovieSceneScriptingFloatChannel, default=0}
       chest_ctrl.Rotation.Y {type=MovieSceneScriptingFloatChannel, default=0}
       chest_ctrl.Rotation.Z {type=MovieSceneScriptingFloatChannel, default=0}
       chest_ctrl.Scale.X {type=MovieSceneScriptingFloatChannel, default=0}
       chest_ctrl.Scale.Y {type=MovieSceneScriptingFloatChannel, default=0}
       chest_ctrl.Scale.Z {type=MovieSceneScriptingFloatChannel, default=0}
       chest_tan_ctrl.Location.X {type=MovieSceneScriptingFloatChannel, default=0}
       chest_tan_ctrl.Location.Y {type=MovieSceneScriptingFloatChannel, default=0}
       chest_tan_ctrl.Location.Z {type=MovieSceneScriptingFloatChannel, default=0}
       chest_tan_ctrl.Rotation.X {type=MovieSceneScriptingFloatChannel, default=0}
       chest_tan_ctrl.Rotation.Y {type=MovieSceneScriptingFloatChannel, default=0}
       chest_tan_ctrl.Rotation.Z {type=MovieSceneScriptingFloatChannel, default=0}
       chest_tan_ctrl.Scale.X {type=MovieSceneScriptingFloatChannel, default=0}
       chest_tan_ctrl.Scale.Y {type=MovieSceneScriptingFloatChannel, default=0}
       chest_tan_ctrl.Scale.Z {type=MovieSceneScriptingFloatChannel, default=0}
      ]
     }
    ]
   }
  ]
 }]
display_rate {denominator=1, numerator=30}
tick_resolution {denominator=1, numerator=24000}
tracks []
